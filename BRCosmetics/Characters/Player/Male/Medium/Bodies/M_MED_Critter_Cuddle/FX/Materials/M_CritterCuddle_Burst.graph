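# Unlit translucent Niagara sprite material for an electricity burst effect on the Critter Cuddle character cosmetic.
# Subtracting particle relative time from a grayscale electricity texture erodes the pattern over each particle's life.
# The saturated result, raised to Color's alpha as exponent, blends two color parameters into the shared emissive adjustments function.
# It also builds radial-distance and particle-alpha variants of the erosion, plus a parameterized camera offset.
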Material /BRCosmetics/Characters/Player/Male/Medium/Bodies/M_MED_Critter_Cuddle/FX/Materials/M_CritterCuddle_Burst
graph {
 properties {bUsedWithNiagaraSprites=true, blendMode=BLEND_Translucent, pixelDepthOffsetMode=PDOM_Legacy, shadingModel=MSM_Unlit, usedShadingModels=Unlit}
 nodes [
  n1 = TextureSample() {
   samplerType=SAMPLERTYPE_LinearGrayscale
   texture=/Game/Characters/Player/Male/Medium/Bodies/M_MED_Critter_Cuddle/FX/Textures/T_CritterCuddle_Electricity
  }
  n2 = ParticleRelativeTime()
  n3 = Subtract(A=@n1.R, B=@n2)
  n4 = Multiply(A=Distance(A=TextureCoordinate(), B=Constant() {r=0.5})) {constB=2}
  Divide(A=@n3, B=SquareRoot(None=Saturate(None=@n4)))
  n5 = VectorParameter() {parameterName=Color, defaultValue=<rgba 2 0.1 1.145 4>}
  n6 = Saturate(None=@n3)
  MaterialFunctionCall(Emissive=LinearInterpolate(A=@n5.RGB, B=VectorParameter() {parameterName=Color B, defaultValue=<rgba 25.6667 0 44 0>}.RGB, Alpha=Power(Base=@n6, Exp=@n5.A))) {materialFunction=&MF_FX_EmissiveAdjustments}
  MaterialFunctionCall(Offset Amount=ScalarParameter() {defaultValue=22, parameterName=CameraOffset}) {materialFunction=/Engine/Functions/Engine_MaterialFunctions02/WorldPositionOffset/CameraOffset}
  Multiply(A=@n3, B=ParticleColor().A)
  Multiply(A=@n1.R, B=@n1.R)
  SquareRoot(None=@n2)
 ]
}
