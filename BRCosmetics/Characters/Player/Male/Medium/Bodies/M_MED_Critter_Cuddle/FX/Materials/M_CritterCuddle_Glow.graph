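# Unlit additive Niagara sprite material for a soft round glow in the Critter Cuddle character body's effects.
# A radial gradient from UV distance to center blends Color at the edge into Color B at the core, with Color's alpha as falloff exponent, then feeds MF_FX_EmissiveAdjustments.
# The gradient also forms a mask scaled by Color B alpha and particle alpha, fading over particle lifetime.
# A CameraOffset function provides depth offsetting.
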Material /BRCosmetics/Characters/Player/Male/Medium/Bodies/M_MED_Critter_Cuddle/FX/Materials/M_CritterCuddle_Glow
graph {
 properties {bUsedWithNiagaraSprites=true, blendMode=BLEND_Additive, pixelDepthOffsetMode=PDOM_Legacy, shadingModel=MSM_Unlit, usedShadingModels=Unlit}
 nodes [
  n1 = VectorParameter() {parameterName=Color, defaultValue=<rgba 2 0.1 1.145 4>}
  n2 = VectorParameter() {parameterName=Color B, defaultValue=<rgba 25.6667 0 44 0.045>}
  n3 = Distance(A=TextureCoordinate(), B=Constant() {r=0.5})
  n4 = OneMinus(None=Saturate(None=Multiply(A=@n3) {constB=2}))
  MaterialFunctionCall(Emissive=LinearInterpolate(A=@n1.RGB, B=@n2.RGB, Alpha=Power(Base=@n4, Exp=@n1.A))) {materialFunction=&MF_FX_EmissiveAdjustments}
  MaterialFunctionCall(Offset Amount=ScalarParameter() {defaultValue=66, parameterName=CameraOffset}) {materialFunction=/Engine/Functions/Engine_MaterialFunctions02/WorldPositionOffset/CameraOffset}
  n5 = ParticleColor()
  Multiply(A=Multiply(A=@n2.A, B=Multiply(A=@n4, B=@n5.A)), B=OneMinus(None=ParticleRelativeTime()))
  n6 = Subtract(A=TextureCoordinate()) {constB=0.5}
  Normalize(VectorInput=@n6)
  Saturate(None=Divide(A=@n6) {constB=0.01})
 ]
}
