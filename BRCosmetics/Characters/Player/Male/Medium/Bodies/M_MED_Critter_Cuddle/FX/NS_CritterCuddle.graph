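# Niagara cosmetic effect for the CritterCuddle character body that flashes brief magenta arcs with glow on the skeletal mesh.
# A looping burst always fires in the front end, while in gameplay a camera-distance check gates it.
# Loop count picks one of three bolt origins from an array, and particles follow that bone or socket as alpha fades.
# Visibility tags switch mesh, sprite and light renderer variants, with special handling when the head is hidden.
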NiagaraSystem /BRCosmetics/Characters/Player/Male/Medium/Bodies/M_MED_Critter_Cuddle/FX/NS_CritterCuddle
dependencies {
 usedDataInterfaces [&NiagaraDataInterfaceCamera, &NiagaraDataInterfaceArrayInt32, &NiagaraDataInterfaceCurve]
 usedDynamicInputs []
 usedModules [&SystemState]
}
emitters {Arcs {
  emitterUpdateScript [
   &EmitterState()
   &SampleSkeletalMesh(
    AdditionalYawPitchRoll=false
    Bone / Socket Index=3
    Bone Sampling Mode:ENiagaraBoneSamplingMode=::NewEnumerator4
    Skeletal Mesh:NiagaraDataInterfaceSkeletalMesh=$User.SkeletalMesh
   )
   &SpawnBurst_Instantaneous(
    Spawn Probability=&FN_FrontEndFloatModifier(
     Gameplay=&MakeCustomFloatFromBool(
      Bool=&SetBoolByFloatComparison(
       A=&VectorLength(Vector=&Subtract_Vector(
         A=$Emitter.SampleSkeletalMesh.SampledPosition
         B=&ReturnCameraVector(Camera Property=::NewEnumerator0, Coordinate Space=::Simulation)
        ))
       B=175
      )
      True Float=1
      False Float=0
     )
     Front End=1
    )
    Use Spawn Probability=true
    UseLoopCountLimit=false
   )
  ]
  particleSpawnScript [
   &InitializeParticle(
    Color=<rgba 11 0 8.6167 0.2>
    Lifetime=&RandomRangeFloat(Minimum=0.2, Maximum=0.4)
    Position Mode=::NewEnumerator0
    Sprite Rotation Mode=::NewEnumerator2
    Sprite Size Mode=::NewEnumerator3
    Uniform Sprite Size=22
    Write Material Random=true
   )
   /SetVariables(
    Particles.BoltOrigin:NiagaraInt32=&SelectIntFromArray(
     Direct Array Index=$System.LoopCount
     Direct Array Mode=::NewEnumerator1
     Int Selection Array=DataInterface() {propertyValues {IntData [0,2,1,2,0,1,0,1,2,0,2,1]}}
    )
    Particles.GlowCameraOffset=22
    Particles.GlowSpriteSize:Vector2f=&Vector2DFromFloat(Value=99)
    Particles.MeshIndex=&RandomRangeInt(Minimum=0, Maximum=1)
    Particles.RandomVisibilityTag:NiagaraInt32=&RandomRangeInt(Minimum=1, Maximum=2)
    Particles.SubImageIndexFlipbook:NiagaraFloat=&RandomRangeFloat(Minimum=0, Maximum=4)
   ) {isSetParameters=true}
  ]
  particleUpdateScript [
   &ParticleState()
   &SkeletalMeshLocation(
    Bone / Socket Index=$Particles.BoltOrigin
    MaskSpawnGroup=false
    Mesh Sampling Type=::NewEnumerator5
    Orientation Sampling=::NewEnumerator1
    Sampled Position Offset=<22 0 0>
    Skeletal Mesh=$User.SkeletalMesh
    Socket Sampling Mode=::NewEnumerator5
    Use Sampled Mesh Attribute as Ribbon ID=false
   )
   &SkeletalMeshLocation(
    Bone / Socket Index=$Particles.BoltOrigin
    Bone Sampling Mode=::NewEnumerator4
    MaskSpawnGroup=false
    Orientation Sampling=::NewEnumerator1
    Position Sampling=::NewEnumerator2
    Skeletal Mesh=$User.SkeletalMesh
   ) {module=SkeletalMeshLocation001}
   &ScaleColor(
    Scale Alpha=&FloatFromCurve(FloatCurve=DataInterface() {propertyValues {Curve {keys [<at 0 1>,<at 1 0>]}}})
    ScaleA=true
    ScaleRGB=true
   )
   /SetVariables(
    Particles.RandomVisibilityTag:NiagaraInt32=&MakeCustomIntFromBool(
     Bool=&SetBoolByIntComparison(A=$Particles.BoltOrigin, B=0, Comparison Type=::NewEnumerator4)
     True INT=&MakeCustomIntFromBool(Bool=$User.bisHeadHidden, True INT=5, False INT=$Particles.Initial.RandomVisibilityTag)
     False INT=$Particles.Initial.RandomVisibilityTag
    )
    Particles.VisibilityTag=&MakeCustomIntFromBool(
     Bool=&SetBoolByIntComparison(A=$Particles.BoltOrigin, B=0, Comparison Type=::NewEnumerator4)
     True INT=&MakeCustomIntFromBool(Bool=$User.bisHeadHidden, True INT=1, False INT=0)
     False INT=0
    )
   ) {isSetParameters=true}
  ]
  rendererClasses [
   &NiagaraMeshRendererProperties
   &NiagaraLightRendererProperties
   &NiagaraSpriteRendererProperties
  ]
  renderers [
   {rendererClass=&NiagaraMeshRendererProperties, rendererIndex=0}
   {rendererClass=&NiagaraMeshRendererProperties, rendererIndex=1}
   {rendererClass=&NiagaraLightRendererProperties, rendererIndex=2}
   {rendererClass=&NiagaraSpriteRendererProperties, rendererIndex=3}
   {rendererClass=&NiagaraSpriteRendererProperties, rendererIndex=4}
  ]
 }}
summary {
 systemName=NS_CritterCuddle
 emitters [{
   emitterName=Arcs
   rendererClasses [
    &NiagaraMeshRendererProperties
    &NiagaraLightRendererProperties
    &NiagaraSpriteRendererProperties
   ]
  }]
 userVariables [
  User.bIsFrontEnd {type=NiagaraBool}
  User.bIsFrontEndPreview {type=NiagaraBool}
  User.bisHeadHidden {type=NiagaraBool}
  User.SkeletalMesh {
   type=NiagaraDataInterfaceSkeletalMesh
   defaultValue {
    struct=NiagaraExt_VariableValue_DataInterface
    value {dataInterface=&self:NiagaraDataInterfaceSkeletalMesh_0}
   }
  }
 ]
}
