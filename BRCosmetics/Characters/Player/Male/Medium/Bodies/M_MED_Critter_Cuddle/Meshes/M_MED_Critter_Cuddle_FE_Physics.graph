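# Physics asset for the Critter Cuddle medium male body cosmetic, simulating its dangling leaf bones on the character's right side.
# Capsule bodies on the pelvis and right clavicle, upper arm, thigh and calf form the base, linked by free-rotation constraints.
# Leaf chains, some forking from shared roots, hang from the clavicle, upper arm, pelvis and calf, with tight uniform swing and twist limits.
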PhysicsAsset /BRCosmetics/Characters/Player/Male/Medium/Bodies/M_MED_Critter_Cuddle/Meshes/M_MED_Critter_Cuddle_FE_Physics
bodies {
 calf_r [{shapeType=Capsule, length=27.8068, radius=5.3786, shapeName=calf_r_capsule, center=<19.1368 1.6309 0>, rotation=<rot 90 0 0>}]
 clavicle_r [
  {shapeType=Capsule, length=10.0392, radius=5.6671, shapeName=clavicle_r_capsule, center=<-10.6106 0.0002 0>, rotation=<rot 90 0 0>}
  {shapeType=Capsule, length=10.0392, radius=5.6671, shapeName=clavicle_r_capsule, center=<-8.2246 5.0542 1.8949>, rotation=<rot 90 0 0>}
 ]
 dyn_leaf_a_01 [{
   shapeType=Capsule
   length=3.3159
   radius=1.5951
   shapeName=dyn_leaf_a_01_capsule
   center=<5.0248 -0.7588 -0.0739>
   rotation=<rot 0 -118.3064 -89.3648>
  }]
 dyn_leaf_a_02 [{
   shapeType=Capsule
   length=2.1471
   radius=1.5951
   shapeName=dyn_leaf_a_02_capsule
   center=<2.1378 -0.7713 0.2272>
   rotation=<rot 0 -96.6108 -87.6974>
  }]
 dyn_leaf_a_03 [{
   shapeType=Capsule
   length=0
   radius=1.5951
   shapeName=dyn_leaf_a_03_capsule
   center=<0.0006 -0.3999 0.0345>
   rotation=<rot 0 128.8418 -78.1144>
  }]
 dyn_leaf_b_00 [{
   shapeType=Capsule
   length=1.8286
   radius=2
   shapeName=dyn_leaf_b_00_capsule
   center=<4.2112 0.0627 -0.9949>
   rotation=<rot 0 -74.1983 -81.7948>
  }]
 dyn_leaf_b_01 [{shapeType=Capsule, length=4.0029, radius=2, shapeName=dyn_leaf_b_01_capsule, center=<5.68 0.0479 0.2865>, rotation=<rot 0 -84.1665 -88.9759>}]
 dyn_leaf_b_02 [{
   shapeType=Capsule
   length=0.9748
   radius=2
   shapeName=dyn_leaf_b_02_capsule
   center=<2.6458 0.6917 0.2392>
   rotation=<rot 0 118.521 -64.7055>
  }]
 dyn_leaf_bc_01 [{
   shapeType=Capsule
   length=0.8918
   radius=2
   shapeName=dyn_leaf_bc_01_capsule
   center=<3.6172 2.056 -0.0057>
   rotation=<rot 0 -74.3975 -88.1923>
  }]
 dyn_leaf_c_01 [{
   shapeType=Capsule
   length=6.0737
   radius=2
   shapeName=dyn_leaf_c_01_capsule
   center=<0.7581 -0.872 -0.908>
   rotation=<rot 0 99.5478 -78.3214>
  }]
 dyn_leaf_c_02 [{
   shapeType=Capsule
   length=4.899
   radius=2
   shapeName=dyn_leaf_c_02_capsule
   center=<6.1778 -0.4349 -1.3438>
   rotation=<rot 0 -93.177 -84.5705>
  }]
 dyn_leaf_d_01 [{
   shapeType=Capsule
   length=3.3796
   radius=1.5951
   shapeName=dyn_leaf_d_01_capsule
   center=<4.7567 0.1938 -0.1906>
   rotation=<rot 0 -115.8761 -88.5755>
  }]
 dyn_leaf_d_02 [{
   shapeType=Capsule
   length=4.6211
   radius=1.5951
   shapeName=dyn_leaf_d_02_capsule
   center=<2.9241 -0.1761 -0.1343>
   rotation=<rot 0 55.8323 -87.8452>
  }]
 dyn_leaf_d_03 [{
   shapeType=Capsule
   length=2.2697
   radius=1.5951
   shapeName=dyn_leaf_d_03_capsule
   center=<2.8406 -0.0389 0.0337>
   rotation=<rot 0 86.6656 -76.8933>
  }]
 dyn_leaf_e_01 [{
   shapeType=Capsule
   length=2.9329
   radius=1.5951
   shapeName=dyn_leaf_e_01_capsule
   center=<5.517 0.3879 0.0721>
   rotation=<rot 0.9545 -96.174 -89.9999>
  }]
 dyn_leaf_e_02 [{
   shapeType=Capsule
   length=3.3803
   radius=1.5951
   shapeName=dyn_leaf_e_02_capsule
   center=<3.6722 1.011 -0.2712>
   rotation=<rot 0 101.1207 -76.49>
  }]
 dyn_leaf_ef_01 [{
   shapeType=Capsule
   length=4.03
   radius=1.5951
   shapeName=dyn_leaf_ef_01_capsule
   center=<5.9067 -0.4716 -1.469>
   rotation=<rot 0 -108.2414 -73.9669>
  }]
 dyn_leaf_f_01 [{
   shapeType=Capsule
   length=3.9289
   radius=1.5951
   shapeName=dyn_leaf_f_01_capsule
   center=<1.1873 -1.431 -0.1587>
   rotation=<rot 0 122.6621 -87.7471>
  }]
 dyn_leaf_f_02 [{
   shapeType=Capsule
   length=1.8128
   radius=1.5951
   shapeName=dyn_leaf_f_02_capsule
   center=<1.7034 -0.2424 -0.2393>
   rotation=<rot 0 101.6679 -89.7587>
  }]
 dyn_leaf_g_01 [{
   shapeType=Capsule
   length=2.6753
   radius=1.5951
   shapeName=dyn_leaf_g_01_capsule
   center=<4.2958 1.1026 -0.2309>
   rotation=<rot 0 43.1309 -84.3638>
  }]
 dyn_leaf_g_02 [{
   shapeType=Capsule
   length=2.7126
   radius=1.5951
   shapeName=dyn_leaf_g_02_capsule
   center=<3.453 1.4303 0.1444>
   rotation=<rot 0 -6.5286 -85.8063>
  }]
 dyn_leaf_g_03 [{
   shapeType=Capsule
   length=4.4863
   radius=1.5951
   shapeName=dyn_leaf_g_03_capsule
   center=<5.6627 -1.6546 -1.2273>
   rotation=<rot 0 -104.0159 -81.4905>
  }]
 dyn_leaf_h_01 [{
   shapeType=Capsule
   length=3.7897
   radius=1.5951
   shapeName=dyn_leaf_h_01_capsule
   center=<5.5977 -0.853 -0.0561>
   rotation=<rot 0 -102.1565 -86.9994>
  }]
 dyn_leaf_h_02 [{
   shapeType=Capsule
   length=1.8769
   radius=1.5951
   shapeName=dyn_leaf_h_02_capsule
   center=<2.6832 -0.2982 -1.025>
   rotation=<rot 0 93.5894 -88.001>
  }]
 dyn_leaf_hi_01 [{
   shapeType=Capsule
   length=3.7231
   radius=1.5951
   shapeName=dyn_leaf_hi_01_capsule
   center=<4.698 -2.8849 -1.2468>
   rotation=<rot 17.7461 123.1341 -89.9999>
  }]
 dyn_leaf_i_01 [{
   shapeType=Capsule
   length=1.3884
   radius=1.5951
   shapeName=dyn_leaf_i_01_capsule
   center=<4.0971 0.0454 0.2853>
   rotation=<rot 0.259 -91.922 -89.9999>
  }]
 dyn_leaf_i_02 [{
   shapeType=Capsule
   length=4.6234
   radius=1.5951
   shapeName=dyn_leaf_i_02_capsule
   center=<1.7676 0.1388 0.0299>
   rotation=<rot 0 97.0718 -81.4903>
  }]
 dyn_r_calfLeafe_a_01 [{
   shapeType=Capsule
   length=2.7717
   radius=1.5951
   shapeName=dyn_r_calfLeafe_a_01_capsule
   center=<2.9431 -0.0061 -0.3676>
   rotation=<rot 0 -87.3553 -80.6158>
  }]
 dyn_r_calfLeafe_a_02 [{
   shapeType=Capsule
   length=2.5129
   radius=1.5951
   shapeName=dyn_r_calfLeafe_a_02_capsule
   center=<1.9622 0.3628 -0.6532>
   rotation=<rot 0 -96.2623 -77.3241>
  }]
 dyn_r_calfLeafe_b_01 [{shapeType=Capsule, length=3.1851, radius=1.5951, shapeName=dyn_r_calfLeafe_b_01_capsule, center=<3.0955 0 -0.0001>, rotation=<rot 90 0 0>}]
 dyn_r_calfLeafe_b_02 [{shapeType=Capsule, length=3.4844, radius=1.5951, shapeName=dyn_r_calfLeafe_b_02_capsule, center=<3.0952 0 0>, rotation=<rot 90 0 0>}]
 dyn_r_calfLeafe_c_01 [{shapeType=Capsule, length=3.0499, radius=1.5951, shapeName=dyn_r_calfLeafe_c_01_capsule, center=<3.1906 0 0>, rotation=<rot 90 0 0>}]
 dyn_r_calfLeafe_c_02 [{shapeType=Capsule, length=3.0888, radius=1.5951, shapeName=dyn_r_calfLeafe_c_02_capsule, center=<2.8053 0 0>, rotation=<rot 90 0 0>}]
 dyn_r_legLeaf_01 [{shapeType=Capsule, length=2.0897, radius=2.2421, shapeName=dyn_r_legLeaf_01_capsule, center=<2.3782 0 2>, rotation=<rot 90 0 0>}]
 pelvis [{shapeType=Capsule, length=11.8776, radius=10.6152}]
 thigh_r [{shapeType=Capsule, length=32.5559, radius=6.8066, shapeName=thigh_r_capsule, center=<18.6713 0 -0.1335>, rotation=<rot 90 0 -2.8354>}]
 upperarm_r [{shapeType=Capsule, length=18.056, radius=4.4375, shapeName=upperarm_r_capsule, center=<-11.9044 0 0>, rotation=<rot 90 0 0>}]
}
constraints [
 {bone1Name=dyn_leaf_bc_01, bone2Name=upperarm_r, swing1LimitDegrees=15, swing2LimitDegrees=15, twistLimitDegrees=5}
 {bone1Name=dyn_leaf_c_01, bone2Name=dyn_leaf_bc_01, swing1LimitDegrees=15, swing2LimitDegrees=15, twistLimitDegrees=5}
 {bone1Name=dyn_leaf_c_02, bone2Name=dyn_leaf_c_01, swing1LimitDegrees=15, swing2LimitDegrees=15, twistLimitDegrees=5}
 {bone1Name=dyn_leaf_d_01, bone2Name=clavicle_r, swing1LimitDegrees=15, swing2LimitDegrees=15, twistLimitDegrees=5}
 {bone1Name=dyn_leaf_d_02, bone2Name=dyn_leaf_d_01, swing1LimitDegrees=15, swing2LimitDegrees=15, twistLimitDegrees=5}
 {bone1Name=dyn_leaf_d_03, bone2Name=dyn_leaf_d_02, swing1LimitDegrees=15, swing2LimitDegrees=15, twistLimitDegrees=5}
 {bone1Name=dyn_leaf_ef_01, bone2Name=clavicle_r, swing1LimitDegrees=15, swing2LimitDegrees=15, twistLimitDegrees=5}
 {bone1Name=dyn_leaf_f_01, bone2Name=dyn_leaf_ef_01, swing1LimitDegrees=15, swing2LimitDegrees=15, twistLimitDegrees=5}
 {bone1Name=dyn_leaf_f_02, bone2Name=dyn_leaf_f_01, swing1LimitDegrees=15, swing2LimitDegrees=15, twistLimitDegrees=5}
 {bone1Name=dyn_leaf_e_01, bone2Name=dyn_leaf_ef_01, swing1LimitDegrees=15, swing2LimitDegrees=15, twistLimitDegrees=5}
 {bone1Name=dyn_leaf_e_02, bone2Name=dyn_leaf_e_01, swing1LimitDegrees=15, swing2LimitDegrees=15, twistLimitDegrees=5}
 {bone1Name=dyn_leaf_g_01, bone2Name=clavicle_r, swing1LimitDegrees=15, swing2LimitDegrees=15, twistLimitDegrees=5}
 {bone1Name=dyn_leaf_g_02, bone2Name=dyn_leaf_g_01, swing1LimitDegrees=15, swing2LimitDegrees=15, twistLimitDegrees=5}
 {bone1Name=dyn_leaf_g_03, bone2Name=dyn_leaf_g_02, swing1LimitDegrees=15, swing2LimitDegrees=15, twistLimitDegrees=5}
 {bone1Name=dyn_leaf_a_01, bone2Name=clavicle_r, swing1LimitDegrees=15, swing2LimitDegrees=15, twistLimitDegrees=5}
 {bone1Name=dyn_leaf_a_02, bone2Name=dyn_leaf_a_01, swing1LimitDegrees=15, swing2LimitDegrees=15, twistLimitDegrees=5}
 {bone1Name=dyn_leaf_a_03, bone2Name=dyn_leaf_a_02, swing1LimitDegrees=15, swing2LimitDegrees=15, twistLimitDegrees=5}
 {bone1Name=dyn_leaf_b_00, bone2Name=clavicle_r, swing1LimitDegrees=15, swing2LimitDegrees=15, twistLimitDegrees=5}
 {bone1Name=dyn_leaf_b_01, bone2Name=dyn_leaf_b_00, swing1LimitDegrees=15, swing2LimitDegrees=15, twistLimitDegrees=5}
 {bone1Name=dyn_leaf_b_02, bone2Name=dyn_leaf_b_01, swing1LimitDegrees=15, swing2LimitDegrees=15, twistLimitDegrees=5}
 {bone1Name=dyn_leaf_hi_01, bone2Name=pelvis, swing1LimitDegrees=15, swing2LimitDegrees=15, twistLimitDegrees=5}
 {bone1Name=dyn_leaf_i_01, bone2Name=dyn_leaf_hi_01, swing1LimitDegrees=15, swing2LimitDegrees=15, twistLimitDegrees=5}
 {bone1Name=dyn_leaf_i_02, bone2Name=dyn_leaf_i_01, swing1LimitDegrees=15, swing2LimitDegrees=15, twistLimitDegrees=5}
 {bone1Name=dyn_leaf_h_01, bone2Name=dyn_leaf_hi_01, swing1LimitDegrees=15, swing2LimitDegrees=15, twistLimitDegrees=5}
 {bone1Name=dyn_leaf_h_02, bone2Name=dyn_leaf_h_01, swing1LimitDegrees=15, swing2LimitDegrees=15, twistLimitDegrees=5}
 {bone1Name=dyn_r_calfLeafe_a_01, bone2Name=calf_r, swing1LimitDegrees=15, swing2LimitDegrees=15, twistLimitDegrees=5}
 {bone1Name=dyn_r_calfLeafe_a_02, bone2Name=dyn_r_calfLeafe_a_01, swing1LimitDegrees=15, swing2LimitDegrees=15, twistLimitDegrees=5}
 {bone1Name=dyn_r_calfLeafe_b_01, bone2Name=calf_r, swing1LimitDegrees=15, swing2LimitDegrees=15, twistLimitDegrees=5}
 {bone1Name=dyn_r_calfLeafe_b_02, bone2Name=dyn_r_calfLeafe_b_01, swing1LimitDegrees=15, swing2LimitDegrees=15, twistLimitDegrees=5}
 {bone1Name=dyn_r_calfLeafe_c_01, bone2Name=calf_r, swing1LimitDegrees=15, swing2LimitDegrees=15, twistLimitDegrees=5}
 {bone1Name=dyn_r_calfLeafe_c_02, bone2Name=dyn_r_calfLeafe_c_01, swing1LimitDegrees=15, swing2LimitDegrees=15, twistLimitDegrees=5}
 {bone1Name=dyn_r_legLeaf_01, bone2Name=calf_r, swing1LimitDegrees=15, swing2LimitDegrees=15, twistLimitDegrees=5}
 {
  bone1Name=clavicle_r
  bone2Name=pelvis
  swing1LimitDegrees=45
  swing1Motion=Free
  swing2LimitDegrees=45
  swing2Motion=Free
  twistLimitDegrees=45
  twistMotion=Free
 }
 {
  bone1Name=upperarm_r
  bone2Name=clavicle_r
  swing1LimitDegrees=45
  swing1Motion=Free
  swing2LimitDegrees=45
  swing2Motion=Free
  twistLimitDegrees=45
  twistMotion=Free
 }
 {
  bone1Name=calf_r
  bone2Name=thigh_r
  swing1LimitDegrees=45
  swing1Motion=Free
  swing2LimitDegrees=45
  swing2Motion=Free
  twistLimitDegrees=45
  twistMotion=Free
 }
 {
  bone1Name=thigh_r
  bone2Name=pelvis
  swing1LimitDegrees=45
  swing1Motion=Free
  swing2LimitDegrees=45
  swing2Motion=Free
  twistLimitDegrees=45
  twistMotion=Free
 }
]
