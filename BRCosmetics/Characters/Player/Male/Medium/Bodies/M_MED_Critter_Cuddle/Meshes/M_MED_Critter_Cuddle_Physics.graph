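# Physics asset for the male medium Critter Cuddle body cosmetic, covering only the pelvis, thighs and a dyn_belt bone chain rather than a full ragdoll.
# Each bone has one capsule, giving the belt chain collision shapes alongside the hips and upper legs.
# Both thighs and the belt master are constrained to the pelvis with free rotation.
# The three belt segments link in series from the master, with narrow twist limits and wider swing limits.
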PhysicsAsset /BRCosmetics/Characters/Player/Male/Medium/Bodies/M_MED_Critter_Cuddle/Meshes/M_MED_Critter_Cuddle_Physics
bodies {
 dyn_belt_01 [{shapeType=Capsule, length=2.3056, radius=2.6121, shapeName=dyn_belt_01_capsule, center=<5.4219 0 0.0001>, rotation=<rot 90 0 0>}]
 dyn_belt_02 [{shapeType=Capsule, length=1.8438, radius=2.3959, shapeName=dyn_belt_02_capsule, center=<4.6988 0 0.0001>, rotation=<rot 90 0 0>}]
 dyn_belt_03 [{shapeType=Capsule, length=0.9792, radius=1.6327, shapeName=dyn_belt_03_capsule, center=<1.8192 0 0.0001>, rotation=<rot 90 0 0>}]
 dyn_belt_MSTR [{shapeType=Capsule, length=0.505, radius=0.505, shapeName=dyn_belt_MSTR_capsule}]
 pelvis [{shapeType=Capsule, length=13.5259, radius=7.902, shapeName=pelvis_capsule}]
 thigh_l [{shapeType=Capsule, length=34.6093, radius=8.1736, shapeName=thigh_l_capsule, center=<-19.6087 -0.0001 0>, rotation=<rot 90 0 -2.0952>}]
 thigh_r [{shapeType=Capsule, length=34.2561, radius=8.5267, shapeName=thigh_l_capsule, center=<19.1399 0.0001 0>, rotation=<rot 90 0 0>}]
}
constraints [
 {
  bone1Name=thigh_l
  bone2Name=pelvis
  swing1LimitDegrees=45
  swing1Motion=Free
  swing2LimitDegrees=45
  swing2Motion=Free
  twistLimitDegrees=45
  twistMotion=Free
 }
 {
  bone1Name=thigh_r
  bone2Name=pelvis
  swing1LimitDegrees=45
  swing1Motion=Free
  swing2LimitDegrees=45
  swing2Motion=Free
  twistLimitDegrees=45
  twistMotion=Free
 }
 {
  bone1Name=dyn_belt_MSTR
  bone2Name=pelvis
  swing1LimitDegrees=45
  swing1Motion=Free
  swing2LimitDegrees=45
  swing2Motion=Free
  twistLimitDegrees=45
  twistMotion=Free
 }
 {bone1Name=dyn_belt_01, bone2Name=dyn_belt_MSTR, swing1LimitDegrees=45, swing2LimitDegrees=45, twistLimitDegrees=5}
 {bone1Name=dyn_belt_02, bone2Name=dyn_belt_01, swing1LimitDegrees=45, swing2LimitDegrees=45, twistLimitDegrees=5}
 {bone1Name=dyn_belt_03, bone2Name=dyn_belt_02, swing1LimitDegrees=45, swing2LimitDegrees=45, twistLimitDegrees=5}
]
