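# Post-process-style cosmetic AnimBP for the M_MED_Critter_Cuddle player body, adding secondary motion to leaves, jacket and shoulder pads.
# It copies the parent mesh pose, blending in jacket and RightArmLift-driven corrections while crouching, then applies bone constraints and AnimDynamics.
# The cached PreFrontEnd pose feeds two RigidBody branches chosen by skydiving AND front-end state, followed by front-end leaf tweaks.
# Several calf, ankle, sleeve and collar dynamics chains are unconnected.
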AnimBlueprint /BRCosmetics/Characters/Player/Male/Medium/Bodies/M_MED_Critter_Cuddle/Meshes/M_MED_Critter_Cuddle_AnimBP
target_skeleton=&/M_MED_Critter_Cuddle_Skeleton
graphs [
 AnimGraph {nodes [
   n1 = Anim Dynamics - Bone: dyn_l_calfString_01!(ComponentPose, Alpha=1, GravityOverride:Vector="0,0,0", ExternalForce="0,0,0") -> (Pose) {
    angularDampingOverride=1
    angularSpringConstant=50
    bAngularSpring=true
    bGravityOverrideInSimSpace=true
    bOverrideAngularDamping=true
    gravityScale=0.35
    simulationSpace=RootRelative
    alphaBoolBlend {blendOption=Linear}
    alphaScaleBias {scale=2}
    boundBone {boneName=dyn_l_calfString_01}
    physicsBodyDefinitions [{
      sphereCollisionRadius=0
      boundBone {boneName=dyn_l_calfString_01}
      boxExtents=<12 22 22>
      constraintSetup {angularLimitsMax=<20 0 50>, angularLimitsMin=<-20 -20 -50>, angularTarget=<0 0 0>}
      localJointOffset=<10 0 0>
     }]
   }
   n2 = Anim Dynamics - Bone: dyn_l_calfString_01(ComponentPose=@n1, Alpha=1, GravityOverride="0,0,0", ExternalForce="0,0,0") {
    angularDampingOverride=1
    angularSpringConstant=50
    bAngularSpring=true
    bGravityOverrideInSimSpace=true
    bOverrideAngularDamping=true
    gravityScale=0.35
    simulationSpace=RootRelative
    alphaBoolBlend {blendOption=Linear}
    alphaScaleBias {scale=2}
    boundBone {boneName=dyn_l_calfString_01}
    physicsBodyDefinitions [{
      sphereCollisionRadius=0
      boundBone {boneName=dyn_l_calfString_01}
      boxExtents=<12 22 22>
      constraintSetup {angularLimitsMax=<20 0 50>, angularLimitsMin=<-20 -20 -50>, angularTarget=<0 0 0>}
      localJointOffset=<10 0 0>
     }]
   }
   n3 = Anim Dynamics - Bone: dyn_r_calfLeafe_a_01!(ComponentPose=@n2, Alpha=0.25, GravityOverride:Vector="0,0,0", ExternalForce="0,0,0") -> (Pose) {
    angularDampingOverride=1
    angularSpringConstant=1500
    bAngularSpring=true
    bGravityOverrideInSimSpace=true
    bOverrideAngularDamping=true
    gravityScale=0.35
    simulationSpace=RootRelative
    alphaBoolBlend {blendOption=Linear}
    alphaScaleBias {scale=2}
    boundBone {boneName=dyn_r_calfLeafe_a_01}
    physicsBodyDefinitions [{
      sphereCollisionRadius=0
      boundBone {boneName=dyn_r_calfLeafe_a_01}
      boxExtents=<10 20 20>
      constraintSetup {angularLimitsMax=<0 0 5>, angularLimitsMin=<0 -20 -5>, angularTarget=<0 0 0>}
      localJointOffset=<10 0 0>
     }]
   }
   n4 = Anim Dynamics - Bone: dyn_r_calfLeafe_a_02!(ComponentPose=@n3, Alpha=0.25, GravityOverride:Vector="0,0,0", ExternalForce="0,0,0") -> (Pose) {
    angularDampingOverride=1
    angularSpringConstant=1500
    bAngularSpring=true
    bGravityOverrideInSimSpace=true
    bOverrideAngularDamping=true
    gravityScale=0.35
    simulationSpace=RootRelative
    alphaBoolBlend {blendOption=Linear}
    alphaScaleBias {scale=2}
    boundBone {boneName=dyn_r_calfLeafe_a_02}
    physicsBodyDefinitions [{
      sphereCollisionRadius=0
      boundBone {boneName=dyn_r_calfLeafe_a_02}
      boxExtents=<10 20 20>
      constraintSetup {angularLimitsMax=<0 0 5>, angularLimitsMin=<0 -20 -5>, angularTarget=<0 0 0>}
      localJointOffset=<10 0 0>
     }]
   }
   n5 = Anim Dynamics - Bone: dyn_r_calfLeafe_b_01!(ComponentPose=@n4, Alpha=0.25, GravityOverride:Vector="0,0,0", ExternalForce="0,0,0") -> (Pose) {
    angularDampingOverride=1
    angularSpringConstant=1500
    bAngularSpring=true
    bGravityOverrideInSimSpace=true
    bOverrideAngularDamping=true
    gravityScale=0.35
    simulationSpace=RootRelative
    alphaBoolBlend {blendOption=Linear}
    alphaScaleBias {scale=2}
    boundBone {boneName=dyn_r_calfLeafe_b_01}
    physicsBodyDefinitions [{
      sphereCollisionRadius=0
      boundBone {boneName=dyn_r_calfLeafe_b_01}
      boxExtents=<10 20 20>
      constraintSetup {angularLimitsMax=<0 0 5>, angularLimitsMin=<0 -20 -5>, angularTarget=<0 0 0>}
      localJointOffset=<10 0 0>
     }]
   }
   n6 = Anim Dynamics - Bone: dyn_r_calfLeafe_b_02!(ComponentPose=@n5, Alpha=0.25, GravityOverride:Vector="0,0,0", ExternalForce="0,0,0") -> (Pose) {
    angularDampingOverride=1
    angularSpringConstant=1500
    bAngularSpring=true
    bGravityOverrideInSimSpace=true
    bOverrideAngularDamping=true
    gravityScale=0.35
    simulationSpace=RootRelative
    alphaBoolBlend {blendOption=Linear}
    alphaScaleBias {scale=2}
    boundBone {boneName=dyn_r_calfLeafe_b_02}
    physicsBodyDefinitions [{
      sphereCollisionRadius=0
      boundBone {boneName=dyn_r_calfLeafe_b_02}
      boxExtents=<10 20 20>
      constraintSetup {angularLimitsMax=<0 0 5>, angularLimitsMin=<0 -20 -5>, angularTarget=<0 0 0>}
      localJointOffset=<10 0 0>
     }]
   }
   n7 = Anim Dynamics - Bone: dyn_r_calfLeafe_c_01!(ComponentPose=@n6, Alpha=0.25, GravityOverride:Vector="0,0,0", ExternalForce="0,0,0") -> (Pose) {
    angularDampingOverride=1
    angularSpringConstant=1500
    bAngularSpring=true
    bGravityOverrideInSimSpace=true
    bOverrideAngularDamping=true
    gravityScale=0.35
    simulationSpace=RootRelative
    alphaBoolBlend {blendOption=Linear}
    alphaScaleBias {scale=2}
    boundBone {boneName=dyn_r_calfLeafe_c_01}
    physicsBodyDefinitions [{
      sphereCollisionRadius=0
      boundBone {boneName=dyn_r_calfLeafe_c_01}
      boxExtents=<10 20 20>
      constraintSetup {angularLimitsMax=<0 0 5>, angularLimitsMin=<0 -20 -5>, angularTarget=<0 0 0>}
      localJointOffset=<10 0 0>
     }]
   }
   n8 = Anim Dynamics - Bone: dyn_r_calfLeafe_c_02!(ComponentPose=@n7, Alpha=0.25, GravityOverride:Vector="0,0,0", ExternalForce="0,0,0") -> (Pose) {
    angularDampingOverride=1
    angularSpringConstant=1500
    bAngularSpring=true
    bGravityOverrideInSimSpace=true
    bOverrideAngularDamping=true
    gravityScale=0.35
    simulationSpace=RootRelative
    alphaBoolBlend {blendOption=Linear}
    alphaScaleBias {scale=2}
    boundBone {boneName=dyn_r_calfLeafe_c_02}
    physicsBodyDefinitions [{
      sphereCollisionRadius=0
      boundBone {boneName=dyn_r_calfLeafe_c_02}
      boxExtents=<10 20 20>
      constraintSetup {angularLimitsMax=<0 0 5>, angularLimitsMin=<0 -20 -5>, angularTarget=<0 0 0>}
      localJointOffset=<10 0 0>
     }]
   }
   n9 = Anim Dynamics - Bone: dyn_r_legLeaf_01!(ComponentPose=@n8, Alpha=0.75, GravityOverride:Vector="0,0,0", ExternalForce="0,0,0") -> (Pose) {
    angularDampingOverride=0.999
    angularSpringConstant=1500
    bAngularSpring=true
    bGravityOverrideInSimSpace=true
    bOverrideAngularDamping=true
    gravityScale=0.35
    simulationSpace=RootRelative
    alphaBoolBlend {blendOption=Linear}
    alphaScaleBias {scale=2}
    boundBone {boneName=dyn_r_legLeaf_01}
    physicsBodyDefinitions [{
      sphereCollisionRadius=0
      boundBone {boneName=dyn_r_legLeaf_01}
      boxExtents=<10 20 20>
      constraintSetup {angularLimitsMax=<20 5 5>, angularLimitsMin=<-10 -10 -10>, angularTarget=<0 0 0>}
      localJointOffset=<-2 0 0>
     }]
   }
   n10 = Anim Dynamics - Bone: dyn_r_legLeaf_02!(ComponentPose=@n9, Alpha=0.75, GravityOverride:Vector="0,0,0", ExternalForce="0,0,0") -> (Pose) {
    angularDampingOverride=0.999
    angularSpringConstant=1500
    bAngularSpring=true
    bGravityOverrideInSimSpace=true
    bOverrideAngularDamping=true
    gravityScale=0.35
    simulationSpace=RootRelative
    alphaBoolBlend {blendOption=Linear}
    alphaScaleBias {scale=2}
    boundBone {boneName=dyn_r_legLeaf_02}
    physicsBodyDefinitions [{
      sphereCollisionRadius=0
      boundBone {boneName=dyn_r_legLeaf_02}
      boxExtents=<10 20 20>
      constraintSetup {angularLimitsMax=<20 5 5>, angularLimitsMin=<-10 -10 -10>, angularTarget=<0 0 0>}
      localJointOffset=<-2 0 0>
     }]
   }
   n11 = Anim Dynamics - Bone: dyn_r_legLeaf_03!(ComponentPose=@n10, Alpha=0.75, GravityOverride:Vector="0,0,0", ExternalForce="0,0,0") -> (Pose) {
    angularDampingOverride=0.999
    angularSpringConstant=1500
    bAngularSpring=true
    bGravityOverrideInSimSpace=true
    bOverrideAngularDamping=true
    gravityScale=0.35
    simulationSpace=RootRelative
    alphaBoolBlend {blendOption=Linear}
    alphaScaleBias {scale=2}
    boundBone {boneName=dyn_r_legLeaf_03}
    physicsBodyDefinitions [{
      sphereCollisionRadius=0
      boundBone {boneName=dyn_r_legLeaf_03}
      boxExtents=<10 20 20>
      constraintSetup {angularLimitsMax=<20 5 5>, angularLimitsMin=<-10 -10 -10>, angularTarget=<0 0 0>}
      localJointOffset=<-2 0 0>
     }]
   }
   n12 = Anim Dynamics - Bone: dyn_l_ankle_01!(ComponentPose=@n11, Alpha=0.75, GravityOverride:Vector="0,0,0", ExternalForce="0,0,0") -> (Pose) {
    angularDampingOverride=1
    angularSpringConstant=1500
    bAngularSpring=true
    bGravityOverrideInSimSpace=true
    bOverrideAngularDamping=true
    gravityScale=0.35
    simulationSpace=RootRelative
    alphaBoolBlend {blendOption=Linear}
    alphaScaleBias {scale=2}
    boundBone {boneName=dyn_l_ankle_01}
    physicsBodyDefinitions [{
      sphereCollisionRadius=0
      boundBone {boneName=dyn_l_ankle_01}
      boxExtents=<10 20 20>
      constraintSetup {angularLimitsMax=<0 0 5>, angularLimitsMin=<0 -20 -5>, angularTarget=<0 0 0>}
      localJointOffset=<10 0 0>
     }]
   }
   n13 = Anim Dynamics - Bone: dyn_l_ankle_02!(ComponentPose=@n12, Alpha=0.75, GravityOverride:Vector="0,0,0", ExternalForce="0,0,0") -> (Pose) {
    angularDampingOverride=1
    angularSpringConstant=1500
    bAngularSpring=true
    bGravityOverrideInSimSpace=true
    bOverrideAngularDamping=true
    gravityScale=0.35
    simulationSpace=RootRelative
    alphaBoolBlend {blendOption=Linear}
    alphaScaleBias {scale=2}
    boundBone {boneName=dyn_l_ankle_02}
    physicsBodyDefinitions [{
      sphereCollisionRadius=0
      boundBone {boneName=dyn_l_ankle_02}
      boxExtents=<10 20 20>
      constraintSetup {angularLimitsMax=<0 0 5>, angularLimitsMin=<0 -20 -5>, angularTarget=<0 0 0>}
      localJointOffset=<10 0 0>
     }]
   }
   n14 = Anim Dynamics - Bone: dyn_l_ankle_03!(ComponentPose=@n13, Alpha=0.75, GravityOverride:Vector="0,0,0", ExternalForce="0,0,0") -> (Pose) {
    angularDampingOverride=1
    angularSpringConstant=1500
    bAngularSpring=true
    bGravityOverrideInSimSpace=true
    bOverrideAngularDamping=true
    gravityScale=0.35
    simulationSpace=RootRelative
    alphaBoolBlend {blendOption=Linear}
    alphaScaleBias {scale=2}
    boundBone {boneName=dyn_l_ankle_03}
    physicsBodyDefinitions [{
      sphereCollisionRadius=0
      boundBone {boneName=dyn_l_ankle_03}
      boxExtents=<10 20 20>
      constraintSetup {angularLimitsMax=<0 0 5>, angularLimitsMin=<0 -20 -5>, angularTarget=<0 0 0>}
      localJointOffset=<10 0 0>
     }]
   }
   Anim Dynamics - Bone: dyn_l_ankle_04!(ComponentPose=@n14, Alpha=0.75, GravityOverride:Vector="0,0,0", ExternalForce="0,0,0") -> (Pose) {
    angularDampingOverride=1
    angularSpringConstant=1500
    bAngularSpring=true
    bGravityOverrideInSimSpace=true
    bOverrideAngularDamping=true
    gravityScale=0.35
    simulationSpace=RootRelative
    alphaBoolBlend {blendOption=Linear}
    alphaScaleBias {scale=2}
    boundBone {boneName=dyn_l_ankle_04}
    physicsBodyDefinitions [{
      sphereCollisionRadius=0
      boundBone {boneName=dyn_l_ankle_04}
      boxExtents=<10 20 20>
      constraintSetup {angularLimitsMax=<0 0 5>, angularLimitsMin=<0 -20 -5>, angularTarget=<0 0 0>}
      localJointOffset=<10 0 0>
     }]
   }
   n15 = Anim Dynamics - Bone: dyn_leaf_c_039!(ComponentPose, Alpha=1, GravityOverride:Vector="0,0,0", ExternalForce="0,0,0") -> (Pose) {
    angularDampingOverride=1
    angularSpringConstant=50
    bAngularSpring=true
    bGravityOverrideInSimSpace=true
    bOverrideAngularDamping=true
    gravityScale=0.35
    simulationSpace=RootRelative
    alphaBoolBlend {blendOption=Linear}
    alphaScaleBias {scale=2}
    boundBone {boneName=dyn_leaf_c_039}
    physicsBodyDefinitions [{
      sphereCollisionRadius=0
      boundBone {boneName=dyn_leaf_c_039}
      boxExtents=<12 22 22>
      constraintSetup {angularLimitsMax=<20 10 50>, angularLimitsMin=<-20 -20 -50>, angularTarget=<0 0 0>}
      localJointOffset=<10 0 0>
     }]
   }
   Anim Dynamics - Bone: dyn_leaf_c_020!(ComponentPose=@n15, Alpha=1, GravityOverride:Vector="0,0,0", ExternalForce="0,0,0") -> (Pose) {
    angularDampingOverride=1
    angularSpringConstant=50
    bAngularSpring=true
    bGravityOverrideInSimSpace=true
    bOverrideAngularDamping=true
    gravityScale=0.35
    simulationSpace=RootRelative
    alphaBoolBlend {blendOption=Linear}
    alphaScaleBias {scale=2}
    boundBone {boneName=dyn_leaf_c_020}
    physicsBodyDefinitions [{
      sphereCollisionRadius=0
      boundBone {boneName=dyn_leaf_c_020}
      boxExtents=<12 22 22>
      constraintSetup {angularLimitsMax=<20 10 50>, angularLimitsMin=<-20 -20 -50>, angularTarget=<0 0 0>}
      localJointOffset=<10 0 0>
     }]
   }
   n16 = Anim Dynamics - Bone: dyn_l_sleeve_01(Alpha=0.4, ExternalForce="0,0,0") {
    angularDampingOverride=0.9999
    angularSpringConstant=1500
    bAngularSpring=true
    bGravityOverrideInSimSpace=true
    bOverrideAngularDamping=true
    gravityScale=0.8
    simulationSpace=BoneRelative
    alphaBoolBlend {blendOption=Linear}
    boundBone {boneName=dyn_l_sleeve_01}
    componentAppliedLinearAccClamp=<5000 5000 3500>
    componentLinearAccScale=<1.5 1.5 1.25>
    componentLinearVelScale=<1.5 1.5 1.25>
    physicsBodyDefinitions [{
      sphereCollisionRadius=0
      boundBone {boneName=dyn_l_sleeve_01}
      boxExtents=<10 20 20>
      constraintSetup {angularLimitsMax=<45 45 45>, angularLimitsMin=<-45 -45 -45>, angularTarget=<0 0 0>}
      localJointOffset=<2 0 0>
     }]
    relativeSpaceBone {boneName=dyn_root_simspace}
   }
   n17 = Anim Dynamics - Bone: dyn_l_sleeve_02(ComponentPose=@n16, Alpha=0.4, ExternalForce="0,0,0") {
    angularDampingOverride=0.9999
    angularSpringConstant=1500
    bAngularSpring=true
    bGravityOverrideInSimSpace=true
    bOverrideAngularDamping=true
    gravityScale=0.8
    simulationSpace=BoneRelative
    alphaBoolBlend {blendOption=Linear}
    boundBone {boneName=dyn_l_sleeve_02}
    componentAppliedLinearAccClamp=<5000 5000 3500>
    componentLinearAccScale=<1.5 1.5 1.25>
    componentLinearVelScale=<1.5 1.5 1.25>
    physicsBodyDefinitions [{
      sphereCollisionRadius=0
      boundBone {boneName=dyn_l_sleeve_02}
      boxExtents=<10 20 20>
      constraintSetup {angularLimitsMax=<45 45 45>, angularLimitsMin=<-45 -45 -45>, angularTarget=<0 0 0>}
      localJointOffset=<2 0 0>
     }]
    relativeSpaceBone {boneName=dyn_root_simspace}
   }
   n18 = Anim Dynamics - Bone: dyn_l_sleeve_03(ComponentPose=@n17, Alpha=0.4, ExternalForce="0,0,0") {
    angularDampingOverride=0.9999
    angularSpringConstant=1500
    bAngularSpring=true
    bGravityOverrideInSimSpace=true
    bOverrideAngularDamping=true
    gravityScale=0.8
    simulationSpace=BoneRelative
    alphaBoolBlend {blendOption=Linear}
    boundBone {boneName=dyn_l_sleeve_03}
    componentAppliedLinearAccClamp=<5000 5000 3500>
    componentLinearAccScale=<1.5 1.5 1.25>
    componentLinearVelScale=<1.5 1.5 1.25>
    physicsBodyDefinitions [{
      sphereCollisionRadius=0
      boundBone {boneName=dyn_l_sleeve_03}
      boxExtents=<10 20 20>
      constraintSetup {angularLimitsMax=<45 45 45>, angularLimitsMin=<-45 -45 -45>, angularTarget=<0 0 0>}
      localJointOffset=<2 0 0>
     }]
    relativeSpaceBone {boneName=dyn_root_simspace}
   }
   Anim Dynamics - Bone: dyn_l_sleeve_04(ComponentPose=@n18, Alpha=0.4, ExternalForce="0,0,0") {
    angularDampingOverride=0.9999
    angularSpringConstant=1500
    bAngularSpring=true
    bGravityOverrideInSimSpace=true
    bOverrideAngularDamping=true
    gravityScale=0.8
    simulationSpace=BoneRelative
    alphaBoolBlend {blendOption=Linear}
    boundBone {boneName=dyn_l_sleeve_04}
    componentAppliedLinearAccClamp=<5000 5000 3500>
    componentLinearAccScale=<1.5 1.5 1.25>
    componentLinearVelScale=<1.5 1.5 1.25>
    physicsBodyDefinitions [{
      sphereCollisionRadius=0
      boundBone {boneName=dyn_l_sleeve_04}
      boxExtents=<10 20 20>
      constraintSetup {angularLimitsMax=<45 45 45>, angularLimitsMin=<-45 -45 -45>, angularTarget=<0 0 0>}
      localJointOffset=<2 0 0>
     }]
    relativeSpaceBone {boneName=dyn_root_simspace}
   }
   n19 = Anim Dynamics - Bone: dyn_r_collarSide(Alpha=1, ExternalForce="0,0,0") {
    angularDampingOverride=1
    angularSpringConstant=2500
    bAngularSpring=true
    bGravityOverrideInSimSpace=true
    bOverrideAngularDamping=true
    gravityScale=0.35
    simulationSpace=RootRelative
    alphaBoolBlend {blendOption=Linear}
    boundBone {boneName=dyn_r_collarSide}
    componentAppliedLinearAccClamp=<2500 2500 2500>
    componentLinearAccScale=<0.6 0.6 0.6>
    componentLinearVelScale=<0.6 0.6 0.6>
    physicsBodyDefinitions [{
      sphereCollisionRadius=0
      boundBone {boneName=dyn_r_collarSide}
      boxExtents=<10 20 20>
      constraintSetup {angularLimitsMax=<0 2 15>, angularLimitsMin=<0 -2 -15>, angularTarget=<1 0 0>}
      localJointOffset=<1 0 0>
     }]
   }
   n20 = Anim Dynamics - Bone: dyn_l_collarSide(ComponentPose=@n19, Alpha=1, ExternalForce="0,0,0") {
    angularDampingOverride=1
    angularSpringConstant=2500
    bAngularSpring=true
    bGravityOverrideInSimSpace=true
    bOverrideAngularDamping=true
    gravityScale=0.35
    simulationSpace=RootRelative
    alphaBoolBlend {blendOption=Linear}
    boundBone {boneName=dyn_l_collarSide}
    componentAppliedLinearAccClamp=<2500 2500 2500>
    componentLinearAccScale=<0.6 0.6 0.6>
    componentLinearVelScale=<0.6 0.6 0.6>
    physicsBodyDefinitions [{
      sphereCollisionRadius=0
      boundBone {boneName=dyn_l_collarSide}
      boxExtents=<10 20 20>
      constraintSetup {angularLimitsMax=<0 2 15>, angularLimitsMin=<0 -2 -15>, angularTarget=<1 0 0>}
      localJointOffset=<1 0 0>
     }]
   }
   n21 = Anim Dynamics - Bone: dyn_l_collarFront(ComponentPose=@n20, Alpha=0.5, ExternalForce="0,0,0") {
    angularDampingOverride=1
    angularSpringConstant=2500
    bAngularSpring=true
    bGravityOverrideInSimSpace=true
    bOverrideAngularDamping=true
    gravityScale=0.6
    simulationSpace=RootRelative
    alphaBoolBlend {blendOption=Linear}
    boundBone {boneName=dyn_l_collarFront}
    componentAppliedLinearAccClamp=<2500 2500 2500>
    componentLinearAccScale=<0.6 0.6 0.6>
    componentLinearVelScale=<0.6 0.6 0.6>
    physicsBodyDefinitions [{
      sphereCollisionRadius=0
      boundBone {boneName=dyn_l_collarFront}
      boxExtents=<10 20 20>
      constraintSetup {angularLimitsMax=<15 20 15>, angularLimitsMin=<-15 -20 -15>, angularTarget=<1 0 0>}
      localJointOffset=<5 0 0>
     }]
   }
   Anim Dynamics - Bone: dyn_collarBack(ComponentPose=@n21, Alpha=1, ExternalForce="0,0,0") {
    angularDampingOverride=1
    angularSpringConstant=2500
    bAngularSpring=true
    bGravityOverrideInSimSpace=true
    bOverrideAngularDamping=true
    gravityScale=0.35
    simulationSpace=RootRelative
    alphaBoolBlend {blendOption=Linear}
    boundBone {boneName=dyn_collarBack}
    componentAppliedLinearAccClamp=<2500 2500 2500>
    componentLinearAccScale=<0.6 0.6 0.6>
    componentLinearVelScale=<0.6 0.6 0.6>
    physicsBodyDefinitions [{
      sphereCollisionRadius=0
      boundBone {boneName=dyn_collarBack}
      boxExtents=<10 20 20>
      constraintSetup {angularLimitsMax=<0 2 15>, angularLimitsMin=<0 -2 -15>, angularTarget=<1 0 0>}
      localJointOffset=<1 0 0>
     }]
   }
   n22 = "Constraint - dyn_jacket_a_MSTR (RR)"!(ComponentPose, Alpha=1, ConstraintWeights_0=0.5, ConstraintWeights_1:float=0.5) -> (Pose) {
    alphaBoolBlend {blendOption=Linear}
    boneToModify {boneName=dyn_jacket_a_MSTR}
    constraintSetup [{transformType=Rotation, targetBone {boneName=thigh_l}}, {transformType=Rotation, targetBone {boneName=pelvis}}]
    constraintWeights [1,1]
   }
   "Constraint - dyn_jacket_a_MSTR (RR)"(ComponentPose=@n22, Alpha=1, ConstraintWeights_0=0.5, ConstraintWeights_1=0.5) {
    alphaBoolBlend {blendOption=Linear}
    boneToModify {boneName=dyn_jacket_a_MSTR}
    constraintSetup [{transformType=Rotation, targetBone {boneName=thigh_l}}, {transformType=Rotation, targetBone {boneName=pelvis}}]
    constraintWeights [1,1]
   }
   n23 = RigidBody!(
    ComponentPose=Local To Component(LocalPose=Use cached pose 'PreFrontEnd'())
    bAlphaBoolEnabled:bool
    OverrideWorldGravity:Vector="0,0,350"
    ExternalForce
    bTransferBoneVelocities:bool=False
   ) -> (Pose) {
    alphaInputType=Bool
    bClampLinearTranslationLimitToRefPose=true
    bOverrideWorldGravity=true
    overridePhysicsAsset=&/M_MED_Critter_Cuddle_FE_Physics
    simulationSpace=BaseBoneSpace
    alphaBoolBlend {blendOption=Linear}
    baseBoneRef {boneName=root}
    componentAppliedLinearAccClamp=<5000 5000 5000>
    componentLinearAccScale=<0.4 0.4 0.4>
    componentLinearVelScale=<0.4 0.4 0.4>
    overrideWorldGravity {z=-980}
   }
   n24 = "Transform (Modify) Bone - Bone: dyn_leaf_hi_01"!(ComponentPose=@n23, bAlphaBoolEnabled:bool=True, Translation="0,0,0", Rotation="0,0,0", Scale="0.1,0.1,0.1") -> (Pose) {
    alphaInputType=Bool
    rotationMode=BMM_Additive
    rotationSpace=BCS_BoneSpace
    scaleMode=BMM_Additive
    scaleSpace=BCS_BoneSpace
    translationMode=BMM_Additive
    translationSpace=BCS_BoneSpace
    alphaBoolBlend {blendOption=Linear}
    alphaScaleBiasClamp {bInterpResult=true, interpSpeedDecreasing=20, interpSpeedIncreasing=20}
    boneToModify {boneName=dyn_leaf_hi_01}
    scale=<0.1 0.1 0.1>
   }
   n25 = Component To Local(ComponentPose=@n24)
   n26 = RigidBody!(ComponentPose=Local To Component(LocalPose=Use cached pose 'PreFrontEnd'()), bAlphaBoolEnabled, OverrideWorldGravity="0,0,0", ExternalForce) -> (Pose) {
    alphaInputType=Bool
    bClampLinearTranslationLimitToRefPose=true
    bOverrideWorldGravity=true
    overridePhysicsAsset=&M_MED_Critter_Cuddle_FE_Physics
    simulationSpace=BaseBoneSpace
    alphaBoolBlend {blendOption=Linear}
    baseBoneRef {boneName=root}
    componentAppliedLinearAccClamp=<5000 5000 5000>
    componentLinearAccScale=<0.4 0.4 0.4>
    componentLinearVelScale=<0.4 0.4 0.4>
   }
   n27 = "Transform (Modify) Bone - Bone: dyn_leaf_i_01"!(ComponentPose=@n26, bAlphaBoolEnabled:bool=True, Translation="0,0,0", Rotation="10.557,2.014,3.213", Scale="0.8,0.8,0.8") -> (Pose) {
    alphaInputType=Bool
    rotationMode=BMM_Replace
    rotationSpace=BCS_BoneSpace
    scaleMode=BMM_Replace
    scaleSpace=BCS_BoneSpace
    translationMode=BMM_Additive
    translationSpace=BCS_BoneSpace
    alphaBoolBlend {blendOption=Linear}
    alphaScaleBiasClamp {bInterpResult=true, interpSpeedDecreasing=20, interpSpeedIncreasing=20}
    boneToModify {boneName=dyn_leaf_i_01}
    rotation=<rot 10.557 2.014 3.213>
    scale=<0.8 0.8 0.8>
   }
   n28 = "Transform (Modify) Bone - Bone: dyn_leaf_h_01"!(ComponentPose=@n27, bAlphaBoolEnabled:bool=True, Translation="0.314,-2.520,-0.092", Rotation="16.748,11.311,4.658", Scale="0.8,0.8,0.8") -> (Pose) {
    alphaInputType=Bool
    rotationMode=BMM_Replace
    rotationSpace=BCS_BoneSpace
    scaleMode=BMM_Replace
    scaleSpace=BCS_BoneSpace
    translationMode=BMM_Additive
    translationSpace=BCS_BoneSpace
    alphaBoolBlend {blendOption=Linear}
    alphaScaleBiasClamp {bInterpResult=true, interpSpeedDecreasing=20, interpSpeedIncreasing=20}
    boneToModify {boneName=dyn_leaf_h_01}
    rotation=<rot 16.748 11.311 4.658>
    scale=<0.8 0.8 0.8>
    translation=<0.314 -2.52 -0.092>
   }
   n29 = "Transform (Modify) Bone - Bone: dyn_leaf_hi_01"(ComponentPose=@n28, bAlphaBoolEnabled=True, Translation="0,0,0", Rotation="13.295,-19.836,-10.089", Scale="0.8,0.8,0.8") {
    alphaInputType=Bool
    rotationMode=BMM_Replace
    rotationSpace=BCS_BoneSpace
    scaleMode=BMM_Replace
    scaleSpace=BCS_BoneSpace
    translationMode=BMM_Additive
    translationSpace=BCS_BoneSpace
    alphaBoolBlend {blendOption=Linear}
    alphaScaleBiasClamp {bInterpResult=true, interpSpeedDecreasing=20, interpSpeedIncreasing=20}
    boneToModify {boneName=dyn_leaf_hi_01}
    rotation=<rot 13.2949 -19.836 -10.089>
    scale=<0.8 0.8 0.8>
   }
   n30 = Component To Local(ComponentPose=@n29)
   n31 = AND Boolean(A=Get bIsSkyDiving(), B=Get bIsInFrontEnd())
   n32 = Get bIsInFrontEnd()
   n33 = "Transform (Modify) Bone - Bone: dyn_leaf_bc_01"!(
    ComponentPose=Local To Component(LocalPose=Blend Poses by bool(BlendPose_0=@n25, BlendPose_1=@n30, bActiveValue=@n31) {blendType=Linear})
    bAlphaBoolEnabled:bool=@n32
    Translation="0,0,-0.5"
    Rotation="0,0,-5"
    Scale="0.8,0.8,0.8"
   ) -> (Pose) {
    alphaInputType=Bool
    rotationMode=BMM_Additive
    rotationSpace=BCS_BoneSpace
    scaleSpace=BCS_BoneSpace
    translationMode=BMM_Additive
    translationSpace=BCS_BoneSpace
    alphaBoolBlend {blendOption=Linear}
    alphaScaleBiasClamp {bInterpResult=true, interpSpeedDecreasing=20, interpSpeedIncreasing=20}
    boneToModify {boneName=dyn_leaf_bc_01}
    rotation {roll=-5}
    scale=<0.8 0.8 0.8>
    translation {z=-0.5}
   }
   n34 = "Transform (Modify) Bone - Bone: dyn_leaf_hi_01"(ComponentPose=@n33, bAlphaBoolEnabled=@n32, Translation="0,0,0.5", Rotation="3,0,-2.5", Scale="0.8,0.8,0.8") {
    alphaInputType=Bool
    rotationMode=BMM_Additive
    rotationSpace=BCS_BoneSpace
    scaleSpace=BCS_BoneSpace
    translationMode=BMM_Additive
    translationSpace=BCS_BoneSpace
    alphaBoolBlend {blendOption=Linear}
    alphaScaleBiasClamp {bInterpResult=true, interpSpeedDecreasing=20, interpSpeedIncreasing=20}
    boneToModify {boneName=dyn_leaf_hi_01}
    rotation {pitch=3, roll=-2.5}
    scale=<0.8 0.8 0.8>
    translation {z=0.5}
   }
   Output Pose(Result=Component To Local(ComponentPose=@n34)) {layerGroup=None}
   n35 = "Constraint - dyn_jacket_a_MSTR (RR)"(
    ComponentPose=Local To Component(LocalPose=Copy Pose From Mesh() {bUseAttachedParent=true})
    Alpha=1
    ConstraintWeights_0=0.5
    ConstraintWeights_1=0.5
   ) {
    alphaBoolBlend {blendOption=Linear}
    boneToModify {boneName=dyn_jacket_a_MSTR}
    constraintSetup [{transformType=Rotation, targetBone {boneName=thigh_l}}, {transformType=Rotation, targetBone {boneName=pelvis}}]
    constraintWeights [1,1]
   }
   n36 = "Constraint - dyn_jacket_a_MSTR (RR)"(ComponentPose=@n35, Alpha=1, ConstraintWeights_0=0.5, ConstraintWeights_1=0.5) {
    alphaBoolBlend {blendOption=Linear}
    boneToModify {boneName=dyn_jacket_a_MSTR}
    constraintSetup [{transformType=Rotation, targetBone {boneName=thigh_l}}, {transformType=Rotation, targetBone {boneName=pelvis}}]
    constraintWeights [1,1]
   }
   n37 = "Transform (Modify) Bone - Bone: dyn_jacket_a_MSTR"(ComponentPose=@n36, Alpha=1, Translation="0,0,0", Rotation="8,50,15", Scale="1,1,1") {
    rotationMode=BMM_Additive
    rotationSpace=BCS_BoneSpace
    scaleMode=BMM_Additive
    scaleSpace=BCS_BoneSpace
    translationMode=BMM_Additive
    translationSpace=BCS_BoneSpace
    alphaBoolBlend {blendOption=Linear}
    boneToModify {boneName=dyn_jacket_a_MSTR}
    rotation=<rot 8 50 15>
   }
   n38 = "Transform (Modify) Bone - Bone: dyn_jacket_a_01"(ComponentPose=@n37, Alpha=1, Translation="0,0,1", Rotation="0,0,0", Scale="1,1,1") {
    rotationMode=BMM_Additive
    rotationSpace=BCS_BoneSpace
    scaleMode=BMM_Additive
    scaleSpace=BCS_BoneSpace
    translationMode=BMM_Additive
    translationSpace=BCS_BoneSpace
    alphaBoolBlend {blendOption=Linear}
    boneToModify {boneName=dyn_jacket_a_01}
    translation {z=1}
   }
   n39 = "Transform (Modify) Bone - Bone: dyn_jacket_a_02"(ComponentPose=@n38, Alpha=1, Translation="0,0,0", Rotation="20,0,0", Scale="1.8,1,1") {
    rotationMode=BMM_Additive
    rotationSpace=BCS_BoneSpace
    scaleMode=BMM_Additive
    scaleSpace=BCS_BoneSpace
    translationMode=BMM_Additive
    translationSpace=BCS_BoneSpace
    alphaBoolBlend {blendOption=Linear}
    boneToModify {boneName=dyn_jacket_a_02}
    rotation {pitch=20}
    scale {x=1.8}
   }
   n40 = "Transform (Modify) Bone - Bone: dyn_jacket_b_MSTR"(ComponentPose=@n39, Alpha=1, Translation="0,0,-0.5", Rotation="10,30,0", Scale="1,1,1") {
    rotationMode=BMM_Additive
    rotationSpace=BCS_BoneSpace
    scaleMode=BMM_Additive
    scaleSpace=BCS_BoneSpace
    translationMode=BMM_Additive
    translationSpace=BCS_BoneSpace
    alphaBoolBlend {blendOption=Linear}
    boneToModify {boneName=dyn_jacket_b_MSTR}
    rotation {pitch=10, yaw=30}
    translation {z=-0.5}
   }
   n41 = Anim Dynamics - Bone: dyn_jacket_a_01!(ComponentPose=@n40, Alpha=0.6, GravityOverride:Vector="0,0,0", ExternalForce="0,0,0") -> (Pose) {
    angularDampingOverride=0.9999
    angularSpringConstant=1500
    bAngularSpring=true
    bGravityOverrideInSimSpace=true
    bOverrideAngularDamping=true
    gravityScale=0.8
    simulationSpace=BoneRelative
    alphaBoolBlend {blendOption=Linear}
    boundBone {boneName=dyn_jacket_a_01}
    componentAppliedLinearAccClamp=<5000 5000 3500>
    componentLinearAccScale=<1.5 1.5 1.25>
    componentLinearVelScale=<1.5 1.5 1.25>
    physicsBodyDefinitions [{
      sphereCollisionRadius=0
      boundBone {boneName=dyn_jacket_a_01}
      boxExtents=<10 20 20>
      constraintSetup {angularLimitsMax=<15 25 15>, angularLimitsMin=<-15 -50 -15>, angularTarget=<0 0 0>}
      localJointOffset=<-2 0 0>
     }]
    relativeSpaceBone {boneName=dyn_root_simspace}
   }
   n42 = Anim Dynamics - Bone: dyn_jacket_a_02!(ComponentPose=@n41, Alpha=0.5, GravityOverride:Vector="0,0,0", ExternalForce="0,0,0") -> (Pose) {
    angularDampingOverride=0.9999
    angularSpringConstant=1500
    bAngularSpring=true
    bGravityOverrideInSimSpace=true
    bOverrideAngularDamping=true
    gravityScale=0.8
    simulationSpace=BoneRelative
    alphaBoolBlend {blendOption=Linear}
    boundBone {boneName=dyn_jacket_a_02}
    componentAppliedLinearAccClamp=<5000 5000 3500>
    componentLinearAccScale=<1.5 1.5 1.25>
    componentLinearVelScale=<1.5 1.5 1.25>
    physicsBodyDefinitions [{
      sphereCollisionRadius=0
      boundBone {boneName=dyn_jacket_a_02}
      boxExtents=<10 20 20>
      constraintSetup {angularLimitsMax=<15 25 15>, angularLimitsMin=<-15 -50 -15>, angularTarget=<0 0 0>}
      localJointOffset=<-2 0 0>
     }]
    relativeSpaceBone {boneName=dyn_root_simspace}
   }
   n43 = Anim Dynamics - Bone: dyn_jacket_a_03!(ComponentPose=@n42, Alpha=0.5, GravityOverride:Vector="0,0,0", ExternalForce="0,0,0") -> (Pose) {
    angularDampingOverride=0.9999
    angularSpringConstant=1500
    bAngularSpring=true
    bGravityOverrideInSimSpace=true
    bOverrideAngularDamping=true
    gravityScale=0.8
    simulationSpace=BoneRelative
    alphaBoolBlend {blendOption=Linear}
    boundBone {boneName=dyn_jacket_a_03}
    componentAppliedLinearAccClamp=<5000 5000 3500>
    componentLinearAccScale=<1.5 1.5 1.25>
    componentLinearVelScale=<1.5 1.5 1.25>
    physicsBodyDefinitions [{
      sphereCollisionRadius=0
      boundBone {boneName=dyn_jacket_a_03}
      boxExtents=<10 20 20>
      constraintSetup {angularLimitsMax=<15 25 15>, angularLimitsMin=<-15 -50 -15>, angularTarget=<0 0 0>}
      localJointOffset=<-2 0 0>
     }]
    relativeSpaceBone {boneName=dyn_root_simspace}
   }
   n44 = "Transform (Modify) Bone - Bone: dyn_jacket_a_MSTR"(ComponentPose=@n43, Alpha=1, Translation="0,0,0", Rotation="8,0,0", Scale="0.4,0.4,0.4") {
    rotationMode=BMM_Additive
    rotationSpace=BCS_BoneSpace
    scaleMode=BMM_Additive
    scaleSpace=BCS_BoneSpace
    translationMode=BMM_Additive
    translationSpace=BCS_BoneSpace
    alphaBoolBlend {blendOption=Linear}
    boneToModify {boneName=dyn_jacket_a_MSTR}
    rotation {pitch=8}
    scale=<0.4 0.4 0.4>
   }
   n45 = Get RightArmLift() -> (RightArmLift:float)
   n46 = "Transform (Modify) Bone - Bone: dyn_collar_r_MSTR"(ComponentPose=@n44, Alpha=@n45, Translation="0,0,3", Rotation="0,0,30", Scale="1,1,1") {
    rotationMode=BMM_Additive
    rotationSpace=BCS_BoneSpace
    scaleMode=BMM_Additive
    scaleSpace=BCS_BoneSpace
    translationMode=BMM_Additive
    translationSpace=BCS_BoneSpace
    alphaBoolBlend {blendOption=Linear}
    alphaScaleBiasClamp {bInterpResult=true, interpSpeedDecreasing=20, interpSpeedIncreasing=20}
    boneToModify {boneName=dyn_collar_r_MSTR}
    rotation {roll=30}
    translation {z=3}
   }
   n47 = "Transform (Modify) Bone - Bone: dyn_leaf_c_01"(ComponentPose=@n46, Alpha=@n45, Translation="0,0,0", Rotation="0,0,0", Scale="1,1,1") {
    rotationMode=BMM_Additive
    rotationSpace=BCS_BoneSpace
    scaleMode=BMM_Additive
    scaleSpace=BCS_BoneSpace
    translationMode=BMM_Additive
    translationSpace=BCS_BoneSpace
    alphaBoolBlend {blendOption=Linear}
    alphaScaleBiasClamp {bInterpResult=true, interpSpeedDecreasing=20, interpSpeedIncreasing=20}
    boneToModify {boneName=dyn_leaf_c_01}
   }
   n48 = "Transform (Modify) Bone - Bone: dyn_leaf_ef_01"(ComponentPose=@n47, Alpha=@n45, Translation="0,0,-1", Rotation="-5,15,0", Scale="1,1,1") {
    rotationMode=BMM_Additive
    rotationSpace=BCS_BoneSpace
    scaleMode=BMM_Additive
    scaleSpace=BCS_BoneSpace
    translationMode=BMM_Additive
    translationSpace=BCS_BoneSpace
    alphaBoolBlend {blendOption=Linear}
    alphaScaleBiasClamp {bInterpResult=true, interpSpeedDecreasing=20, interpSpeedIncreasing=20}
    boneToModify {boneName=dyn_leaf_ef_01}
    rotation {pitch=-5, yaw=15}
    translation {z=-1}
   }
   n49 = "Transform (Modify) Bone - Bone: dyn_leaf_e_01"(ComponentPose=@n48, Alpha=@n45, Translation="0,0,0", Rotation="0,0,0", Scale="1,1,1") {
    rotationMode=BMM_Additive
    rotationSpace=BCS_BoneSpace
    scaleMode=BMM_Additive
    scaleSpace=BCS_BoneSpace
    translationMode=BMM_Additive
    translationSpace=BCS_BoneSpace
    alphaBoolBlend {blendOption=Linear}
    alphaScaleBiasClamp {bInterpResult=true, interpSpeedDecreasing=20, interpSpeedIncreasing=20}
    boneToModify {boneName=dyn_leaf_e_01}
   }
   n50 = "Transform (Modify) Bone - Bone: dyn_leaf_f_01"(ComponentPose=@n49, Alpha=@n45, Translation="0,0,0", Rotation="-10,0,20", Scale="1,1,1") {
    rotationMode=BMM_Additive
    rotationSpace=BCS_BoneSpace
    scaleMode=BMM_Additive
    scaleSpace=BCS_BoneSpace
    translationMode=BMM_Additive
    translationSpace=BCS_BoneSpace
    alphaBoolBlend {blendOption=Linear}
    alphaScaleBiasClamp {bInterpResult=true, interpSpeedDecreasing=20, interpSpeedIncreasing=20}
    boneToModify {boneName=dyn_leaf_f_01}
    rotation {pitch=-10, roll=20}
   }
   n51 = "Transform (Modify) Bone - Bone: dyn_leaf_b_01"(ComponentPose=@n50, Alpha=@n45, Translation="0,0,0", Rotation="0,0,0", Scale="1,1,1") {
    rotationMode=BMM_Additive
    rotationSpace=BCS_BoneSpace
    scaleMode=BMM_Additive
    scaleSpace=BCS_BoneSpace
    translationMode=BMM_Additive
    translationSpace=BCS_BoneSpace
    alphaBoolBlend {blendOption=Linear}
    alphaScaleBiasClamp {bInterpResult=true, interpSpeedDecreasing=20, interpSpeedIncreasing=20}
    boneToModify {boneName=dyn_leaf_b_01}
   }
   n52 = "Transform (Modify) Bone - Bone: dyn_leaf_b_02"(ComponentPose=@n51, Alpha=@n45, Translation="0,0,0", Rotation="0,0,0", Scale="1,1,1") {
    rotationMode=BMM_Additive
    rotationSpace=BCS_BoneSpace
    scaleMode=BMM_Additive
    scaleSpace=BCS_BoneSpace
    translationMode=BMM_Additive
    translationSpace=BCS_BoneSpace
    alphaBoolBlend {blendOption=Linear}
    alphaScaleBiasClamp {bInterpResult=true, interpSpeedDecreasing=20, interpSpeedIncreasing=20}
    boneToModify {boneName=dyn_leaf_b_02}
   }
   n53 = "Transform (Modify) Bone - Bone: dyn_leaf_d_01"(ComponentPose=@n52, Alpha=@n45, Translation="0,0,0", Rotation="0,0,0", Scale="1,1,1") {
    rotationMode=BMM_Additive
    rotationSpace=BCS_BoneSpace
    scaleMode=BMM_Additive
    scaleSpace=BCS_BoneSpace
    translationMode=BMM_Additive
    translationSpace=BCS_BoneSpace
    alphaBoolBlend {blendOption=Linear}
    alphaScaleBiasClamp {bInterpResult=true, interpSpeedDecreasing=20, interpSpeedIncreasing=20}
    boneToModify {boneName=dyn_leaf_d_01}
   }
   n54 = "Transform (Modify) Bone - Bone: dyn_leaf_g_01"(ComponentPose=@n53, Alpha=@n45, Translation="0,0,0", Rotation="-10,5,5", Scale="1,1,1") {
    rotationMode=BMM_Additive
    rotationSpace=BCS_BoneSpace
    scaleMode=BMM_Additive
    scaleSpace=BCS_BoneSpace
    translationMode=BMM_Additive
    translationSpace=BCS_BoneSpace
    alphaBoolBlend {blendOption=Linear}
    alphaScaleBiasClamp {bInterpResult=true, interpSpeedDecreasing=20, interpSpeedIncreasing=20}
    boneToModify {boneName=dyn_leaf_g_01}
    rotation=<rot -10 5 5>
   }
   n55 = Component To Local(ComponentPose=@n54)
   n56 = Copy Pose From Mesh() {bUseAttachedParent=true}
   n57 = Get bIsCrouching()
   n58 = "Constraint - dyn_leaf_c_039 (PPPPPPPPPPPPPPPPPPP)"!(
    ComponentPose=Local To Component(LocalPose=Blend Poses by bool(BlendPose_0=@n55, BlendPose_1=@n56, bActiveValue=@n57) {blendType=Linear})
    Alpha=1
    ConstraintWeights_0=0.0195
    ConstraintWeights_1:float=0.1454
    ConstraintWeights_2:float=0.0102
    ConstraintWeights_3:float=0
    ConstraintWeights_4:float=0.0189
    ConstraintWeights_5:float=0.1829
    ConstraintWeights_6:float=0
    ConstraintWeights_7:float=0
    ConstraintWeights_8:float=0
    ConstraintWeights_9:float=0.4883
    ConstraintWeights_10:float=0
    ConstraintWeights_11:float=0.0046
    ConstraintWeights_12:float=0.0006
    ConstraintWeights_13:float=0.0012
    ConstraintWeights_14:float=0.0308
    ConstraintWeights_15:float=0.0976
    ConstraintWeights_16:float=0
    ConstraintWeights_17:float=0
    ConstraintWeights_18:float=0
   ) -> (Pose) {
    alphaBoolBlend {blendOption=Linear}
    boneToModify {boneName=dyn_leaf_c_039}
    constraintSetup [
     {targetBone {boneName=lowerarm_twist_02_r}}
     {targetBone {boneName=upperarm_r}}
     {targetBone {boneName=lowerarm_r}}
     {targetBone {boneName=clavicle_l}}
     {targetBone {boneName=lowerarm_twist_01_r}}
     {targetBone {boneName=upperarm_twist_01_r}}
     {targetBone {boneName=dyn_jacket_f_01}}
     {targetBone {boneName=dyn_r_collarFront}}
     {targetBone {boneName=dyn_jacket_e_01}}
     {targetBone {boneName=upperarm_twist_02_r}}
     {targetBone {boneName=spine_01}}
     {targetBone {boneName=spine_02}}
     {targetBone {boneName=spine_03}}
     {targetBone {boneName=spine_04}}
     {targetBone {boneName=spine_05}}
     {targetBone {boneName=clavicle_r}}
     {targetBone {boneName=neck_02}}
     {targetBone {boneName=neck_01}}
     {targetBone {boneName=dyn_r_collarSide}}
    ]
    constraintWeights [1,1,1,1,1,1,1,1,1,1,1,1,1,1,1,1,1,1,1]
   }
   n59 = Anim Dynamics - Bone: dyn_shoulderPad_a_01(ComponentPose=@n58, Alpha=0.75, ExternalForce="0,0,0") {
    angularDampingOverride=1
    bAngularSpring=true
    bGravityOverrideInSimSpace=true
    bOverrideAngularDamping=true
    gravityScale=0.6
    simulationSpace=RootRelative
    alphaBoolBlend {blendOption=Linear}
    alphaScaleBias {scale=2}
    boundBone {boneName=dyn_shoulderPad_a_01}
    physicsBodyDefinitions [{
      sphereCollisionRadius=0
      boundBone {boneName=dyn_shoulderPad_a_01}
      boxExtents=<20 25 25>
      constraintSetup {angularLimitsMax=<0 0 5>, angularLimitsMin=<0 -20 -5>, angularTarget=<0 0 0>}
      localJointOffset=<10 0 0>
     }]
   }
   n60 = Anim Dynamics - Bone: dyn_shoulderPad_b_01(ComponentPose=@n59, Alpha=0.75, ExternalForce="0,0,0") {
    angularDampingOverride=1
    bAngularSpring=true
    bGravityOverrideInSimSpace=true
    bOverrideAngularDamping=true
    gravityScale=0.6
    simulationSpace=RootRelative
    alphaBoolBlend {blendOption=Linear}
    alphaScaleBias {scale=2}
    boundBone {boneName=dyn_shoulderPad_b_01}
    physicsBodyDefinitions [{
      sphereCollisionRadius=0
      boundBone {boneName=dyn_shoulderPad_b_01}
      boxExtents=<20 25 25>
      constraintSetup {angularLimitsMax=<0 0 5>, angularLimitsMin=<0 -20 -5>, angularTarget=<0 0 0>}
      localJointOffset=<10 0 0>
     }]
   }
   n61 = Anim Dynamics - Bone: dyn_shoulderPad(ComponentPose=@n60, Alpha=0.25, ExternalForce="0,0,0") {
    angularDampingOverride=1
    bAngularSpring=true
    bGravityOverrideInSimSpace=true
    bOverrideAngularDamping=true
    gravityScale=0.6
    simulationSpace=RootRelative
    alphaBoolBlend {blendOption=Linear}
    alphaScaleBias {scale=2}
    boundBone {boneName=dyn_shoulderPad}
    physicsBodyDefinitions [{
      sphereCollisionRadius=0
      boundBone {boneName=dyn_shoulderPad}
      boxExtents=<20 25 25>
      constraintSetup {angularLimitsMax=<0 0 5>, angularLimitsMin=<0 -20 -5>, angularTarget=<0 0 0>}
      localJointOffset=<10 0 0>
     }]
   }
   n62 = Anim Dynamics - Bone: dyn_shoulderPad_c_01(ComponentPose=@n61, Alpha=0.75, ExternalForce="0,0,0") {
    angularDampingOverride=1
    bAngularSpring=true
    bGravityOverrideInSimSpace=true
    bOverrideAngularDamping=true
    gravityScale=0.6
    simulationSpace=RootRelative
    alphaBoolBlend {blendOption=Linear}
    alphaScaleBias {scale=2}
    boundBone {boneName=dyn_shoulderPad_c_01}
    physicsBodyDefinitions [{
      sphereCollisionRadius=0
      boundBone {boneName=dyn_shoulderPad_c_01}
      boxExtents=<20 25 25>
      constraintSetup {angularLimitsMax=<0 0 5>, angularLimitsMin=<0 -20 -5>, angularTarget=<0 0 0>}
      localJointOffset=<10 0 0>
     }]
   }
   n63 = Anim Dynamics - Bone: dyn_shoulderPad_c_02(ComponentPose=@n62, Alpha=0.75, ExternalForce="0,0,0") {
    angularDampingOverride=1
    bAngularSpring=true
    bGravityOverrideInSimSpace=true
    bOverrideAngularDamping=true
    gravityScale=0.6
    simulationSpace=RootRelative
    alphaBoolBlend {blendOption=Linear}
    alphaScaleBias {scale=2}
    boundBone {boneName=dyn_shoulderPad_c_02}
    physicsBodyDefinitions [{
      sphereCollisionRadius=0
      boundBone {boneName=dyn_shoulderPad_c_02}
      boxExtents=<20 25 25>
      constraintSetup {angularLimitsMax=<0 0 5>, angularLimitsMin=<0 -20 -5>, angularTarget=<0 0 0>}
      localJointOffset=<10 0 0>
     }]
   }
   n64 = Anim Dynamics - Bone: dyn_shoulderPad_d_01(ComponentPose=@n63, Alpha=0.75, ExternalForce="0,0,0") {
    angularDampingOverride=1
    bAngularSpring=true
    bGravityOverrideInSimSpace=true
    bOverrideAngularDamping=true
    gravityScale=0.6
    simulationSpace=RootRelative
    alphaBoolBlend {blendOption=Linear}
    alphaScaleBias {scale=2}
    boundBone {boneName=dyn_shoulderPad_d_01}
    physicsBodyDefinitions [{
      sphereCollisionRadius=0
      boundBone {boneName=dyn_shoulderPad_d_01}
      boxExtents=<20 25 25>
      constraintSetup {angularLimitsMax=<0 0 5>, angularLimitsMin=<0 -20 -5>, angularTarget=<0 0 0>}
      localJointOffset=<10 0 0>
     }]
   }
   n65 = Anim Dynamics - Bone: dyn_shoulderPad_d_02(ComponentPose=@n64, Alpha=0.75, ExternalForce="0,0,0") {
    angularDampingOverride=1
    bAngularSpring=true
    bGravityOverrideInSimSpace=true
    bOverrideAngularDamping=true
    gravityScale=0.6
    simulationSpace=RootRelative
    alphaBoolBlend {blendOption=Linear}
    alphaScaleBias {scale=2}
    boundBone {boneName=dyn_shoulderPad_d_02}
    physicsBodyDefinitions [{
      sphereCollisionRadius=0
      boundBone {boneName=dyn_shoulderPad_d_02}
      boxExtents=<20 25 25>
      constraintSetup {angularLimitsMax=<0 0 5>, angularLimitsMin=<0 -20 -5>, angularTarget=<0 0 0>}
      localJointOffset=<10 0 0>
     }]
   }
   n66 = Anim Dynamics - Bone: dyn_shoulderPad_e_01(ComponentPose=@n65, Alpha=0.75, ExternalForce="0,0,0") {
    angularDampingOverride=1
    bAngularSpring=true
    bGravityOverrideInSimSpace=true
    bOverrideAngularDamping=true
    gravityScale=0.6
    simulationSpace=RootRelative
    alphaBoolBlend {blendOption=Linear}
    alphaScaleBias {scale=2}
    boundBone {boneName=dyn_shoulderPad_e_01}
    physicsBodyDefinitions [{
      sphereCollisionRadius=0
      boundBone {boneName=dyn_shoulderPad_e_01}
      boxExtents=<20 25 25>
      constraintSetup {angularLimitsMax=<0 0 5>, angularLimitsMin=<0 -20 -5>, angularTarget=<0 0 0>}
      localJointOffset=<10 0 0>
     }]
   }
   n67 = Anim Dynamics - Bone: dyn_shoulderPad_e_02(ComponentPose=@n66, Alpha=0.75, ExternalForce="0,0,0") {
    angularDampingOverride=1
    bAngularSpring=true
    bGravityOverrideInSimSpace=true
    bOverrideAngularDamping=true
    gravityScale=0.6
    simulationSpace=RootRelative
    alphaBoolBlend {blendOption=Linear}
    alphaScaleBias {scale=2}
    boundBone {boneName=dyn_shoulderPad_e_02}
    physicsBodyDefinitions [{
      sphereCollisionRadius=0
      boundBone {boneName=dyn_shoulderPad_e_02}
      boxExtents=<20 25 25>
      constraintSetup {angularLimitsMax=<0 0 5>, angularLimitsMin=<0 -20 -5>, angularTarget=<0 0 0>}
      localJointOffset=<10 0 0>
     }]
   }
   n68 = Anim Dynamics - Bone: dyn_leaf_a_01!(ComponentPose=@n67, Alpha=0.25, GravityOverride:Vector="0,0,0", ExternalForce="0,0,0") -> (Pose) {
    angularDampingOverride=1
    angularSpringConstant=75
    bAngularSpring=true
    bGravityOverrideInSimSpace=true
    bOverrideAngularDamping=true
    gravityScale=0.45
    simulationSpace=RootRelative
    alphaBoolBlend {blendOption=Linear}
    alphaScaleBias {scale=2}
    boundBone {boneName=dyn_leaf_a_01}
    componentAppliedLinearAccClamp=<2500 2500 2500>
    componentLinearAccScale=<0.5 0.5 0.5>
    componentLinearVelScale=<0.5 0.5 0.5>
    physicsBodyDefinitions [{
      sphereCollisionRadius=0
      boundBone {boneName=dyn_leaf_a_01}
      boxExtents=<10 20 20>
      constraintSetup {angularLimitsMax=<15 20 15>, angularLimitsMin=<-15 -10 -15>, angularTarget=<0 0 0>}
      localJointOffset=<10 0 0>
     }]
   }
   n69 = Anim Dynamics - Bone: dyn_leaf_a_02!(ComponentPose=@n68, Alpha=0.25, GravityOverride:Vector="0,0,0", ExternalForce="0,0,0") -> (Pose) {
    angularDampingOverride=1
    angularSpringConstant=75
    bAngularSpring=true
    bGravityOverrideInSimSpace=true
    bOverrideAngularDamping=true
    gravityScale=0.45
    simulationSpace=RootRelative
    alphaBoolBlend {blendOption=Linear}
    alphaScaleBias {scale=2}
    boundBone {boneName=dyn_leaf_a_02}
    componentAppliedLinearAccClamp=<2500 2500 2500>
    componentLinearAccScale=<0.5 0.5 0.5>
    componentLinearVelScale=<0.5 0.5 0.5>
    physicsBodyDefinitions [{
      sphereCollisionRadius=0
      boundBone {boneName=dyn_leaf_a_02}
      boxExtents=<10 20 20>
      constraintSetup {angularLimitsMax=<15 20 15>, angularLimitsMin=<-15 -10 -15>, angularTarget=<0 0 0>}
      localJointOffset=<10 0 0>
     }]
   }
   n70 = Anim Dynamics - Bone: dyn_leaf_b_01!(ComponentPose=@n69, Alpha=0.25, GravityOverride:Vector="0,0,0", ExternalForce="0,0,0") -> (Pose) {
    angularDampingOverride=1
    angularSpringConstant=75
    bAngularSpring=true
    bGravityOverrideInSimSpace=true
    bOverrideAngularDamping=true
    gravityScale=0.45
    simulationSpace=RootRelative
    alphaBoolBlend {blendOption=Linear}
    alphaScaleBias {scale=2}
    boundBone {boneName=dyn_leaf_b_01}
    componentAppliedLinearAccClamp=<2500 2500 2500>
    componentLinearAccScale=<0.5 0.5 0.5>
    componentLinearVelScale=<0.5 0.5 0.5>
    physicsBodyDefinitions [{
      sphereCollisionRadius=0
      boundBone {boneName=dyn_leaf_b_01}
      boxExtents=<10 20 20>
      constraintSetup {angularLimitsMax=<15 0 15>, angularLimitsMin=<-15 -45 -15>, angularTarget=<0 0 0>}
      localJointOffset=<10 0 0>
     }]
   }
   n71 = Anim Dynamics - Bone: dyn_leaf_a_03!(ComponentPose=@n70, Alpha=0.25, GravityOverride:Vector="0,0,0", ExternalForce="0,0,0") -> (Pose) {
    angularDampingOverride=1
    angularSpringConstant=75
    bAngularSpring=true
    bGravityOverrideInSimSpace=true
    bOverrideAngularDamping=true
    gravityScale=0.45
    simulationSpace=RootRelative
    alphaBoolBlend {blendOption=Linear}
    alphaScaleBias {scale=2}
    boundBone {boneName=dyn_leaf_a_03}
    componentAppliedLinearAccClamp=<2500 2500 2500>
    componentLinearAccScale=<0.5 0.5 0.5>
    componentLinearVelScale=<0.5 0.5 0.5>
    physicsBodyDefinitions [{
      sphereCollisionRadius=0
      boundBone {boneName=dyn_leaf_a_03}
      boxExtents=<10 20 20>
      constraintSetup {angularLimitsMax=<15 30 15>, angularLimitsMin=<-15 -10 -15>, angularTarget=<0 0 0>}
      localJointOffset=<10 0 0>
     }]
   }
   n72 = Anim Dynamics - Bone: dyn_leaf_b_02!(ComponentPose=@n71, Alpha=0.25, GravityOverride:Vector="0,0,0", ExternalForce="0,0,0") -> (Pose) {
    angularDampingOverride=1
    angularSpringConstant=75
    bAngularSpring=true
    bGravityOverrideInSimSpace=true
    bOverrideAngularDamping=true
    gravityScale=0.45
    simulationSpace=RootRelative
    alphaBoolBlend {blendOption=Linear}
    alphaScaleBias {scale=2}
    boundBone {boneName=dyn_leaf_b_02}
    componentAppliedLinearAccClamp=<2500 2500 2500>
    componentLinearAccScale=<0.5 0.5 0.5>
    componentLinearVelScale=<0.5 0.5 0.5>
    physicsBodyDefinitions [{
      sphereCollisionRadius=0
      boundBone {boneName=dyn_leaf_b_02}
      boxExtents=<10 20 20>
      constraintSetup {angularLimitsMax=<15 0 15>, angularLimitsMin=<-15 -45 -15>, angularTarget=<0 0 0>}
      localJointOffset=<10 0 0>
     }]
   }
   n73 = Anim Dynamics - Bone: dyn_leaf_c_01!(ComponentPose=@n72, Alpha=0.25, GravityOverride:Vector="0,0,0", ExternalForce="0,0,0") -> (Pose) {
    angularDampingOverride=1
    angularSpringConstant=75
    bAngularSpring=true
    bGravityOverrideInSimSpace=true
    bOverrideAngularDamping=true
    gravityScale=0.45
    simulationSpace=RootRelative
    alphaBoolBlend {blendOption=Linear}
    alphaScaleBias {scale=2}
    boundBone {boneName=dyn_leaf_c_01}
    componentAppliedLinearAccClamp=<2500 2500 2500>
    componentLinearAccScale=<0.5 0.5 0.5>
    componentLinearVelScale=<0.5 0.5 0.5>
    physicsBodyDefinitions [{
      sphereCollisionRadius=0
      boundBone {boneName=dyn_leaf_c_01}
      boxExtents=<10 20 20>
      constraintSetup {angularLimitsMax=<15 0 15>, angularLimitsMin=<-15 -25 -15>, angularTarget=<0 0 0>}
      localJointOffset=<10 0 0>
     }]
   }
   n74 = Anim Dynamics - Bone: dyn_leaf_c_02!(ComponentPose=@n73, Alpha=0.25, GravityOverride:Vector="0,0,0", ExternalForce="0,0,0") -> (Pose) {
    angularDampingOverride=1
    angularSpringConstant=75
    bAngularSpring=true
    bGravityOverrideInSimSpace=true
    bOverrideAngularDamping=true
    gravityScale=0.45
    simulationSpace=RootRelative
    alphaBoolBlend {blendOption=Linear}
    alphaScaleBias {scale=2}
    boundBone {boneName=dyn_leaf_c_02}
    componentAppliedLinearAccClamp=<2500 2500 2500>
    componentLinearAccScale=<0.5 0.5 0.5>
    componentLinearVelScale=<0.5 0.5 0.5>
    physicsBodyDefinitions [{
      sphereCollisionRadius=0
      boundBone {boneName=dyn_leaf_c_02}
      boxExtents=<10 20 20>
      constraintSetup {angularLimitsMax=<15 0 15>, angularLimitsMin=<-15 -25 -15>, angularTarget=<0 0 0>}
      localJointOffset=<10 0 0>
     }]
   }
   n75 = Anim Dynamics - Bone: dyn_leaf_d_01!(ComponentPose=@n74, Alpha=0.25, GravityOverride:Vector="0,0,0", ExternalForce="0,0,0") -> (Pose) {
    angularDampingOverride=1
    angularSpringConstant=75
    bAngularSpring=true
    bGravityOverrideInSimSpace=true
    bOverrideAngularDamping=true
    gravityScale=0.45
    simulationSpace=RootRelative
    alphaBoolBlend {blendOption=Linear}
    alphaScaleBias {scale=2}
    boundBone {boneName=dyn_leaf_d_01}
    componentAppliedLinearAccClamp=<2500 2500 2500>
    componentLinearAccScale=<0.5 0.5 0.5>
    componentLinearVelScale=<0.5 0.5 0.5>
    physicsBodyDefinitions [{
      sphereCollisionRadius=0
      boundBone {boneName=dyn_leaf_d_01}
      boxExtents=<10 20 20>
      constraintSetup {angularLimitsMax=<15 5 15>, angularLimitsMin=<-15 -20 -15>, angularTarget=<0 0 0>}
      localJointOffset=<10 0 0>
     }]
   }
   n76 = Anim Dynamics - Bone: dyn_leaf_d_02!(ComponentPose=@n75, Alpha=0.25, GravityOverride:Vector="0,0,0", ExternalForce="0,0,0") -> (Pose) {
    angularDampingOverride=1
    angularSpringConstant=75
    bAngularSpring=true
    bGravityOverrideInSimSpace=true
    bOverrideAngularDamping=true
    gravityScale=0.45
    simulationSpace=RootRelative
    alphaBoolBlend {blendOption=Linear}
    alphaScaleBias {scale=2}
    boundBone {boneName=dyn_leaf_d_02}
    componentAppliedLinearAccClamp=<2500 2500 2500>
    componentLinearAccScale=<0.5 0.5 0.5>
    componentLinearVelScale=<0.5 0.5 0.5>
    physicsBodyDefinitions [{
      sphereCollisionRadius=0
      boundBone {boneName=dyn_leaf_d_02}
      boxExtents=<10 20 20>
      constraintSetup {angularLimitsMax=<15 5 15>, angularLimitsMin=<-15 -20 -15>, angularTarget=<0 0 0>}
      localJointOffset=<10 0 0>
     }]
   }
   n77 = Anim Dynamics - Bone: dyn_leaf_d_03!(ComponentPose=@n76, Alpha=0.25, GravityOverride:Vector="0,0,0", ExternalForce="0,0,0") -> (Pose) {
    angularDampingOverride=1
    angularSpringConstant=75
    bAngularSpring=true
    bGravityOverrideInSimSpace=true
    bOverrideAngularDamping=true
    gravityScale=0.45
    simulationSpace=RootRelative
    alphaBoolBlend {blendOption=Linear}
    alphaScaleBias {scale=2}
    boundBone {boneName=dyn_leaf_d_03}
    componentAppliedLinearAccClamp=<2500 2500 2500>
    componentLinearAccScale=<0.5 0.5 0.5>
    componentLinearVelScale=<0.5 0.5 0.5>
    physicsBodyDefinitions [{
      sphereCollisionRadius=0
      boundBone {boneName=dyn_leaf_d_03}
      boxExtents=<10 20 20>
      constraintSetup {angularLimitsMax=<15 5 15>, angularLimitsMin=<-15 -20 -15>, angularTarget=<0 0 0>}
      localJointOffset=<10 0 0>
     }]
   }
   n78 = Anim Dynamics - Bone: dyn_leaf_ef_01!(ComponentPose=@n77, Alpha=0.25, GravityOverride:Vector="0,0,0", ExternalForce="0,0,0") -> (Pose) {
    angularDampingOverride=1
    angularSpringConstant=75
    bAngularSpring=true
    bGravityOverrideInSimSpace=true
    bOverrideAngularDamping=true
    gravityScale=0.45
    simulationSpace=RootRelative
    alphaBoolBlend {blendOption=Linear}
    alphaScaleBias {scale=2}
    boundBone {boneName=dyn_leaf_ef_01}
    componentAppliedLinearAccClamp=<2500 2500 2500>
    componentLinearAccScale=<0.5 0.5 0.5>
    componentLinearVelScale=<0.5 0.5 0.5>
    physicsBodyDefinitions [{
      sphereCollisionRadius=0
      boundBone {boneName=dyn_leaf_ef_01}
      boxExtents=<10 20 20>
      constraintSetup {angularLimitsMax=<5 0 5>, angularLimitsMin=<-5 -20 -5>, angularTarget=<0 0 0>}
      localJointOffset=<10 0 0>
     }]
   }
   n79 = Anim Dynamics - Bone: dyn_leaf_e_01!(ComponentPose=@n78, Alpha=0.25, GravityOverride:Vector="0,0,0", ExternalForce="0,0,0") -> (Pose) {
    angularDampingOverride=1
    angularSpringConstant=75
    bAngularSpring=true
    bGravityOverrideInSimSpace=true
    bOverrideAngularDamping=true
    gravityScale=0.45
    simulationSpace=RootRelative
    alphaBoolBlend {blendOption=Linear}
    alphaScaleBias {scale=2}
    boundBone {boneName=dyn_leaf_e_01}
    componentAppliedLinearAccClamp=<2500 2500 2500>
    componentLinearAccScale=<0.5 0.5 0.5>
    componentLinearVelScale=<0.5 0.5 0.5>
    physicsBodyDefinitions [{
      sphereCollisionRadius=0
      boundBone {boneName=dyn_leaf_e_01}
      boxExtents=<10 20 20>
      constraintSetup {angularLimitsMax=<15 0 15>, angularLimitsMin=<-15 -20 -15>, angularTarget=<0 0 0>}
      localJointOffset=<10 0 0>
     }]
   }
   n80 = Anim Dynamics - Bone: dyn_leaf_e_02!(ComponentPose=@n79, Alpha=0.25, GravityOverride:Vector="0,0,0", ExternalForce="0,0,0") -> (Pose) {
    angularDampingOverride=1
    angularSpringConstant=75
    bAngularSpring=true
    bGravityOverrideInSimSpace=true
    bOverrideAngularDamping=true
    gravityScale=0.45
    simulationSpace=RootRelative
    alphaBoolBlend {blendOption=Linear}
    alphaScaleBias {scale=2}
    boundBone {boneName=dyn_leaf_e_02}
    componentAppliedLinearAccClamp=<2500 2500 2500>
    componentLinearAccScale=<0.5 0.5 0.5>
    componentLinearVelScale=<0.5 0.5 0.5>
    physicsBodyDefinitions [{
      sphereCollisionRadius=0
      boundBone {boneName=dyn_leaf_e_02}
      boxExtents=<10 20 20>
      constraintSetup {angularLimitsMax=<15 0 15>, angularLimitsMin=<-15 -20 -15>, angularTarget=<0 0 0>}
      localJointOffset=<10 0 0>
     }]
   }
   n81 = Anim Dynamics - Bone: dyn_leaf_f_01!(ComponentPose=@n80, Alpha=0.25, GravityOverride:Vector="0,0,0", ExternalForce="0,0,0") -> (Pose) {
    angularDampingOverride=1
    angularSpringConstant=75
    bAngularSpring=true
    bGravityOverrideInSimSpace=true
    bOverrideAngularDamping=true
    gravityScale=0.45
    simulationSpace=RootRelative
    alphaBoolBlend {blendOption=Linear}
    alphaScaleBias {scale=2}
    boundBone {boneName=dyn_leaf_f_01}
    componentAppliedLinearAccClamp=<2500 2500 2500>
    componentLinearAccScale=<0.5 0.5 0.5>
    componentLinearVelScale=<0.5 0.5 0.5>
    physicsBodyDefinitions [{
      sphereCollisionRadius=0
      boundBone {boneName=dyn_leaf_f_01}
      boxExtents=<10 20 20>
      constraintSetup {angularLimitsMax=<15 0 15>, angularLimitsMin=<-15 -20 -15>, angularTarget=<0 0 0>}
      localJointOffset=<10 0 0>
     }]
   }
   n82 = Anim Dynamics - Bone: dyn_leaf_f_02!(ComponentPose=@n81, Alpha=0.25, GravityOverride:Vector="0,0,0", ExternalForce="0,0,0") -> (Pose) {
    angularDampingOverride=1
    angularSpringConstant=75
    bAngularSpring=true
    bGravityOverrideInSimSpace=true
    bOverrideAngularDamping=true
    gravityScale=0.45
    simulationSpace=RootRelative
    alphaBoolBlend {blendOption=Linear}
    alphaScaleBias {scale=2}
    boundBone {boneName=dyn_leaf_f_02}
    componentAppliedLinearAccClamp=<2500 2500 2500>
    componentLinearAccScale=<0.5 0.5 0.5>
    componentLinearVelScale=<0.5 0.5 0.5>
    physicsBodyDefinitions [{
      sphereCollisionRadius=0
      boundBone {boneName=dyn_leaf_f_02}
      boxExtents=<10 20 20>
      constraintSetup {angularLimitsMax=<15 0 15>, angularLimitsMin=<-15 -20 -15>, angularTarget=<0 0 0>}
      localJointOffset=<10 0 0>
     }]
   }
   n83 = Anim Dynamics - Bone: dyn_leaf_g_01!(ComponentPose=@n82, Alpha=0.25, GravityOverride:Vector="0,0,0", ExternalForce="0,0,0") -> (Pose) {
    angularDampingOverride=1
    angularSpringConstant=75
    bAngularSpring=true
    bGravityOverrideInSimSpace=true
    bOverrideAngularDamping=true
    gravityScale=0.45
    simulationSpace=RootRelative
    alphaBoolBlend {blendOption=Linear}
    alphaScaleBias {scale=2}
    boundBone {boneName=dyn_leaf_g_01}
    componentAppliedLinearAccClamp=<2500 2500 2500>
    componentLinearAccScale=<0.5 0.5 0.5>
    componentLinearVelScale=<0.5 0.5 0.5>
    physicsBodyDefinitions [{
      sphereCollisionRadius=0
      boundBone {boneName=dyn_leaf_g_01}
      boxExtents=<10 20 20>
      constraintSetup {angularLimitsMax=<15 0 15>, angularLimitsMin=<-15 -45 -15>, angularTarget=<0 0 0>}
      localJointOffset=<10 0 0>
     }]
   }
   n84 = Anim Dynamics - Bone: dyn_leaf_g_02!(ComponentPose=@n83, Alpha=0.25, GravityOverride:Vector="0,0,0", ExternalForce="0,0,0") -> (Pose) {
    angularDampingOverride=1
    angularSpringConstant=75
    bAngularSpring=true
    bGravityOverrideInSimSpace=true
    bOverrideAngularDamping=true
    gravityScale=0.45
    simulationSpace=RootRelative
    alphaBoolBlend {blendOption=Linear}
    alphaScaleBias {scale=2}
    boundBone {boneName=dyn_leaf_g_02}
    componentAppliedLinearAccClamp=<2500 2500 2500>
    componentLinearAccScale=<0.5 0.5 0.5>
    componentLinearVelScale=<0.5 0.5 0.5>
    physicsBodyDefinitions [{
      sphereCollisionRadius=0
      boundBone {boneName=dyn_leaf_g_02}
      boxExtents=<10 20 20>
      constraintSetup {angularLimitsMax=<15 0 15>, angularLimitsMin=<-15 -45 -15>, angularTarget=<0 0 0>}
      localJointOffset=<10 0 0>
     }]
   }
   n85 = Anim Dynamics - Bone: dyn_leaf_g_03!(ComponentPose=@n84, Alpha=0.25, GravityOverride:Vector="0,0,0", ExternalForce="0,0,0") -> (Pose) {
    angularDampingOverride=1
    angularSpringConstant=75
    bAngularSpring=true
    bGravityOverrideInSimSpace=true
    bOverrideAngularDamping=true
    gravityScale=0.45
    simulationSpace=RootRelative
    alphaBoolBlend {blendOption=Linear}
    alphaScaleBias {scale=2}
    boundBone {boneName=dyn_leaf_g_03}
    componentAppliedLinearAccClamp=<2500 2500 2500>
    componentLinearAccScale=<0.5 0.5 0.5>
    componentLinearVelScale=<0.5 0.5 0.5>
    physicsBodyDefinitions [{
      sphereCollisionRadius=0
      boundBone {boneName=dyn_leaf_g_03}
      boxExtents=<10 20 20>
      constraintSetup {angularLimitsMax=<15 0 15>, angularLimitsMin=<-15 -45 -15>, angularTarget=<0 0 0>}
      localJointOffset=<10 0 0>
     }]
   }
   n86 = Anim Dynamics - Bone: dyn_leaf_hi_01!(ComponentPose=@n85, Alpha=0.25, GravityOverride:Vector="0,0,0", ExternalForce="0,0,0") -> (Pose) {
    angularDampingOverride=1
    angularSpringConstant=75
    bAngularSpring=true
    bGravityOverrideInSimSpace=true
    bOverrideAngularDamping=true
    gravityScale=0.45
    simulationSpace=RootRelative
    alphaBoolBlend {blendOption=Linear}
    alphaScaleBias {scale=2}
    boundBone {boneName=dyn_leaf_hi_01}
    componentAppliedLinearAccClamp=<2500 2500 2500>
    componentLinearAccScale=<0.5 0.5 0.5>
    componentLinearVelScale=<0.5 0.5 0.5>
    physicsBodyDefinitions [{
      sphereCollisionRadius=0
      boundBone {boneName=dyn_leaf_hi_01}
      boxExtents=<10 20 20>
      constraintSetup {angularLimitsMax=<15 0 15>, angularLimitsMin=<-15 -30 -15>, angularTarget=<0 0 0>}
      localJointOffset=<10 0 0>
     }]
   }
   n87 = Anim Dynamics - Bone: dyn_leaf_h_01!(ComponentPose=@n86, Alpha=0.25, GravityOverride:Vector="0,0,0", ExternalForce="0,0,0") -> (Pose) {
    angularDampingOverride=1
    angularSpringConstant=75
    bAngularSpring=true
    bGravityOverrideInSimSpace=true
    bOverrideAngularDamping=true
    gravityScale=0.45
    simulationSpace=RootRelative
    alphaBoolBlend {blendOption=Linear}
    alphaScaleBias {scale=2}
    boundBone {boneName=dyn_leaf_h_01}
    componentAppliedLinearAccClamp=<2500 2500 2500>
    componentLinearAccScale=<0.5 0.5 0.5>
    componentLinearVelScale=<0.5 0.5 0.5>
    physicsBodyDefinitions [{
      sphereCollisionRadius=0
      boundBone {boneName=dyn_leaf_h_01}
      boxExtents=<10 20 20>
      constraintSetup {angularLimitsMax=<15 0 15>, angularLimitsMin=<-15 -45 -15>, angularTarget=<0 0 0>}
      localJointOffset=<10 0 0>
     }]
   }
   n88 = Anim Dynamics - Bone: dyn_leaf_h_02!(ComponentPose=@n87, Alpha=0.25, GravityOverride:Vector="0,0,0", ExternalForce="0,0,0") -> (Pose) {
    angularDampingOverride=1
    angularSpringConstant=75
    bAngularSpring=true
    bGravityOverrideInSimSpace=true
    bOverrideAngularDamping=true
    gravityScale=0.45
    simulationSpace=RootRelative
    alphaBoolBlend {blendOption=Linear}
    alphaScaleBias {scale=2}
    boundBone {boneName=dyn_leaf_h_02}
    componentAppliedLinearAccClamp=<2500 2500 2500>
    componentLinearAccScale=<0.5 0.5 0.5>
    componentLinearVelScale=<0.5 0.5 0.5>
    physicsBodyDefinitions [{
      sphereCollisionRadius=0
      boundBone {boneName=dyn_leaf_h_02}
      boxExtents=<10 20 20>
      constraintSetup {angularLimitsMax=<15 0 15>, angularLimitsMin=<-15 -45 -15>, angularTarget=<0 0 0>}
      localJointOffset=<10 0 0>
     }]
   }
   n89 = Get AnimDynamicsAlpha_LeafI() -> (AnimDynamicsAlpha_LeafI:float)
   n90 = Anim Dynamics - Bone: dyn_leaf_i_01!(ComponentPose=@n88, Alpha=@n89, GravityOverride:Vector="0,0,0", ExternalForce="0,0,0") -> (Pose) {
    angularDampingOverride=1
    angularSpringConstant=75
    bAngularSpring=true
    bGravityOverrideInSimSpace=true
    bOverrideAngularDamping=true
    gravityScale=0.45
    simulationSpace=RootRelative
    alphaBoolBlend {blendOption=Linear}
    alphaScaleBias {scale=2}
    boundBone {boneName=dyn_leaf_i_01}
    componentAppliedLinearAccClamp=<2500 2500 2500>
    componentLinearAccScale=<0.5 0.5 0.5>
    componentLinearVelScale=<0.5 0.5 0.5>
    physicsBodyDefinitions [{
      sphereCollisionRadius=0
      boundBone {boneName=dyn_leaf_i_01}
      boxExtents=<10 20 20>
      constraintSetup {angularLimitsMax=<15 0 15>, angularLimitsMin=<-15 -35 -15>, angularTarget=<0 0 0>}
      localJointOffset=<10 0 0>
     }]
   }
   n91 = Anim Dynamics - Bone: dyn_leaf_i_02!(ComponentPose=@n90, Alpha=@n89, GravityOverride:Vector="0,0,0", ExternalForce="0,0,0") -> (Pose) {
    angularDampingOverride=1
    angularSpringConstant=75
    bAngularSpring=true
    bGravityOverrideInSimSpace=true
    bOverrideAngularDamping=true
    gravityScale=0.45
    simulationSpace=RootRelative
    alphaBoolBlend {blendOption=Linear}
    alphaScaleBias {scale=2}
    boundBone {boneName=dyn_leaf_i_02}
    componentAppliedLinearAccClamp=<2500 2500 2500>
    componentLinearAccScale=<0.5 0.5 0.5>
    componentLinearVelScale=<0.5 0.5 0.5>
    physicsBodyDefinitions [{
      sphereCollisionRadius=0
      boundBone {boneName=dyn_leaf_i_02}
      boxExtents=<10 20 20>
      constraintSetup {angularLimitsMax=<15 0 15>, angularLimitsMin=<-15 -35 -15>, angularTarget=<0 0 0>}
      localJointOffset=<10 0 0>
     }]
   }
   n92 = Anim Dynamics - Bone: dyn_jacket_a_01(ComponentPose=@n91, Alpha=0.5, GravityOverride="0,0,0", ExternalForce="0,0,0") {
    angularDampingOverride=0.999
    angularSpringConstant=50
    bAngularSpring=true
    bGravityOverrideInSimSpace=true
    bOverrideAngularDamping=true
    gravityScale=0.8
    simulationSpace=BoneRelative
    alphaBoolBlend {blendOption=Linear}
    boundBone {boneName=dyn_jacket_a_01}
    componentAppliedLinearAccClamp=<5000 5000 3500>
    componentLinearAccScale=<1.25 1.25 1.1>
    componentLinearVelScale=<1.25 1.25 1.1>
    physicsBodyDefinitions [{
      sphereCollisionRadius=0
      boundBone {boneName=dyn_jacket_a_01}
      boxExtents=<10 20 20>
      constraintSetup {angularLimitsMax=<15 8 15>, angularLimitsMin=<-15 -90 -15>, angularTarget=<0 0 0>}
      localJointOffset=<-2 0 0>
     }]
    relativeSpaceBone {boneName=dyn_root_simspace}
   }
   n93 = Anim Dynamics - Bone: dyn_jacket_a_02(ComponentPose=@n92, Alpha=0.5, GravityOverride="0,0,0", ExternalForce="0,0,0") {
    angularDampingOverride=0.999
    angularSpringConstant=50
    bAngularSpring=true
    bGravityOverrideInSimSpace=true
    bOverrideAngularDamping=true
    gravityScale=0.8
    simulationSpace=BoneRelative
    alphaBoolBlend {blendOption=Linear}
    boundBone {boneName=dyn_jacket_a_02}
    componentAppliedLinearAccClamp=<5000 5000 3500>
    componentLinearAccScale=<1.25 1.25 1.1>
    componentLinearVelScale=<1.25 1.25 1.1>
    physicsBodyDefinitions [{
      sphereCollisionRadius=0
      boundBone {boneName=dyn_jacket_a_02}
      boxExtents=<10 20 20>
      constraintSetup {angularLimitsMax=<15 8 15>, angularLimitsMin=<-15 -90 -15>, angularTarget=<0 0 0>}
      localJointOffset=<-2 0 0>
     }]
    relativeSpaceBone {boneName=dyn_root_simspace}
   }
   n94 = Anim Dynamics - Bone: dyn_jacket_a_03(ComponentPose=@n93, Alpha=0.5, GravityOverride="0,0,0", ExternalForce="0,0,0") {
    angularDampingOverride=0.999
    angularSpringConstant=50
    bAngularSpring=true
    bGravityOverrideInSimSpace=true
    bOverrideAngularDamping=true
    gravityScale=0.8
    simulationSpace=BoneRelative
    alphaBoolBlend {blendOption=Linear}
    boundBone {boneName=dyn_jacket_a_03}
    componentAppliedLinearAccClamp=<5000 5000 3500>
    componentLinearAccScale=<1.25 1.25 1.1>
    componentLinearVelScale=<1.25 1.25 1.1>
    physicsBodyDefinitions [{
      sphereCollisionRadius=0
      boundBone {boneName=dyn_jacket_a_03}
      boxExtents=<10 20 20>
      constraintSetup {angularLimitsMax=<15 8 15>, angularLimitsMin=<-15 -90 -15>, angularTarget=<0 0 0>}
      localJointOffset=<-2 0 0>
     }]
    relativeSpaceBone {boneName=dyn_root_simspace}
   }
   n95 = Anim Dynamics - Bone: dyn_jacket_b_01!(ComponentPose=@n94, Alpha=0.65, GravityOverride:Vector="0,0,0", ExternalForce="0,0,0") -> (Pose) {
    angularDampingOverride=0.999
    angularSpringConstant=50
    bAngularSpring=true
    bGravityOverrideInSimSpace=true
    bOverrideAngularDamping=true
    gravityScale=0.8
    simulationSpace=BoneRelative
    alphaBoolBlend {blendOption=Linear}
    boundBone {boneName=dyn_jacket_b_01}
    componentAppliedLinearAccClamp=<5000 5000 3500>
    componentLinearAccScale=<1.25 1.25 1.1>
    componentLinearVelScale=<1.25 1.25 1.1>
    physicsBodyDefinitions [{
      sphereCollisionRadius=0
      boundBone {boneName=dyn_jacket_b_01}
      boxExtents=<10 20 20>
      constraintSetup {angularLimitsMax=<15 5 15>, angularLimitsMin=<-15 -90 -15>, angularTarget=<0 0 0>}
      localJointOffset=<-2 0 0>
     }]
    relativeSpaceBone {boneName=dyn_root_simspace}
   }
   n96 = Anim Dynamics - Bone: dyn_jacket_b_02!(ComponentPose=@n95, Alpha=0.65, GravityOverride:Vector="0,0,0", ExternalForce="0,0,0") -> (Pose) {
    angularDampingOverride=0.999
    angularSpringConstant=50
    bAngularSpring=true
    bGravityOverrideInSimSpace=true
    bOverrideAngularDamping=true
    gravityScale=0.8
    simulationSpace=BoneRelative
    alphaBoolBlend {blendOption=Linear}
    boundBone {boneName=dyn_jacket_b_02}
    componentAppliedLinearAccClamp=<5000 5000 3500>
    componentLinearAccScale=<1.25 1.25 1.1>
    componentLinearVelScale=<1.25 1.25 1.1>
    physicsBodyDefinitions [{
      sphereCollisionRadius=0
      boundBone {boneName=dyn_jacket_b_02}
      boxExtents=<10 20 20>
      constraintSetup {angularLimitsMax=<15 5 15>, angularLimitsMin=<-15 -90 -15>, angularTarget=<0 0 0>}
      localJointOffset=<-2 0 0>
     }]
    relativeSpaceBone {boneName=dyn_root_simspace}
   }
   n97 = Anim Dynamics - Bone: dyn_jacket_c_01!(ComponentPose=@n96, Alpha=0.6, GravityOverride:Vector="0,0,0", ExternalForce="0,0,0") -> (Pose) {
    angularDampingOverride=0.999
    angularSpringConstant=50
    bAngularSpring=true
    bGravityOverrideInSimSpace=true
    bOverrideAngularDamping=true
    gravityScale=0.8
    simulationSpace=BoneRelative
    alphaBoolBlend {blendOption=Linear}
    boundBone {boneName=dyn_jacket_c_01}
    componentAppliedLinearAccClamp=<5000 5000 3500>
    componentLinearAccScale=<1.25 1.25 1.1>
    componentLinearVelScale=<1.25 1.25 1.1>
    physicsBodyDefinitions [{
      sphereCollisionRadius=0
      boundBone {boneName=dyn_jacket_c_01}
      boxExtents=<10 20 20>
      constraintSetup {angularLimitsMax=<15 5 15>, angularLimitsMin=<-15 -90 -15>, angularTarget=<0 0 0>}
      localJointOffset=<-2 0 0>
     }]
    relativeSpaceBone {boneName=dyn_root_simspace}
   }
   n98 = Anim Dynamics - Bone: dyn_jacket_c_02!(ComponentPose=@n97, Alpha=0.6, GravityOverride:Vector="0,0,0", ExternalForce="0,0,0") -> (Pose) {
    angularDampingOverride=0.999
    angularSpringConstant=50
    bAngularSpring=true
    bGravityOverrideInSimSpace=true
    bOverrideAngularDamping=true
    gravityScale=0.8
    simulationSpace=BoneRelative
    alphaBoolBlend {blendOption=Linear}
    boundBone {boneName=dyn_jacket_c_02}
    componentAppliedLinearAccClamp=<5000 5000 3500>
    componentLinearAccScale=<1.25 1.25 1.1>
    componentLinearVelScale=<1.25 1.25 1.1>
    physicsBodyDefinitions [{
      sphereCollisionRadius=0
      boundBone {boneName=dyn_jacket_c_02}
      boxExtents=<10 20 20>
      constraintSetup {angularLimitsMax=<15 5 15>, angularLimitsMin=<-15 -90 -15>, angularTarget=<0 0 0>}
      localJointOffset=<-2 0 0>
     }]
    relativeSpaceBone {boneName=dyn_root_simspace}
   }
   n99 = Anim Dynamics - Bone: dyn_jacket_d_01!(ComponentPose=@n98, Alpha=0.8, GravityOverride:Vector="0,0,0", ExternalForce="0,0,0") -> (Pose) {
    angularDampingOverride=0.999
    angularSpringConstant=50
    bAngularSpring=true
    bGravityOverrideInSimSpace=true
    bOverrideAngularDamping=true
    gravityScale=0.8
    simulationSpace=BoneRelative
    alphaBoolBlend {blendOption=Linear}
    boundBone {boneName=dyn_jacket_d_01}
    componentAppliedLinearAccClamp=<5000 5000 3500>
    componentLinearAccScale=<1.25 1.25 1.1>
    componentLinearVelScale=<1.25 1.25 1.1>
    physicsBodyDefinitions [{
      sphereCollisionRadius=0
      boundBone {boneName=dyn_jacket_d_01}
      boxExtents=<10 20 20>
      constraintSetup {angularLimitsMax=<15 5 15>, angularLimitsMin=<-15 -90 -15>, angularTarget=<0 0 0>}
      localJointOffset=<2 0 0>
     }]
    relativeSpaceBone {boneName=dyn_root_simspace}
   }
   n100 = Anim Dynamics - Bone: dyn_jacket_d_02!(ComponentPose=@n99, Alpha=0.8, GravityOverride:Vector="0,0,0", ExternalForce="0,0,0") -> (Pose) {
    angularDampingOverride=0.999
    angularSpringConstant=50
    bAngularSpring=true
    bGravityOverrideInSimSpace=true
    bOverrideAngularDamping=true
    gravityScale=0.8
    simulationSpace=BoneRelative
    alphaBoolBlend {blendOption=Linear}
    boundBone {boneName=dyn_jacket_d_02}
    componentAppliedLinearAccClamp=<5000 5000 3500>
    componentLinearAccScale=<1.25 1.25 1.1>
    componentLinearVelScale=<1.25 1.25 1.1>
    physicsBodyDefinitions [{
      sphereCollisionRadius=0
      boundBone {boneName=dyn_jacket_d_02}
      boxExtents=<10 20 20>
      constraintSetup {angularLimitsMax=<15 5 15>, angularLimitsMin=<-15 -90 -15>, angularTarget=<0 0 0>}
      localJointOffset=<2 0 0>
     }]
    relativeSpaceBone {boneName=dyn_root_simspace}
   }
   n101 = Anim Dynamics - Bone: dyn_jacket_e_01!(ComponentPose=@n100, Alpha=0.5, GravityOverride:Vector="0,0,0", ExternalForce="0,0,0") -> (Pose) {
    angularDampingOverride=0.999
    angularSpringConstant=50
    bAngularSpring=true
    bGravityOverrideInSimSpace=true
    bOverrideAngularDamping=true
    gravityScale=0.8
    simulationSpace=BoneRelative
    alphaBoolBlend {blendOption=Linear}
    boundBone {boneName=dyn_jacket_e_01}
    componentAppliedLinearAccClamp=<5000 5000 3500>
    componentLinearAccScale=<1.25 1.25 1.1>
    componentLinearVelScale=<1.25 1.25 1.1>
    physicsBodyDefinitions [{
      sphereCollisionRadius=0
      boundBone {boneName=dyn_jacket_e_01}
      boxExtents=<10 20 20>
      constraintSetup {angularLimitsMax=<15 5 15>, angularLimitsMin=<-15 -90 -15>, angularTarget=<0 0 0>}
      localJointOffset=<2 0 0>
     }]
    relativeSpaceBone {boneName=dyn_root_simspace}
   }
   n102 = Anim Dynamics - Bone: dyn_jacket_e_02!(ComponentPose=@n101, Alpha=0.5, GravityOverride:Vector="0,0,0", ExternalForce="0,0,0") -> (Pose) {
    angularDampingOverride=0.999
    angularSpringConstant=50
    bAngularSpring=true
    bGravityOverrideInSimSpace=true
    bOverrideAngularDamping=true
    gravityScale=0.8
    simulationSpace=BoneRelative
    alphaBoolBlend {blendOption=Linear}
    boundBone {boneName=dyn_jacket_e_02}
    componentAppliedLinearAccClamp=<5000 5000 3500>
    componentLinearAccScale=<1.25 1.25 1.1>
    componentLinearVelScale=<1.25 1.25 1.1>
    physicsBodyDefinitions [{
      sphereCollisionRadius=0
      boundBone {boneName=dyn_jacket_e_02}
      boxExtents=<10 20 20>
      constraintSetup {angularLimitsMax=<15 5 15>, angularLimitsMin=<-15 -90 -15>, angularTarget=<0 0 0>}
      localJointOffset=<2 0 0>
     }]
    relativeSpaceBone {boneName=dyn_root_simspace}
   }
   n103 = Anim Dynamics - Bone: dyn_jacket_f_01!(ComponentPose=@n102, Alpha=0.7, GravityOverride:Vector="0,0,0", ExternalForce="0,0,0") -> (Pose) {
    angularDampingOverride=0.999
    angularSpringConstant=50
    bAngularSpring=true
    bGravityOverrideInSimSpace=true
    bOverrideAngularDamping=true
    gravityScale=0.8
    simulationSpace=BoneRelative
    alphaBoolBlend {blendOption=Linear}
    boundBone {boneName=dyn_jacket_f_01}
    componentAppliedLinearAccClamp=<5000 5000 3500>
    componentLinearAccScale=<1.25 1.25 1.1>
    componentLinearVelScale=<1.25 1.25 1.1>
    physicsBodyDefinitions [{
      sphereCollisionRadius=0
      boundBone {boneName=dyn_jacket_f_01}
      boxExtents=<10 20 20>
      constraintSetup {angularLimitsMax=<15 5 15>, angularLimitsMin=<-15 -90 -15>, angularTarget=<0 0 0>}
      localJointOffset=<2 0 0>
     }]
    relativeSpaceBone {boneName=dyn_root_simspace}
   }
   n104 = Anim Dynamics - Bone: dyn_jacket_f_02!(ComponentPose=@n103, Alpha=0.7, GravityOverride:Vector="0,0,0", ExternalForce="0,0,0") -> (Pose) {
    angularDampingOverride=0.999
    angularSpringConstant=50
    bAngularSpring=true
    bGravityOverrideInSimSpace=true
    bOverrideAngularDamping=true
    gravityScale=0.8
    simulationSpace=BoneRelative
    alphaBoolBlend {blendOption=Linear}
    boundBone {boneName=dyn_jacket_f_02}
    componentAppliedLinearAccClamp=<5000 5000 3500>
    componentLinearAccScale=<1.25 1.25 1.1>
    componentLinearVelScale=<1.25 1.25 1.1>
    physicsBodyDefinitions [{
      sphereCollisionRadius=0
      boundBone {boneName=dyn_jacket_f_02}
      boxExtents=<10 20 20>
      constraintSetup {angularLimitsMax=<15 5 15>, angularLimitsMin=<-15 -90 -15>, angularTarget=<0 0 0>}
      localJointOffset=<2 0 0>
     }]
    relativeSpaceBone {boneName=dyn_root_simspace}
   }
   n105 = Anim Dynamics - Bone: dyn_jacket_g_01!(ComponentPose=@n104, Alpha=0.5, GravityOverride:Vector="0,0,0", ExternalForce="0,0,0") -> (Pose) {
    angularDampingOverride=0.999
    angularSpringConstant=50
    bAngularSpring=true
    bGravityOverrideInSimSpace=true
    bOverrideAngularDamping=true
    gravityScale=0.8
    simulationSpace=BoneRelative
    alphaBoolBlend {blendOption=Linear}
    boundBone {boneName=dyn_jacket_g_01}
    componentAppliedLinearAccClamp=<5000 5000 3500>
    componentLinearAccScale=<1.25 1.25 1.1>
    componentLinearVelScale=<1.25 1.25 1.1>
    physicsBodyDefinitions [{
      sphereCollisionRadius=0
      boundBone {boneName=dyn_jacket_g_01}
      boxExtents=<10 20 20>
      constraintSetup {angularLimitsMax=<15 8 15>, angularLimitsMin=<-15 -90 -15>, angularTarget=<0 0 0>}
      localJointOffset=<2 0 0>
     }]
    relativeSpaceBone {boneName=dyn_root_simspace}
   }
   n106 = Anim Dynamics - Bone: dyn_jacket_g_02!(ComponentPose=@n105, Alpha=0.5, GravityOverride:Vector="0,0,0", ExternalForce="0,0,0") -> (Pose) {
    angularDampingOverride=0.999
    angularSpringConstant=50
    bAngularSpring=true
    bGravityOverrideInSimSpace=true
    bOverrideAngularDamping=true
    gravityScale=0.8
    simulationSpace=BoneRelative
    alphaBoolBlend {blendOption=Linear}
    boundBone {boneName=dyn_jacket_g_02}
    componentAppliedLinearAccClamp=<5000 5000 3500>
    componentLinearAccScale=<1.25 1.25 1.1>
    componentLinearVelScale=<1.25 1.25 1.1>
    physicsBodyDefinitions [{
      sphereCollisionRadius=0
      boundBone {boneName=dyn_jacket_g_02}
      boxExtents=<10 20 20>
      constraintSetup {angularLimitsMax=<15 8 15>, angularLimitsMin=<-15 -90 -15>, angularTarget=<0 0 0>}
      localJointOffset=<2 0 0>
     }]
    relativeSpaceBone {boneName=dyn_root_simspace}
   }
   n107 = Anim Dynamics - Bone: dyn_jacket_g_03!(ComponentPose=@n106, Alpha=0.5, GravityOverride:Vector="0,0,0", ExternalForce="0,0,0") -> (Pose) {
    angularDampingOverride=0.999
    angularSpringConstant=50
    bAngularSpring=true
    bGravityOverrideInSimSpace=true
    bOverrideAngularDamping=true
    gravityScale=0.8
    simulationSpace=BoneRelative
    alphaBoolBlend {blendOption=Linear}
    boundBone {boneName=dyn_jacket_g_03}
    componentAppliedLinearAccClamp=<5000 5000 3500>
    componentLinearAccScale=<1.25 1.25 1.1>
    componentLinearVelScale=<1.25 1.25 1.1>
    physicsBodyDefinitions [{
      sphereCollisionRadius=0
      boundBone {boneName=dyn_jacket_g_03}
      boxExtents=<10 20 20>
      constraintSetup {angularLimitsMax=<15 8 15>, angularLimitsMin=<-15 -90 -15>, angularTarget=<0 0 0>}
      localJointOffset=<2 0 0>
     }]
    relativeSpaceBone {boneName=dyn_root_simspace}
   }
   n108 = RigidBody!(ComponentPose=@n107, Alpha:float, OverrideWorldGravity="0,0,-980", ExternalForce) -> (Pose) {
    bClampLinearTranslationLimitToRefPose=true
    bOverrideWorldGravity=true
    overridePhysicsAsset=&/M_MED_Critter_Cuddle_Physics
    simulationSpace=BaseBoneSpace
    alphaBoolBlend {blendOption=Linear}
    baseBoneRef {boneName=root}
    componentAppliedLinearAccClamp=<5000 5000 5000>
    overrideWorldGravity {z=-980}
   }
   n109 = "Constraint - dyn_leaf_c_01 (P)"(ComponentPose=@n108, Alpha=1, ConstraintWeights_0=1) {
    alphaBoolBlend {blendOption=Linear}
    boneToModify {boneName=dyn_leaf_c_01}
    constraintSetup [{targetBone {boneName=clavicle_r}}]
    constraintWeights [1]
   }
   n110 = "Transform (Modify) Bone - Bone: dyn_leaf_c_01"!(ComponentPose=@n109, bAlphaBoolEnabled:bool=True, Translation="0,1,-1", Rotation="-15,10,0", Scale="1,1,1") -> (Pose) {
    alphaInputType=Bool
    rotationMode=BMM_Additive
    rotationSpace=BCS_BoneSpace
    scaleMode=BMM_Additive
    scaleSpace=BCS_BoneSpace
    translationMode=BMM_Additive
    translationSpace=BCS_BoneSpace
    alphaBoolBlend {blendInTime=0.1, blendOption=Linear, blendOutTime=0.1}
    alphaScaleBiasClamp {bInterpResult=true, interpSpeedDecreasing=20, interpSpeedIncreasing=20}
    boneToModify {boneName=dyn_leaf_c_01}
    rotation {pitch=-15, yaw=10}
    translation {y=1, z=-1}
   }
   n111 = "Constraint - dyn_collar_r_MSTR (R)"(ComponentPose=@n110, Alpha=1, ConstraintWeights_0=1) {
    alphaBoolBlend {blendOption=Linear}
    boneToModify {boneName=dyn_collar_r_MSTR}
    constraintSetup [{transformType=Rotation, targetBone {boneName=clavicle_r}}]
    constraintWeights [1]
   }
   n112 = Get RightArmLift()
   n113 = "Transform (Modify) Bone - Bone: dyn_leaf_c_01"!(ComponentPose=@n111, Alpha:float=@n112, Translation="0,1.5,1", Rotation="0,0,80", Scale="1,1,1") -> (Pose) {
    rotationMode=BMM_Additive
    rotationSpace=BCS_BoneSpace
    scaleMode=BMM_Additive
    scaleSpace=BCS_BoneSpace
    translationMode=BMM_Additive
    translationSpace=BCS_BoneSpace
    alphaBoolBlend {blendOption=Linear}
    alphaScaleBiasClamp {bInterpResult=true, interpSpeedDecreasing=20, interpSpeedIncreasing=20}
    boneToModify {boneName=dyn_leaf_c_01}
    rotation {roll=80}
    translation {y=1.5, z=1}
   }
   n114 = "Transform (Modify) Bone - Bone: dyn_leaf_ef_01"(ComponentPose=@n113, Alpha=@n112, Translation="0,0,-1", Rotation="-5,0,30", Scale="1,1,1") {
    rotationMode=BMM_Additive
    rotationSpace=BCS_BoneSpace
    scaleMode=BMM_Additive
    scaleSpace=BCS_BoneSpace
    translationMode=BMM_Additive
    translationSpace=BCS_BoneSpace
    alphaBoolBlend {blendOption=Linear}
    alphaScaleBiasClamp {bInterpResult=true, interpSpeedDecreasing=20, interpSpeedIncreasing=20}
    boneToModify {boneName=dyn_leaf_ef_01}
    rotation {pitch=-5, roll=30}
    translation {z=-1}
   }
   n115 = "Transform (Modify) Bone - Bone: dyn_leaf_e_01"(ComponentPose=@n114, Alpha=@n112, Translation="0,0,1", Rotation="-10,30,0", Scale="1,1,1") {
    rotationMode=BMM_Additive
    rotationSpace=BCS_BoneSpace
    scaleMode=BMM_Additive
    scaleSpace=BCS_BoneSpace
    translationMode=BMM_Additive
    translationSpace=BCS_BoneSpace
    alphaBoolBlend {blendOption=Linear}
    alphaScaleBiasClamp {bInterpResult=true, interpSpeedDecreasing=20, interpSpeedIncreasing=20}
    boneToModify {boneName=dyn_leaf_e_01}
    rotation {pitch=-10, yaw=30}
    translation {z=1}
   }
   n116 = "Transform (Modify) Bone - Bone: dyn_leaf_f_01"(ComponentPose=@n115, Alpha=@n112, Translation="0,0,0", Rotation="-20,0,-20", Scale="1,1,1") {
    rotationMode=BMM_Additive
    rotationSpace=BCS_BoneSpace
    scaleMode=BMM_Additive
    scaleSpace=BCS_BoneSpace
    translationMode=BMM_Additive
    translationSpace=BCS_BoneSpace
    alphaBoolBlend {blendOption=Linear}
    alphaScaleBiasClamp {bInterpResult=true, interpSpeedDecreasing=20, interpSpeedIncreasing=20}
    boneToModify {boneName=dyn_leaf_f_01}
    rotation {pitch=-20, roll=-20}
   }
   n117 = Get bIsWearingBackpackOrCape() -> (bIsWearingBackpackOrCape:bool)
   n118 = "Transform (Modify) Bone - Bone: dyn_leaf_h_01"(ComponentPose=@n116, bAlphaBoolEnabled=@n117, Translation="0,0,0", Rotation="10,0,0", Scale="0.800,0.517,0.800") {
    alphaInputType=Bool
    rotationMode=BMM_Replace
    rotationSpace=BCS_BoneSpace
    scaleMode=BMM_Replace
    scaleSpace=BCS_BoneSpace
    translationMode=BMM_Additive
    translationSpace=BCS_BoneSpace
    alphaBoolBlend {blendOption=Linear}
    alphaScaleBiasClamp {bInterpResult=true, interpSpeedDecreasing=20, interpSpeedIncreasing=20}
    boneToModify {boneName=dyn_leaf_h_01}
    rotation {pitch=10}
    scale=<0.8 0.5173 0.8>
   }
   n119 = "Transform (Modify) Bone - Bone: dyn_leaf_h_02"!(ComponentPose=@n118, bAlphaBoolEnabled:bool=@n117, Translation="0,0,0", Rotation="0,0.331,0.55", Scale="0.600,0.850,0.600") -> (Pose) {
    alphaInputType=Bool
    rotationMode=BMM_Replace
    rotationSpace=BCS_BoneSpace
    scaleMode=BMM_Replace
    scaleSpace=BCS_BoneSpace
    translationMode=BMM_Additive
    translationSpace=BCS_BoneSpace
    alphaBoolBlend {blendOption=Linear}
    alphaScaleBiasClamp {bInterpResult=true, interpSpeedDecreasing=20, interpSpeedIncreasing=20}
    boneToModify {boneName=dyn_leaf_h_02}
    rotation {roll=0.55, yaw=0.331}
    scale=<0.6 0.85 0.6>
   }
   n120 = "Transform (Modify) Bone - Bone: dyn_leaf_hi_01"(ComponentPose=@n119, bAlphaBoolEnabled=@n117, Translation="0,0,0", Rotation="30,0,0", Scale="0.550,0.800,0.800") {
    alphaInputType=Bool
    rotationMode=BMM_Replace
    rotationSpace=BCS_BoneSpace
    scaleMode=BMM_Replace
    scaleSpace=BCS_BoneSpace
    translationMode=BMM_Additive
    translationSpace=BCS_BoneSpace
    alphaBoolBlend {blendOption=Linear}
    alphaScaleBiasClamp {bInterpResult=true, interpSpeedDecreasing=20, interpSpeedIncreasing=20}
    boneToModify {boneName=dyn_leaf_hi_01}
    rotation {pitch=30}
    scale=<0.55 0.8 0.8>
   }
   n121 = "Transform (Modify) Bone - Bone: dyn_leaf_i_01"(ComponentPose=@n120, bAlphaBoolEnabled=@n117, Translation="0,0,0", Rotation="0,0,0", Scale="0.823,0.560,0.800") {
    alphaInputType=Bool
    rotationMode=BMM_Replace
    rotationSpace=BCS_BoneSpace
    scaleMode=BMM_Replace
    scaleSpace=BCS_BoneSpace
    translationMode=BMM_Additive
    translationSpace=BCS_BoneSpace
    alphaBoolBlend {blendOption=Linear}
    alphaScaleBiasClamp {bInterpResult=true, interpSpeedDecreasing=20, interpSpeedIncreasing=20}
    boneToModify {boneName=dyn_leaf_i_01}
    scale=<0.823 0.56 0.8>
   }
   n122 = "Transform (Modify) Bone - Bone: dyn_leaf_i_02"!(ComponentPose=@n121, bAlphaBoolEnabled:bool=@n117, Translation="0,0,0", Rotation="0,0,0", Scale="1,1,1") -> (Pose) {
    alphaInputType=Bool
    rotationMode=BMM_Additive
    rotationSpace=BCS_BoneSpace
    scaleMode=BMM_Additive
    scaleSpace=BCS_BoneSpace
    translationMode=BMM_Additive
    translationSpace=BCS_BoneSpace
    alphaBoolBlend {blendOption=Linear}
    alphaScaleBiasClamp {bInterpResult=true, interpSpeedDecreasing=20, interpSpeedIncreasing=20}
    boneToModify {boneName=dyn_leaf_i_02}
   }
   n123 = "Transform (Modify) Bone - Bone: dyn_leaf_g_01"!(ComponentPose=@n122, bAlphaBoolEnabled:bool=@n117, Translation="0,0,0", Rotation="0,-7,0", Scale="1,1,1") -> (Pose) {
    alphaInputType=Bool
    rotationMode=BMM_Additive
    rotationSpace=BCS_BoneSpace
    scaleMode=BMM_Additive
    scaleSpace=BCS_BoneSpace
    translationMode=BMM_Additive
    translationSpace=BCS_BoneSpace
    alphaBoolBlend {blendOption=Linear}
    alphaScaleBiasClamp {bInterpResult=true, interpSpeedDecreasing=20, interpSpeedIncreasing=20}
    boneToModify {boneName=dyn_leaf_g_01}
    rotation {yaw=-7}
   }
   n124 = "Transform (Modify) Bone - Bone: dyn_leaf_hi_01"(
    ComponentPose=@n123
    bAlphaBoolEnabled=Get bIsWearingCape() -> (bIsWearingCape:bool)
    Translation="0,0,0"
    Rotation="0,0,0"
    Scale="0.8,0.8,0.8"
   ) {
    alphaInputType=Bool
    rotationMode=BMM_Additive
    rotationSpace=BCS_BoneSpace
    scaleMode=BMM_Additive
    scaleSpace=BCS_BoneSpace
    translationMode=BMM_Additive
    translationSpace=BCS_BoneSpace
    alphaBoolBlend {blendOption=Linear}
    alphaScaleBiasClamp {bInterpResult=true, interpSpeedDecreasing=20, interpSpeedIncreasing=20}
    boneToModify {boneName=dyn_leaf_hi_01}
    scale=<0.8 0.8 0.8>
   }
   n125 = "Transform (Modify) Bone - Bone: dyn_leaf_d_01"(ComponentPose=@n124, Alpha=Get RightArmLift(), Translation="0,0,1", Rotation="0,0,20", Scale="1,1,1") {
    rotationMode=BMM_Additive
    rotationSpace=BCS_BoneSpace
    scaleMode=BMM_Additive
    scaleSpace=BCS_BoneSpace
    translationMode=BMM_Additive
    translationSpace=BCS_BoneSpace
    alphaBoolBlend {blendOption=Linear}
    alphaScaleBiasClamp {bInterpResult=true, interpSpeedDecreasing=20, interpSpeedIncreasing=20}
    boneToModify {boneName=dyn_leaf_d_01}
    rotation {roll=20}
    translation {z=1}
   }
   n126 = "Transform (Modify) Bone - Bone: dyn_leaf_hi_01"(ComponentPose=@n125, bAlphaBoolEnabled=True, Translation="0,0,0", Rotation="0,0,0", Scale="1,1,1") {
    alphaInputType=Bool
    rotationMode=BMM_Additive
    rotationSpace=BCS_BoneSpace
    scaleMode=BMM_Additive
    scaleSpace=BCS_BoneSpace
    translationMode=BMM_Additive
    translationSpace=BCS_BoneSpace
    alphaBoolBlend {blendInTime=0.25, blendOption=Linear, blendOutTime=0.1}
    alphaScaleBiasClamp {bInterpResult=true, interpSpeedDecreasing=20, interpSpeedIncreasing=20}
    boneToModify {boneName=dyn_leaf_hi_01}
   }
   n127 = "Constraint - dyn_leaf_b_00 (PPPPPP)"!(
    ComponentPose=@n126
    Alpha=1
    ConstraintWeights_0=0.023
    ConstraintWeights_1:float=0.6843
    ConstraintWeights_2:float=0.0247
    ConstraintWeights_3:float=0.0151
    ConstraintWeights_4:float=0.0272
    ConstraintWeights_5:float=0.2257
   ) -> (Pose) {
    alphaBoolBlend {blendOption=Linear}
    boneToModify {boneName=dyn_leaf_b_00}
    constraintSetup [
     {targetBone {boneName=upperarm_r}}
     {targetBone {boneName=upperarm_twist_01_r}}
     {targetBone {boneName=upperarm_twist_02_r}}
     {targetBone {boneName=spine_04}}
     {targetBone {boneName=spine_05}}
     {targetBone {boneName=clavicle_r}}
    ]
    constraintWeights [1,1,1,1,1,1]
   }
   n128 = "Transform (Modify) Bone - Bone: dyn_leaf_c_01"(ComponentPose=@n127, Alpha=1, Translation="0,0,0", Rotation="10,0,0", Scale="1,1,1") {
    rotationMode=BMM_Additive
    rotationSpace=BCS_BoneSpace
    scaleMode=BMM_Additive
    scaleSpace=BCS_BoneSpace
    translationMode=BMM_Additive
    translationSpace=BCS_BoneSpace
    alphaBoolBlend {blendOption=Linear}
    alphaScaleBiasClamp {bInterpResult=true, interpSpeedDecreasing=20, interpSpeedIncreasing=20}
    boneToModify {boneName=dyn_leaf_c_01}
    rotation {pitch=10}
   }
   n129 = "Transform (Modify) Bone - Bone: dyn_leaf_b_00"(ComponentPose=@n128, Alpha=1, Translation="0,0,1", Rotation="0,0,-20", Scale="1,1,1") {
    rotationMode=BMM_Additive
    rotationSpace=BCS_BoneSpace
    scaleMode=BMM_Additive
    scaleSpace=BCS_BoneSpace
    translationMode=BMM_Additive
    translationSpace=BCS_BoneSpace
    alphaBoolBlend {blendOption=Linear}
    alphaScaleBiasClamp {bInterpResult=true, interpSpeedDecreasing=20, interpSpeedIncreasing=20}
    boneToModify {boneName=dyn_leaf_b_00}
    rotation {roll=-20}
    translation {z=1}
   }
   n130 = "Constraint - dyn_shoulderPad_c_01 (P)"(ComponentPose=@n129, Alpha=0.8, ConstraintWeights_0=1) {
    alphaBoolBlend {blendOption=Linear}
    boneToModify {boneName=dyn_shoulderPad_c_01}
    constraintSetup [{targetBone {boneName=dyn_leaf_c_039}}]
    constraintWeights [1]
   }
   n131 = "Transform (Modify) Bone - Bone: dyn_leaf_hi_01"(
    ComponentPose=@n130
    bAlphaBoolEnabled=Get bIsInAirWithoutCape() -> (bIsInAirWithoutCape:bool)
    Translation="0,0,0"
    Rotation="30,0,0"
    Scale="1,1,1"
   ) {
    alphaInputType=Bool
    rotationMode=BMM_Additive
    rotationSpace=BCS_BoneSpace
    scaleMode=BMM_Additive
    scaleSpace=BCS_BoneSpace
    translationMode=BMM_Additive
    translationSpace=BCS_BoneSpace
    alphaBoolBlend {blendInTime=0.2, blendOption=Linear, blendOutTime=0.1}
    alphaScaleBiasClamp {bInterpResult=true, interpSpeedDecreasing=20, interpSpeedIncreasing=20}
    boneToModify {boneName=dyn_leaf_hi_01}
    rotation {pitch=30}
   }
   n132 = "Transform (Modify) Bone - Bone: dyn_leaf_g_01"(
    ComponentPose=@n131
    bAlphaBoolEnabled=Get bIsSkydivingAndNotGliding() -> (bIsSkydivingAndNotGliding:bool)
    Translation="0,0,0"
    Rotation="20,0,0"
    Scale="1,1,1"
   ) {
    alphaInputType=Bool
    rotationMode=BMM_Additive
    rotationSpace=BCS_BoneSpace
    scaleMode=BMM_Additive
    scaleSpace=BCS_BoneSpace
    translationMode=BMM_Additive
    translationSpace=BCS_BoneSpace
    alphaBoolBlend {blendInTime=0.2, blendOption=Linear, blendOutTime=0.1}
    alphaScaleBiasClamp {bInterpResult=true, interpSpeedDecreasing=20, interpSpeedIncreasing=20}
    boneToModify {boneName=dyn_leaf_g_01}
    rotation {pitch=20}
   }
   Save cached pose 'PreFrontEnd'(Pose=Component To Local(ComponentPose=@n132)) {cacheName=PreFrontEnd}
  ]}
 EventGraph {nodes [
   Event BlueprintUpdateAnimation()
   > n1 = Sequence()
   Set FrontEnd_Bling(execute=@n1.then_1, FrontEnd_Bling:bool=AND Boolean(A=Get bIsInFrontEnd(), B=Get bIsWearingBackpackOrCape())) -> (Output_Get:bool)
   n2 = Gravity Override State Function DEPRECATED(
    execute=@n1.then_0
    CharacterPartAnimInstance=Self-Reference()
    GameTimeInSeconds=GetGameTimeInSeconds()
    GravityOverrideParams=Get Gravity Override Params()
   )
   > Set GravityOverride(GravityOverride=@n2)
   Event OnCharacterCustomizationLoaded()
   > DetectBackbling(self:CCP Anim Instance Body Critter Cuddle Object Reference)
  ]}
]
variables [
 FrontEnd_Bling {type=boolean}
 Gravity Override Params {type=GravityOverrideParamsStruct}
 GravityOverride {type=Vector}
]
