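# Control Rig for the face accessory part of the M_MED_Critter_Cuddle character body that adds procedural secondary motion to the jaw bones.
# A single ChainHarmonics unit rooted at dyn_jaw2 combines reach, wave and pendulum motion, with debug drawing enabled.
# A timed random float is remapped and smoothed by two accumulate-lerps at different blend rates to vary the X and Z wave amplitudes.
# Other dynamic bones, such as leaves, mustache and neck chain, are not driven here.
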ControlRigBlueprint /BRCosmetics/Characters/Player/Male/Medium/Bodies/M_MED_Critter_Cuddle/Meshes/Parts/M_MED_Critter_Cuddle_FaceAcc_CtrlRig
graphs [
 RigVMFunctionLibrary {links [], nodes []}
 RigVMModel {nodes [
   RigUnit_BeginExecution = &RigUnit_BeginExecution()
   n1 = /Script/RigVM.RigVMFunction_RandomFloat(Duration=3, Maximum=1, Seed=217) {is=struct RigVMUnitNode, template="Random::Execute(in Seed,in Minimum,in Maximum,in Duration,out Result)"}
   /Script/ControlRig.RigUnit_ChainHarmonicsPerItem(
    ChainRoot=(Type=Bone,Name="dyn_jaw2")
    DrawWorldOffset=(Rotation=(X=0,Y=0,Z=0,W=1),Translation=(X=0,Y=0,Z=0),Scale3D=(X=1,Y=1,Z=1))
    Pendulum=(bEnabled=True,PendulumStiffness=2,PendulumGravity=(X=0,Y=0,Z=0),PendulumBlend=0.75,PendulumDrag=0.98,PendulumMaximum=0.1,UnwindAxis=(X=0,Y=1,Z=0),UnwindMinimum=0.2,UnwindMaximum=0.05)
    Reach=(bEnabled=True,ReachTarget=(X=0,Y=0,Z=0),ReachAxis=(X=1,Y=0,Z=0))
    Speed=(X=1,Y=1,Z=1)
    Wave=(bEnabled=True,WaveFrequency=(X=0,Y=300,Z=0),WaveAmplitude=(X=0,Y=0.5,Z=0),WaveOffset=(X=0,Y=1,Z=0),WaveNoise=(X=1.5,Y=0.5,Z=1.5),WaveMinimum=1,WaveMaximum=0)
    WaveCurve=(EditorCurveData=(Keys=((),(Time=1,Value=1)),DefaultValue=340282346638528859811704183484516925440,PreInfinityExtrap=RCCE_Constant,PostInfinityExtrap=RCCE_Constant))
    bDrawDebug=True
    Wave.WaveAmplitude.Z=/Script/RigVM.RigVMFunction_AccumulateFloatLerp(
     Blend=1
     InitialValue=1
     bIntegrateDeltaTime=true
     TargetValue=&RigVMFunction_MathFloatRemap(SourceMaximum=1, SourceMinimum=0, TargetMaximum=0.3, TargetMinimum=0, bClamp=False, Value=@n1.Result).Result
    ) {is=struct RigVMUnitNode, template="AccumulateLerp::Execute(in TargetValue,in InitialValue,in Blend,in bIntegrateDeltaTime,out Result)"}.Result
    Wave.WaveAmplitude.X=&RigVMFunction_AccumulateFloatLerp(
     Blend=12
     InitialValue=1
     bIntegrateDeltaTime=true
     TargetValue=&RigVMFunction_MathFloatRemap(SourceMaximum=1, SourceMinimum=0, TargetMaximum=0.3, TargetMinimum=0, bClamp=false, Value=@n1.Result).Result
    ) {template="AccumulateLerp::Execute(in TargetValue,in InitialValue,in Blend,in bIntegrateDeltaTime,out Result)"}.Result
    ExecutePin=@RigUnit_BeginExecution.ExecutePin
   ) {is=struct RigVMUnitNode}
  ]}
]
hierarchy [
 root {loc=<0 0 0>, rot=<0 0 0 1>}
 pelvis {parent=root, loc=<0 -0.7071 93.1191>, rot=<0.0127 0.707 -0.0127 -0.707>}
 spine_01 {parent=pelvis, loc=<6.7494 -0.6388 0>, rot=<0 0 0.034 -0.9994>}
 spine_02 {parent=spine_01, loc=<7.2518 -0.035 0>, rot=<0 0 0.0238 -0.9997>}
 spine_03 {parent=spine_02, loc=<6.0104 0 0>, rot=<0 0 0.0149 -0.9999>}
 spine_04 {parent=spine_03, loc=<8.1024 0 0>, rot=<0 0 -0.0132 -0.9999>}
 spine_05 {parent=spine_04, loc=<10.0146 0 0>, rot=<0 0 -0.0015 -1>}
 clavicle_l {parent=spine_05, loc=<4.3575 -0.1466 -2.3223>, rot=<0.0465 -0.7289 -0.009 -0.6829>}
 clavicle_r {parent=spine_05, loc=<4.3575 -0.1466 2.3223>, rot=<0.7289 0.0465 -0.6829 0.009>}
 neck_01 {parent=spine_05, loc=<11.3138 0.1529 0>, rot=<0 0 -0.2003 -0.9797>}
 neck_02 {parent=neck_01, loc=<4.5627 0 0>, rot=<0 0 0.0504 -0.9987>}
 head {parent=neck_02, loc=<5.3914 0 0>, rot=<0 0 0.115 -0.9934>}
 dyn_eye {parent=head, loc=<5.3662 10.4013 7.1387>, rot=<-0.7042 -0.71 0 0>}
 dyn_ear {parent=head, loc=<13.6555 3.8413 11.5577>, rot=<-0.0006 0.1381 0.9904 -0.0041>}
 dyn_tomato_leaf_a_MSTR {parent=head, loc=<17.464 9.4765 -2.8079>, rot=<0.0039 -0.9548 -0.2972 0.0012>}
 dyn_tomato_leaf_a_01 {parent=dyn_tomato_leaf_a_MSTR, loc=<0 0 0>, rot=<0 0 0 1>}
 dyn_tomato_leaf_a_02 {parent=dyn_tomato_leaf_a_01, loc=<3.4905 -0.5891 5.1354>, rot=<-0.0961 0 0 -0.9954>}
 dyn_tomato_leaf_a_03 {parent=dyn_tomato_leaf_a_02, loc=<6.0115 -2.476 4.4035>, rot=<0 0 0 1>}
 dyn_tomato_leaf_b_MSTR {parent=head, loc=<20.2457 6.2814 -2.9553>, rot=<0.0041 -0.9915 -0.1299 0.0005>}
 dyn_tomato_leaf_b_01 {parent=dyn_tomato_leaf_b_MSTR, loc=<0 0 0>, rot=<0 0 0 -1>}
 dyn_tomato_leaf_b_02 {parent=dyn_tomato_leaf_b_01, loc=<-0.3538 0.0074 4.3489>, rot=<0.0675 0 0 -0.9977>}
 dyn_tomato_leaf_b_03 {parent=dyn_tomato_leaf_b_02, loc=<6.3833 -1.9362 3.853>, rot=<-0.329 0 0 -0.9443>}
 dyn_tomato_leaf_b_04 {parent=dyn_tomato_leaf_b_03, loc=<2.1922 -0.194 4.9245>, rot=<-0.0509 0 0 -0.9987>}
 dyn_tomato_leaf_c_MSTR {parent=dyn_tomato_leaf_b_MSTR, loc=<-0.4061 -4.1225 1.6619>, rot=<-0.4931 0 0 -0.87>}
 dyn_tomato_leaf_c_01 {parent=dyn_tomato_leaf_c_MSTR, loc=<0 0 0>, rot=<0 0 0 1>}
 dyn_tomato_leaf_c_02 {parent=dyn_tomato_leaf_c_01, loc=<2.7479 -1.1097 4.3032>, rot=<0.0428 0 0 -0.9991>}
 dyn_tomato_leaf_d_MSTR {parent=head, loc=<21.7778 -0.7519 -0.6158>, rot=<0.0034 -0.8309 0.5563 -0.0023>}
 dyn_tomato_leaf_d_01 {parent=dyn_tomato_leaf_d_MSTR, loc=<0 0 0>, rot=<0 0 0 1>}
 dyn_tomato_leaf_d_02 {parent=dyn_tomato_leaf_d_01, loc=<3.7463 -0.7187 5.5368>, rot=<-0.0909 0 0 -0.9959>}
 dyn_tomato_leaf_d_03 {parent=dyn_tomato_leaf_d_02, loc=<2.118 -0.9226 4.3585>, rot=<-0.06 0 0 -0.9982>}
 dyn_tomato_leaf_e_MSTR {parent=head, loc=<21.2706 1.0648 3.2384>, rot=<0.0022 -0.5256 0.8507 -0.0035>}
 dyn_tomato_leaf_e_01 {parent=dyn_tomato_leaf_e_MSTR, loc=<0 0 0>, rot=<0 0 0 -1>}
 dyn_tomato_leaf_e_02 {parent=dyn_tomato_leaf_e_01, loc=<2.2125 -0.39 4.977>, rot=<-0.148 0 0 -0.989>}
 dyn_brow_01 {parent=head, loc=<9.9314 13.2084 8.6376>, rot=<0.4941 0.4968 0.5017 -0.5073>}
 dyn_brow_02 {parent=head, loc=<8.1079 11.216 11.6247>, rot=<0.3524 0.4205 0.6363 -0.5423>}
 dyn_tomato_stem_MSTR {parent=head, loc=<17.3988 5.9941 0.3532>, rot=<0.0029 -0.7071 -0.0029 -0.7071>}
 dyn_tomato_stem {parent=dyn_tomato_stem_MSTR, loc=<0.409 0.0104 4.9531>, rot=<0.5 -0.5 -0.5 -0.5>}
 dyn_mustache_MSTR {parent=head, loc=<0.3315 14.0319 -5.2395>, rot=<0.0041 -1 0 0>}
 dyn_mustache_01 {parent=dyn_mustache_MSTR, loc=<0 0 0>, rot=<0 0 0 1>}
 dyn_mustache_02 {parent=dyn_mustache_01, loc=<0.8345 -2.8973 5.3793>, rot=<0 0 0 1>}
 dyn_jaw2 {parent=head, loc=<0.0695 3.2665 0>, rot=<0.3941 0.5871 0.5871 0.3941>}
 dyn_jaw1 {parent=dyn_jaw2, loc=<1.7385 0 0.3989>, rot=<0 0 0 -1>}
 dyn_jaw {parent=dyn_jaw1, loc=<1.6261 0 0.3546>, rot=<0 0 0 -1>}
 dyn_chain_MSTR {parent=neck_01, loc=<2.5689 9.0178 0>, rot=<-0.2106 -0.6762 -0.6738 -0.2106>}
 dyn_chain_01 {parent=dyn_chain_MSTR, loc=<0 0 0>, rot=<0 0 0 -1>}
 dyn_chain_02 {parent=dyn_chain_01, loc=<3.2366 0 0>, rot=<0.0021 0.2517 -0.0012 -0.9678>}
 dyn_chain_03 {parent=dyn_chain_02, loc=<3.2475 0 0>, rot=<-0.0012 -0.077 0.001 -0.997>}
 dyn_chain_04 {parent=dyn_chain_03, loc=<3.0593 0 0>, rot=<0.0352 -0.0164 -0.0252 -0.9989>}
 dyn_chain_05 {parent=dyn_chain_04, loc=<2.9885 0 0>, rot=<-0.0372 -0.0894 0.0178 -0.9951>}
]
