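# Physics asset for the face accessory of the male medium Critter Cuddle cosmetic body, covering a dynamic chain and tomato leaves and stem.
# A head sphere and spine capsules form the body base.
# The five-segment chain hangs from spine_05, with wider swing limits than twist.
# Five leaf chains and a lone stem body root at the head, leaf C via B's master, with tight leaf limits and free-rotating master joints.
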PhysicsAsset /BRCosmetics/Characters/Player/Male/Medium/Bodies/M_MED_Critter_Cuddle/Meshes/Parts/M_MED_Critter_Cuddle_FaceAcc_FE_Physics
bodies {
 dyn_chain_01 [{shapeType=Capsule, length=1.7197, radius=1.4949, shapeName=dyn_chain_01_capsule, center=<2.5904 0 0>, rotation=<rot 90 0 0>}]
 dyn_chain_02 [{shapeType=Capsule, length=1.3941, radius=1.8205, shapeName=dyn_chain_02_capsule, center=<2.5904 0 0>, rotation=<rot 90 0 0>}]
 dyn_chain_03 [{shapeType=Capsule, length=0.6287, radius=2.5859, shapeName=dyn_chain_03_capsule, center=<2.5904 0 0>, rotation=<rot 90 0 0>}]
 dyn_chain_04 [{shapeType=Capsule, length=0.6493, radius=2.5652, shapeName=dyn_chain_04_capsule, center=<2.5904 0 0>, rotation=<rot 90 0 0>}]
 dyn_chain_05 [{shapeType=Capsule, length=1.1942, radius=2.0204, shapeName=dyn_chain_05_capsule, center=<2.5904 0 0>, rotation=<rot 90 0 0>}]
 dyn_chain_MSTR [{shapeType=Capsule, length=0.505, radius=0.505, shapeName=dyn_chain_MSTR_capsule, rotation=<rot 0 -90 0>}]
 dyn_tomato_leaf_a_01 [{shapeType=Capsule, length=3, radius=3, shapeName=dyn_tomato_leaf_a_01_capsule, center=<0.158 0 0>}]
 dyn_tomato_leaf_a_02 [{shapeType=Capsule, length=3, radius=3, shapeName=dyn_tomato_leaf_a_02_capsule, center=<0.158 0 0>}]
 dyn_tomato_leaf_a_03 [{shapeType=Capsule, length=3, radius=3, shapeName=dyn_tomato_leaf_a_03_capsule, center=<0.158 0 0>}]
 dyn_tomato_leaf_a_MSTR [{shapeType=Capsule, length=0.505, radius=0.505, shapeName=dyn_tomato_leaf_a_MSTR_capsule}]
 dyn_tomato_leaf_b_01 [{shapeType=Capsule, length=3, radius=3, shapeName=dyn_tomato_leaf_b_01_capsule, center=<0.158 0 0>}]
 dyn_tomato_leaf_b_02 [{shapeType=Capsule, length=3, radius=3, shapeName=dyn_tomato_leaf_b_02_capsule, center=<0.158 0 0>}]
 dyn_tomato_leaf_b_03 [{shapeType=Capsule, length=3, radius=3, shapeName=dyn_tomato_leaf_b_03_capsule, center=<0.158 0 0>}]
 dyn_tomato_leaf_b_04 [{shapeType=Capsule, length=3, radius=3, shapeName=dyn_tomato_leaf_b_04_capsule, center=<0.158 0 0>}]
 dyn_tomato_leaf_b_MSTR [{shapeType=Capsule, length=0.505, radius=0.505, shapeName=dyn_tomato_leaf_b_MSTR_capsule}]
 dyn_tomato_leaf_c_01 [{shapeType=Capsule, length=3, radius=3, shapeName=dyn_tomato_leaf_c_01_capsule, center=<0.158 0 0>}]
 dyn_tomato_leaf_c_02 [{shapeType=Capsule, length=3, radius=3, shapeName=dyn_tomato_leaf_c_02_capsule, center=<0.158 0 0>}]
 dyn_tomato_leaf_c_MSTR [{shapeType=Capsule, length=0.505, radius=0.505, shapeName=dyn_tomato_leaf_c_MSTR_capsule}]
 dyn_tomato_leaf_d_01 [{shapeType=Capsule, length=3, radius=3, shapeName=dyn_tomato_leaf_d_01_capsule, center=<0.158 0 0>}]
 dyn_tomato_leaf_d_02 [{shapeType=Capsule, length=3, radius=3, shapeName=dyn_tomato_leaf_d_02_capsule, center=<0.158 0 0>}]
 dyn_tomato_leaf_d_03 [{shapeType=Capsule, length=3, radius=3, shapeName=dyn_tomato_leaf_d_03_capsule, center=<0.158 0 0>}]
 dyn_tomato_leaf_e_01 [{shapeType=Capsule, length=3, radius=3, shapeName=dyn_tomato_leaf_e_01_capsule, center=<0.158 0 0>}]
 dyn_tomato_leaf_e_02 [{shapeType=Capsule, length=3, radius=3, shapeName=dyn_tomato_leaf_e_02_capsule, center=<0.158 0 0>}]
 dyn_tomato_leaf_e_MSTR [{shapeType=Capsule, length=0.505, radius=0.505, shapeName=dyn_tomato_leaf_e_MSTR_capsule}]
 dyn_tomato_stem_MSTR [{shapeType=Capsule, length=0.505, radius=0.505, shapeName=dyn_tomato_stem_MSTR_capsule}]
 head [{shapeType=Sphere, radius=13.7204, shapeName=head_sphere, center=<6.6272 1.8112 0.5976>}]
 spine_03 [{shapeType=Capsule, length=23.0654, radius=11.111, shapeName=spine_03_capsule}]
 spine_04 [{shapeType=Capsule, length=23.0654, radius=11.111, shapeName=spine_04_capsule}]
 spine_05 [{shapeType=Capsule, length=23.0654, radius=11.111, shapeName=spine_05_capsule}]
}
constraints [
 {
  bone1Name=head
  bone2Name=spine_05
  swing1LimitDegrees=45
  swing1Motion=Free
  swing2LimitDegrees=45
  swing2Motion=Free
  twistLimitDegrees=45
  twistMotion=Free
 }
 {
  bone1Name=spine_05
  bone2Name=spine_04
  swing1LimitDegrees=45
  swing1Motion=Free
  swing2LimitDegrees=45
  swing2Motion=Free
  twistLimitDegrees=45
  twistMotion=Free
 }
 {
  bone1Name=spine_04
  bone2Name=spine_03
  swing1LimitDegrees=45
  swing1Motion=Free
  swing2LimitDegrees=45
  swing2Motion=Free
  twistLimitDegrees=45
  twistMotion=Free
 }
 {
  bone1Name=dyn_chain_MSTR
  bone2Name=spine_05
  swing1LimitDegrees=45
  swing1Motion=Free
  swing2LimitDegrees=45
  swing2Motion=Free
  twistLimitDegrees=45
  twistMotion=Free
 }
 {bone1Name=dyn_chain_01, bone2Name=dyn_chain_MSTR, swing1LimitDegrees=45, swing2LimitDegrees=45, twistLimitDegrees=5}
 {bone1Name=dyn_chain_02, bone2Name=dyn_chain_01, swing1LimitDegrees=45, swing2LimitDegrees=45, twistLimitDegrees=5}
 {bone1Name=dyn_chain_03, bone2Name=dyn_chain_02, swing1LimitDegrees=45, swing2LimitDegrees=45, twistLimitDegrees=5}
 {bone1Name=dyn_chain_04, bone2Name=dyn_chain_03, swing1LimitDegrees=45, swing2LimitDegrees=45, twistLimitDegrees=5}
 {bone1Name=dyn_chain_05, bone2Name=dyn_chain_04, swing1LimitDegrees=45, swing2LimitDegrees=45, twistLimitDegrees=5}
 {
  bone1Name=dyn_tomato_leaf_a_MSTR
  bone2Name=head
  swing1LimitDegrees=45
  swing1Motion=Free
  swing2LimitDegrees=45
  swing2Motion=Free
  twistLimitDegrees=45
  twistMotion=Free
 }
 {
  bone1Name=dyn_tomato_leaf_b_MSTR
  bone2Name=head
  swing1LimitDegrees=45
  swing1Motion=Free
  swing2LimitDegrees=45
  swing2Motion=Free
  twistLimitDegrees=45
  twistMotion=Free
 }
 {
  bone1Name=dyn_tomato_leaf_c_MSTR
  bone2Name=dyn_tomato_leaf_b_MSTR
  swing1LimitDegrees=45
  swing1Motion=Free
  swing2LimitDegrees=45
  swing2Motion=Free
  twistLimitDegrees=45
  twistMotion=Free
 }
 {
  bone1Name=dyn_tomato_leaf_e_MSTR
  bone2Name=head
  swing1LimitDegrees=45
  swing1Motion=Free
  swing2LimitDegrees=45
  swing2Motion=Free
  twistLimitDegrees=45
  twistMotion=Free
 }
 {
  bone1Name=dyn_tomato_stem_MSTR
  bone2Name=head
  swing1LimitDegrees=45
  swing1Motion=Free
  swing2LimitDegrees=45
  swing2Motion=Free
  twistLimitDegrees=45
  twistMotion=Free
 }
 {bone1Name=dyn_tomato_leaf_a_01, bone2Name=dyn_tomato_leaf_a_MSTR, swing1LimitDegrees=5, swing2LimitDegrees=5, twistLimitDegrees=5}
 {bone1Name=dyn_tomato_leaf_a_02, bone2Name=dyn_tomato_leaf_a_01, swing1LimitDegrees=5, swing2LimitDegrees=5, twistLimitDegrees=5}
 {bone1Name=dyn_tomato_leaf_a_03, bone2Name=dyn_tomato_leaf_a_02, swing1LimitDegrees=5, swing2LimitDegrees=5, twistLimitDegrees=5}
 {bone1Name=dyn_tomato_leaf_b_01, bone2Name=dyn_tomato_leaf_b_MSTR, swing1LimitDegrees=5, swing2LimitDegrees=5, twistLimitDegrees=5}
 {bone1Name=dyn_tomato_leaf_b_02, bone2Name=dyn_tomato_leaf_b_01, swing1LimitDegrees=5, swing2LimitDegrees=5, twistLimitDegrees=5}
 {bone1Name=dyn_tomato_leaf_b_03, bone2Name=dyn_tomato_leaf_b_02, swing1LimitDegrees=5, swing2LimitDegrees=5, twistLimitDegrees=5}
 {bone1Name=dyn_tomato_leaf_b_04, bone2Name=dyn_tomato_leaf_b_03, swing1LimitDegrees=5, swing2LimitDegrees=5, twistLimitDegrees=5}
 {bone1Name=dyn_tomato_leaf_c_01, bone2Name=dyn_tomato_leaf_c_MSTR, swing1LimitDegrees=5, swing2LimitDegrees=5, twistLimitDegrees=5}
 {bone1Name=dyn_tomato_leaf_c_02, bone2Name=dyn_tomato_leaf_c_01, swing1LimitDegrees=5, swing2LimitDegrees=5, twistLimitDegrees=5}
 {bone1Name=dyn_tomato_leaf_d_01, bone2Name=head, swing1LimitDegrees=5, swing2LimitDegrees=5, twistLimitDegrees=5}
 {bone1Name=dyn_tomato_leaf_d_03, bone2Name=dyn_tomato_leaf_d_02, swing1LimitDegrees=5, swing2LimitDegrees=5, twistLimitDegrees=5}
 {bone1Name=dyn_tomato_leaf_d_02, bone2Name=dyn_tomato_leaf_d_01, swing1LimitDegrees=5, swing2LimitDegrees=5, twistLimitDegrees=5}
 {bone1Name=dyn_tomato_leaf_e_01, bone2Name=dyn_tomato_leaf_e_MSTR, swing1LimitDegrees=5, swing2LimitDegrees=5, twistLimitDegrees=5}
 {bone1Name=dyn_tomato_leaf_e_02, bone2Name=dyn_tomato_leaf_e_01, swing1LimitDegrees=5, swing2LimitDegrees=5, twistLimitDegrees=5}
]
