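# Post-processes the face accessory part of the Cuddle critter player cosmetic, following the attached parent mesh's pose.
# A cached pose applies the jaw constraint and Control Rig when not moving, blending to the plain copied pose when moving.
# Weighted head-to-shoulder constraints place bolt and chain, with head-limit and targeting-gated bolt rigs between them.
# Eye pitch offsets, per-bone AnimDynamics for leaves, stem, chain, mustache, brows and ear, then RigidBody follow.
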AnimBlueprint /BRCosmetics/Characters/Player/Male/Medium/Bodies/M_MED_Critter_Cuddle/Meshes/Parts/M_MED_Critter_Cuddle_FaceAcc_AnimBP
target_skeleton=&/M_MED_Critter_Cuddle_FaceAcc_Skeleton
graphs [
 AnimGraph {nodes [
   n1 = "Constraint - dyn_bolt (PPPPPP)"!(
    ComponentPose=Local To Component(LocalPose=Use cached pose 'jawmovement'())
    Alpha=1
    ConstraintWeights_0=0.0391
    ConstraintWeights_1:float=0.0005
    ConstraintWeights_2:float=0.5646
    ConstraintWeights_3:float=0.0006
    ConstraintWeights_4:float=0.194
    ConstraintWeights_5:float=0.2012
   ) -> (Pose) {
    alphaBoolBlend {blendOption=Linear}
    boneToModify {boneName=dyn_bolt}
    constraintSetup [
     {targetBone {boneName=head}}
     {targetBone {boneName=clavicle_l}}
     {targetBone {boneName=spine_05}}
     {targetBone {boneName=clavicle_r}}
     {targetBone {boneName=neck_02}}
     {targetBone {boneName=neck_01}}
    ]
    constraintWeights [1,1,1,1,1,1]
   }
   n2 = Control Rig(Source=Component To Local(ComponentPose=@n1)) {
    alphaBoolBlend {blendOption=Linear}
    controlRigAssetReference {blueprintRigClass=&/M_MED_Critter_Cuddle_headlimit_CtrlRig.M_MED_Critter_Cuddle_headlimit_CtrlRig_C}
   }
   n3 = Control Rig!(Source=@n2, bAlphaBoolEnabled:bool=Get bIsNotTargeting() -> (bIsNotTargeting:bool)) -> (Pose) {
    alphaInputType=Bool
    alphaBoolBlend {blendOption=Linear}
    controlRigAssetReference {blueprintRigClass=&/M_MED_Critter_Cuddle_bolt_CtrlRig.M_MED_Critter_Cuddle_bolt_CtrlRig_C}
   }
   n4 = "Constraint - dyn_chain_MSTR (PPPPPP)"!(
    ComponentPose=Local To Component(LocalPose=@n3)
    Alpha=1
    ConstraintWeights_0=0.0391
    ConstraintWeights_1:float=0.0005
    ConstraintWeights_2:float=0.5646
    ConstraintWeights_3:float=0.0006
    ConstraintWeights_4:float=0.194
    ConstraintWeights_5:float=0.2012
   ) -> (Pose) {
    alphaBoolBlend {blendOption=Linear}
    boneToModify {boneName=dyn_chain_MSTR}
    constraintSetup [
     {targetBone {boneName=head}}
     {targetBone {boneName=clavicle_l}}
     {targetBone {boneName=spine_05}}
     {targetBone {boneName=clavicle_r}}
     {targetBone {boneName=neck_02}}
     {targetBone {boneName=neck_01}}
    ]
    constraintWeights [1,1,1,1,1,1]
   }
   n5 = "Transform (Modify) Bone - Bone: dyn_eye"(ComponentPose=@n4, Alpha=Get EyeForward() -> (EyeForward:float), Translation="0,0,0", Rotation="45,0.000,0.000", Scale="1,1,1") {
    rotationMode=BMM_Additive
    rotationSpace=BCS_BoneSpace
    alphaBoolBlend {blendOption=Linear}
    alphaScaleBiasClamp {bInterpResult=true, interpSpeedDecreasing=15, interpSpeedIncreasing=15}
    boneToModify {boneName=dyn_eye}
    rotation {pitch=45}
   }
   n6 = "Transform (Modify) Bone - Bone: dyn_eye"(ComponentPose=@n5, Alpha=Get EyeBackward() -> (EyeBackward:float), Translation="0,0,0", Rotation="-45,0.000,0.000", Scale="1,1,1") {
    rotationMode=BMM_Additive
    rotationSpace=BCS_BoneSpace
    alphaBoolBlend {blendOption=Linear}
    alphaScaleBiasClamp {bInterpResult=true, interpSpeedDecreasing=15, interpSpeedIncreasing=15}
    boneToModify {boneName=dyn_eye}
    rotation {pitch=-45}
   }
   n7 = Anim Dynamics - Bone: dyn_tomato_leaf_a_01(ComponentPose=@n6, Alpha=1, ExternalForce="0,0,0") {
    angularDampingOverride=0.9999
    angularSpringConstant=200
    bAngularSpring=true
    bGravityOverrideInSimSpace=true
    bOverrideAngularDamping=true
    bUseGravityOverride=true
    simulationSpace=BoneRelative
    alphaBoolBlend {blendOption=Linear}
    boundBone {boneName=dyn_tomato_leaf_a_01}
    componentAppliedLinearAccClamp=<3500 3500 3500>
    componentLinearAccScale=<0.6 0.6 0.6>
    componentLinearVelScale=<0.6 0.6 0.6>
    gravityOverride {z=-720}
    physicsBodyDefinitions [{
      sphereCollisionRadius=0
      boundBone {boneName=dyn_tomato_leaf_a_01}
      boxExtents=<25 15 15>
      constraintSetup {angularLimitsMax=<45 0 20>, angularLimitsMin=<-45 -30 -5>, angularTarget=<0 0 0>}
      localJointOffset=<1 0 0>
     }]
    relativeSpaceBone {boneName=root}
   }
   n8 = Anim Dynamics - Bone: dyn_tomato_leaf_a_02(ComponentPose=@n7, Alpha=1, ExternalForce="0,0,0") {
    angularDampingOverride=0.9999
    angularSpringConstant=200
    bAngularSpring=true
    bGravityOverrideInSimSpace=true
    bOverrideAngularDamping=true
    bUseGravityOverride=true
    simulationSpace=BoneRelative
    alphaBoolBlend {blendOption=Linear}
    boundBone {boneName=dyn_tomato_leaf_a_02}
    componentAppliedLinearAccClamp=<3500 3500 3500>
    componentLinearAccScale=<0.6 0.6 0.6>
    componentLinearVelScale=<0.6 0.6 0.6>
    gravityOverride {z=-720}
    physicsBodyDefinitions [{
      sphereCollisionRadius=0
      boundBone {boneName=dyn_tomato_leaf_a_02}
      boxExtents=<25 15 15>
      constraintSetup {angularLimitsMax=<45 0 20>, angularLimitsMin=<-45 -30 -5>, angularTarget=<0 0 0>}
      localJointOffset=<1 0 0>
     }]
    relativeSpaceBone {boneName=root}
   }
   n9 = Anim Dynamics - Bone: dyn_tomato_leaf_a_03(ComponentPose=@n8, Alpha=1, ExternalForce="0,0,0") {
    angularDampingOverride=0.9999
    angularSpringConstant=200
    bAngularSpring=true
    bGravityOverrideInSimSpace=true
    bOverrideAngularDamping=true
    bUseGravityOverride=true
    simulationSpace=BoneRelative
    alphaBoolBlend {blendOption=Linear}
    boundBone {boneName=dyn_tomato_leaf_a_03}
    componentAppliedLinearAccClamp=<3500 3500 3500>
    componentLinearAccScale=<0.6 0.6 0.6>
    componentLinearVelScale=<0.6 0.6 0.6>
    gravityOverride {z=-720}
    physicsBodyDefinitions [{
      sphereCollisionRadius=0
      boundBone {boneName=dyn_tomato_leaf_a_03}
      boxExtents=<25 15 15>
      constraintSetup {angularLimitsMax=<45 5 20>, angularLimitsMin=<-45 -30 -5>, angularTarget=<0 0 0>}
      localJointOffset=<1 0 0>
     }]
    relativeSpaceBone {boneName=root}
   }
   n10 = Anim Dynamics - Bone: dyn_tomato_leaf_b_01(ComponentPose=@n9, Alpha=1, ExternalForce="0,0,0") {
    angularDampingOverride=0.9999
    angularSpringConstant=100
    bAngularSpring=true
    bGravityOverrideInSimSpace=true
    bOverrideAngularDamping=true
    bUseGravityOverride=true
    simulationSpace=BoneRelative
    alphaBoolBlend {blendOption=Linear}
    boundBone {boneName=dyn_tomato_leaf_b_01}
    componentAppliedLinearAccClamp=<3500 3500 3500>
    componentLinearAccScale=<0.8 0.8 0.8>
    componentLinearVelScale=<0.8 0.8 0.8>
    gravityOverride {z=-980}
    physicsBodyDefinitions [{
      sphereCollisionRadius=0
      boundBone {boneName=dyn_tomato_leaf_b_01}
      boxExtents=<20 10 10>
      constraintSetup {angularLimitsMax=<45 0 20>, angularLimitsMin=<-45 -30 -5>, angularTarget=<0 0 0>}
      localJointOffset=<2 0 0>
     }]
    relativeSpaceBone {boneName=root}
   }
   n11 = Anim Dynamics - Bone: dyn_tomato_leaf_b_02(ComponentPose=@n10, Alpha=1, ExternalForce="0,0,0") {
    angularDampingOverride=0.9999
    angularSpringConstant=100
    bAngularSpring=true
    bGravityOverrideInSimSpace=true
    bOverrideAngularDamping=true
    bUseGravityOverride=true
    simulationSpace=BoneRelative
    alphaBoolBlend {blendOption=Linear}
    boundBone {boneName=dyn_tomato_leaf_b_02}
    componentAppliedLinearAccClamp=<3500 3500 3500>
    componentLinearAccScale=<0.8 0.8 0.8>
    componentLinearVelScale=<0.8 0.8 0.8>
    gravityOverride {z=-980}
    physicsBodyDefinitions [{
      sphereCollisionRadius=0
      boundBone {boneName=dyn_tomato_leaf_b_02}
      boxExtents=<20 10 10>
      constraintSetup {angularLimitsMax=<45 0 20>, angularLimitsMin=<-45 -30 -5>, angularTarget=<0 0 0>}
      localJointOffset=<2 0 0>
     }]
    relativeSpaceBone {boneName=root}
   }
   n12 = Anim Dynamics - Bone: dyn_tomato_leaf_b_03(ComponentPose=@n11, Alpha=1, ExternalForce="0,0,0") {
    angularDampingOverride=0.9999
    angularSpringConstant=100
    bAngularSpring=true
    bGravityOverrideInSimSpace=true
    bOverrideAngularDamping=true
    bUseGravityOverride=true
    simulationSpace=BoneRelative
    alphaBoolBlend {blendOption=Linear}
    boundBone {boneName=dyn_tomato_leaf_b_03}
    componentAppliedLinearAccClamp=<3500 3500 3500>
    componentLinearAccScale=<0.8 0.8 0.8>
    componentLinearVelScale=<0.8 0.8 0.8>
    gravityOverride {z=-980}
    physicsBodyDefinitions [{
      sphereCollisionRadius=0
      boundBone {boneName=dyn_tomato_leaf_b_03}
      boxExtents=<20 10 10>
      constraintSetup {angularLimitsMax=<45 5 20>, angularLimitsMin=<-45 -30 -5>, angularTarget=<0 0 0>}
      localJointOffset=<2 0 0>
     }]
    relativeSpaceBone {boneName=root}
   }
   n13 = Anim Dynamics - Bone: dyn_tomato_leaf_b_04(ComponentPose=@n12, Alpha=1, ExternalForce="0,0,0") {
    angularDampingOverride=0.9999
    angularSpringConstant=100
    bAngularSpring=true
    bGravityOverrideInSimSpace=true
    bOverrideAngularDamping=true
    bUseGravityOverride=true
    simulationSpace=BoneRelative
    alphaBoolBlend {blendOption=Linear}
    boundBone {boneName=dyn_tomato_leaf_b_04}
    componentAppliedLinearAccClamp=<3500 3500 3500>
    componentLinearAccScale=<0.8 0.8 0.8>
    componentLinearVelScale=<0.8 0.8 0.8>
    gravityOverride {z=-980}
    physicsBodyDefinitions [{
      sphereCollisionRadius=0
      boundBone {boneName=dyn_tomato_leaf_b_04}
      boxExtents=<20 10 10>
      constraintSetup {angularLimitsMax=<45 5 20>, angularLimitsMin=<-45 -30 -5>, angularTarget=<0 0 0>}
      localJointOffset=<2 0 0>
     }]
    relativeSpaceBone {boneName=root}
   }
   n14 = Anim Dynamics - Bone: dyn_tomato_leaf_c_01(ComponentPose=@n13, Alpha=1, ExternalForce="0,0,0") {
    angularDampingOverride=0.9999
    angularSpringConstant=200
    bAngularSpring=true
    bGravityOverrideInSimSpace=true
    bOverrideAngularDamping=true
    bUseGravityOverride=true
    simulationSpace=BoneRelative
    alphaBoolBlend {blendOption=Linear}
    boundBone {boneName=dyn_tomato_leaf_c_01}
    componentAppliedLinearAccClamp=<3500 3500 3500>
    componentLinearAccScale=<0.6 0.6 0.6>
    componentLinearVelScale=<0.6 0.6 0.6>
    gravityOverride {z=-720}
    physicsBodyDefinitions [{
      sphereCollisionRadius=0
      boundBone {boneName=dyn_tomato_leaf_c_01}
      boxExtents=<15 10 10>
      constraintSetup {angularLimitsMax=<15 5 15>, angularLimitsMin=<-15 -20 -15>, angularTarget=<0 0 0>}
      localJointOffset=<1 0 0>
     }]
    relativeSpaceBone {boneName=root}
   }
   n15 = Anim Dynamics - Bone: dyn_tomato_leaf_c_02(ComponentPose=@n14, Alpha=1, ExternalForce="0,0,0") {
    angularDampingOverride=0.9999
    angularSpringConstant=200
    bAngularSpring=true
    bGravityOverrideInSimSpace=true
    bOverrideAngularDamping=true
    bUseGravityOverride=true
    simulationSpace=BoneRelative
    alphaBoolBlend {blendOption=Linear}
    boundBone {boneName=dyn_tomato_leaf_c_02}
    componentAppliedLinearAccClamp=<3500 3500 3500>
    componentLinearAccScale=<0.6 0.6 0.6>
    componentLinearVelScale=<0.6 0.6 0.6>
    gravityOverride {z=-720}
    physicsBodyDefinitions [{
      sphereCollisionRadius=0
      boundBone {boneName=dyn_tomato_leaf_c_02}
      boxExtents=<15 10 10>
      constraintSetup {angularLimitsMax=<15 5 15>, angularLimitsMin=<-15 -20 -15>, angularTarget=<0 0 0>}
      localJointOffset=<1 0 0>
     }]
    relativeSpaceBone {boneName=root}
   }
   n16 = Anim Dynamics - Bone: dyn_tomato_leaf_d_01(ComponentPose=@n15, Alpha=1, ExternalForce="0,0,0") {
    angularDampingOverride=0.9999
    angularSpringConstant=100
    bAngularSpring=true
    bGravityOverrideInSimSpace=true
    bOverrideAngularDamping=true
    bUseGravityOverride=true
    simulationSpace=BoneRelative
    alphaBoolBlend {blendOption=Linear}
    boundBone {boneName=dyn_tomato_leaf_d_01}
    componentAppliedLinearAccClamp=<3500 3500 3500>
    componentLinearAccScale=<0.9 0.9 0.9>
    componentLinearVelScale=<0.9 0.9 0.9>
    gravityOverride {z=-720}
    physicsBodyDefinitions [{
      sphereCollisionRadius=0
      boundBone {boneName=dyn_tomato_leaf_d_01}
      boxExtents=<20 10 10>
      constraintSetup {angularLimitsMax=<10 0 15>, angularLimitsMin=<-10 -10 -15>, angularTarget=<0 0 0>}
      localJointOffset=<2 0 0>
     }]
    relativeSpaceBone {boneName=root}
   }
   n17 = Anim Dynamics - Bone: dyn_tomato_leaf_d_02(ComponentPose=@n16, Alpha=1, ExternalForce="0,0,0") {
    angularDampingOverride=0.9999
    angularSpringConstant=100
    bAngularSpring=true
    bGravityOverrideInSimSpace=true
    bOverrideAngularDamping=true
    bUseGravityOverride=true
    simulationSpace=BoneRelative
    alphaBoolBlend {blendOption=Linear}
    boundBone {boneName=dyn_tomato_leaf_d_02}
    componentAppliedLinearAccClamp=<3500 3500 3500>
    componentLinearAccScale=<0.9 0.9 0.9>
    componentLinearVelScale=<0.9 0.9 0.9>
    gravityOverride {z=-720}
    physicsBodyDefinitions [{
      sphereCollisionRadius=0
      boundBone {boneName=dyn_tomato_leaf_d_02}
      boxExtents=<20 10 10>
      constraintSetup {angularLimitsMax=<10 10 15>, angularLimitsMin=<-10 -10 -15>, angularTarget=<0 0 0>}
      localJointOffset=<2 0 0>
     }]
    relativeSpaceBone {boneName=root}
   }
   n18 = Anim Dynamics - Bone: dyn_tomato_leaf_d_03(ComponentPose=@n17, Alpha=1, ExternalForce="0,0,0") {
    angularDampingOverride=0.9999
    angularSpringConstant=100
    bAngularSpring=true
    bGravityOverrideInSimSpace=true
    bOverrideAngularDamping=true
    bUseGravityOverride=true
    simulationSpace=BoneRelative
    alphaBoolBlend {blendOption=Linear}
    boundBone {boneName=dyn_tomato_leaf_d_03}
    componentAppliedLinearAccClamp=<3500 3500 3500>
    componentLinearAccScale=<0.9 0.9 0.9>
    componentLinearVelScale=<0.9 0.9 0.9>
    gravityOverride {z=-720}
    physicsBodyDefinitions [{
      sphereCollisionRadius=0
      boundBone {boneName=dyn_tomato_leaf_d_03}
      boxExtents=<20 10 10>
      constraintSetup {angularLimitsMax=<10 10 15>, angularLimitsMin=<-10 -10 -15>, angularTarget=<0 0 0>}
      localJointOffset=<2 0 0>
     }]
    relativeSpaceBone {boneName=root}
   }
   n19 = Anim Dynamics - Bone: dyn_tomato_leaf_e_01(ComponentPose=@n18, Alpha=1, ExternalForce="0,0,0") {
    angularDampingOverride=0.9999
    angularSpringConstant=100
    bAngularSpring=true
    bGravityOverrideInSimSpace=true
    bOverrideAngularDamping=true
    bUseGravityOverride=true
    simulationSpace=BoneRelative
    alphaBoolBlend {blendOption=Linear}
    boundBone {boneName=dyn_tomato_leaf_e_01}
    componentAppliedLinearAccClamp=<3500 3500 3500>
    componentLinearAccScale=<0.8 0.8 0.8>
    componentLinearVelScale=<0.8 0.8 0.8>
    gravityOverride {z=-720}
    physicsBodyDefinitions [{
      sphereCollisionRadius=0
      boundBone {boneName=dyn_tomato_leaf_e_01}
      boxExtents=<20 10 10>
      constraintSetup {angularLimitsMax=<10 0 15>, angularLimitsMin=<-10 -10 -15>, angularTarget=<0 0 0>}
      localJointOffset=<2 0 0>
     }]
    relativeSpaceBone {boneName=root}
   }
   n20 = Anim Dynamics - Bone: dyn_tomato_leaf_e_02(ComponentPose=@n19, Alpha=1, ExternalForce="0,0,0") {
    angularDampingOverride=0.9999
    angularSpringConstant=100
    bAngularSpring=true
    bGravityOverrideInSimSpace=true
    bOverrideAngularDamping=true
    bUseGravityOverride=true
    simulationSpace=BoneRelative
    alphaBoolBlend {blendOption=Linear}
    boundBone {boneName=dyn_tomato_leaf_e_02}
    componentAppliedLinearAccClamp=<3500 3500 3500>
    componentLinearAccScale=<0.8 0.8 0.8>
    componentLinearVelScale=<0.8 0.8 0.8>
    gravityOverride {z=-720}
    physicsBodyDefinitions [{
      sphereCollisionRadius=0
      boundBone {boneName=dyn_tomato_leaf_e_02}
      boxExtents=<20 10 10>
      constraintSetup {angularLimitsMax=<10 10 15>, angularLimitsMin=<-10 -10 -15>, angularTarget=<0 0 0>}
      localJointOffset=<2 0 0>
     }]
    relativeSpaceBone {boneName=root}
   }
   n21 = Anim Dynamics - Bone: dyn_tomato_stem(ComponentPose=@n20, Alpha=1, ExternalForce="0,0,0") {
    angularDampingOverride=0.9999
    angularSpringConstant=100
    bAngularSpring=true
    bGravityOverrideInSimSpace=true
    bOverrideAngularDamping=true
    bUseGravityOverride=true
    simulationSpace=BoneRelative
    alphaBoolBlend {blendOption=Linear}
    boundBone {boneName=dyn_tomato_stem}
    componentAppliedLinearAccClamp=<3500 3500 3500>
    componentLinearAccScale=<0.3 0.3 0.3>
    componentLinearVelScale=<0.3 0.3 0.3>
    gravityOverride {z=-980}
    physicsBodyDefinitions [{
      sphereCollisionRadius=0
      boundBone {boneName=dyn_tomato_stem}
      boxExtents=<20 15 15>
      constraintSetup {angularLimitsMax=<15 15 15>, angularLimitsMin=<-15 -15 -15>, angularTarget=<0 0 0>}
      localJointOffset=<2 0 0>
     }]
    relativeSpaceBone {boneName=root}
   }
   n22 = Anim Dynamics - Bone: dyn_chain_01(ComponentPose=@n21, Alpha=1, ExternalForce="0,0,0") {
    angularDampingOverride=0.998
    angularSpringConstant=70
    bAngularSpring=true
    bGravityOverrideInSimSpace=true
    bOverrideAngularDamping=true
    bUseGravityOverride=true
    bUsePlanarLimit=false
    simulationSpace=RootRelative
    alphaBoolBlend {blendOption=Linear}
    boundBone {boneName=dyn_chain_01}
    componentAppliedLinearAccClamp=<3000 3000 3000>
    componentLinearAccScale=<0.3 0.3 0.3>
    componentLinearVelScale=<0.3 0.3 0.3>
    gravityOverride {z=-980}
    physicsBodyDefinitions [{
      collisionType=InnerSphere
      sphereCollisionRadius=1
      boundBone {boneName=dyn_chain_01}
      boxExtents=<7 12 12>
      constraintSetup {angularLimitsMax=<5 5 15>, angularLimitsMin=<-15 -35 -15>, angularTarget=<0 0 0>}
      localJointOffset=<2 0 0>
     }]
    relativeSpaceBone {boneName=root}
   }
   n23 = Anim Dynamics - Bone: dyn_chain_02(ComponentPose=@n22, Alpha=1, ExternalForce="0,0,0") {
    angularDampingOverride=0.998
    angularSpringConstant=70
    bAngularSpring=true
    bGravityOverrideInSimSpace=true
    bOverrideAngularDamping=true
    bUseGravityOverride=true
    bUsePlanarLimit=false
    simulationSpace=RootRelative
    alphaBoolBlend {blendOption=Linear}
    boundBone {boneName=dyn_chain_02}
    componentAppliedLinearAccClamp=<3000 3000 3000>
    componentLinearAccScale=<0.3 0.3 0.3>
    componentLinearVelScale=<0.3 0.3 0.3>
    gravityOverride {z=-980}
    physicsBodyDefinitions [{
      collisionType=InnerSphere
      sphereCollisionRadius=1
      boundBone {boneName=dyn_chain_02}
      boxExtents=<7 12 12>
      constraintSetup {angularLimitsMax=<5 5 15>, angularLimitsMin=<-15 -35 -15>, angularTarget=<0 0 0>}
      localJointOffset=<2 0 0>
     }]
    relativeSpaceBone {boneName=root}
   }
   n24 = Anim Dynamics - Bone: dyn_chain_03(ComponentPose=@n23, Alpha=1, ExternalForce="0,0,0") {
    angularDampingOverride=0.998
    angularSpringConstant=70
    bAngularSpring=true
    bGravityOverrideInSimSpace=true
    bOverrideAngularDamping=true
    bUseGravityOverride=true
    bUsePlanarLimit=false
    simulationSpace=RootRelative
    alphaBoolBlend {blendOption=Linear}
    boundBone {boneName=dyn_chain_03}
    componentAppliedLinearAccClamp=<3000 3000 3000>
    componentLinearAccScale=<0.3 0.3 0.3>
    componentLinearVelScale=<0.3 0.3 0.3>
    gravityOverride {z=-980}
    physicsBodyDefinitions [{
      collisionType=InnerSphere
      sphereCollisionRadius=1
      boundBone {boneName=dyn_chain_03}
      boxExtents=<7 12 12>
      constraintSetup {angularLimitsMax=<5 5 15>, angularLimitsMin=<-15 -35 -15>, angularTarget=<0 0 0>}
      localJointOffset=<2 0 0>
     }]
    relativeSpaceBone {boneName=root}
   }
   n25 = Anim Dynamics - Bone: dyn_chain_04(ComponentPose=@n24, Alpha=1, ExternalForce="0,0,0") {
    angularDampingOverride=0.998
    angularSpringConstant=70
    bAngularSpring=true
    bGravityOverrideInSimSpace=true
    bOverrideAngularDamping=true
    bUseGravityOverride=true
    bUsePlanarLimit=false
    simulationSpace=RootRelative
    alphaBoolBlend {blendOption=Linear}
    boundBone {boneName=dyn_chain_04}
    componentAppliedLinearAccClamp=<3000 3000 3000>
    componentLinearAccScale=<0.3 0.3 0.3>
    componentLinearVelScale=<0.3 0.3 0.3>
    gravityOverride {z=-980}
    physicsBodyDefinitions [{
      collisionType=InnerSphere
      sphereCollisionRadius=1
      boundBone {boneName=dyn_chain_04}
      boxExtents=<7 12 12>
      constraintSetup {angularLimitsMax=<5 5 15>, angularLimitsMin=<-15 -35 -15>, angularTarget=<0 0 0>}
      localJointOffset=<2 0 0>
     }]
    relativeSpaceBone {boneName=root}
   }
   n26 = Anim Dynamics - Bone: dyn_chain_05(ComponentPose=@n25, Alpha=1, ExternalForce="0,0,0") {
    angularDampingOverride=0.998
    angularSpringConstant=70
    bAngularSpring=true
    bGravityOverrideInSimSpace=true
    bOverrideAngularDamping=true
    bUseGravityOverride=true
    bUsePlanarLimit=false
    simulationSpace=RootRelative
    alphaBoolBlend {blendOption=Linear}
    boundBone {boneName=dyn_chain_05}
    componentAppliedLinearAccClamp=<3000 3000 3000>
    componentLinearAccScale=<0.3 0.3 0.3>
    componentLinearVelScale=<0.3 0.3 0.3>
    gravityOverride {z=-980}
    physicsBodyDefinitions [{
      collisionType=InnerSphere
      sphereCollisionRadius=1
      boundBone {boneName=dyn_chain_05}
      boxExtents=<7 12 12>
      constraintSetup {angularLimitsMax=<5 5 15>, angularLimitsMin=<-15 -35 -15>, angularTarget=<0 0 0>}
      localJointOffset=<2 0 0>
     }]
    relativeSpaceBone {boneName=root}
   }
   n27 = Anim Dynamics - Bone: dyn_mustache_01(ComponentPose=@n26, Alpha=1, ExternalForce="0,0,0") {
    angularDampingOverride=0.9999
    angularSpringConstant=500
    bAngularSpring=true
    bGravityOverrideInSimSpace=true
    bOverrideAngularDamping=true
    bUseGravityOverride=true
    simulationSpace=BoneRelative
    alphaBoolBlend {blendOption=Linear}
    boundBone {boneName=dyn_mustache_01}
    componentAppliedLinearAccClamp=<3500 3500 3500>
    componentLinearAccScale=<0.8 0.8 0.8>
    componentLinearVelScale=<0.8 0.8 0.8>
    gravityOverride {z=-980}
    physicsBodyDefinitions [{
      sphereCollisionRadius=0
      boundBone {boneName=dyn_mustache_01}
      boxExtents=<25 15 15>
      constraintSetup {angularLimitsMax=<10 5 15>, angularLimitsMin=<-10 -10 -15>, angularTarget=<0 0 0>}
      localJointOffset=<2 0 0>
     }]
    relativeSpaceBone {boneName=root}
   }
   n28 = Anim Dynamics - Bone: dyn_mustache_02(ComponentPose=@n27, Alpha=1, ExternalForce="0,0,0") {
    angularDampingOverride=0.9999
    angularSpringConstant=500
    bAngularSpring=true
    bGravityOverrideInSimSpace=true
    bOverrideAngularDamping=true
    bUseGravityOverride=true
    simulationSpace=BoneRelative
    alphaBoolBlend {blendOption=Linear}
    boundBone {boneName=dyn_mustache_02}
    componentAppliedLinearAccClamp=<3500 3500 3500>
    componentLinearAccScale=<0.8 0.8 0.8>
    componentLinearVelScale=<0.8 0.8 0.8>
    gravityOverride {z=-980}
    physicsBodyDefinitions [{
      sphereCollisionRadius=0
      boundBone {boneName=dyn_mustache_02}
      boxExtents=<25 15 15>
      constraintSetup {angularLimitsMax=<10 5 15>, angularLimitsMin=<-10 -10 -15>, angularTarget=<0 0 0>}
      localJointOffset=<2 0 0>
     }]
    relativeSpaceBone {boneName=root}
   }
   n29 = Anim Dynamics - Bone: dyn_brow_01(ComponentPose=@n28, Alpha=1, ExternalForce="0,0,0") {
    angularDampingOverride=0.9999
    angularSpringConstant=500
    bAngularSpring=true
    bGravityOverrideInSimSpace=true
    bOverrideAngularDamping=true
    bUseGravityOverride=true
    simulationSpace=BoneRelative
    alphaBoolBlend {blendOption=Linear}
    boundBone {boneName=dyn_brow_01}
    componentAppliedLinearAccClamp=<3500 3500 3500>
    componentLinearAccScale=<1 1 1>
    componentLinearVelScale=<1 1 1>
    gravityOverride {z=-980}
    physicsBodyDefinitions [{
      sphereCollisionRadius=0
      boundBone {boneName=dyn_brow_01}
      boxExtents=<25 15 15>
      constraintSetup {angularLimitsMax=<10 5 15>, angularLimitsMin=<-10 -10 -15>, angularTarget=<0 0 0>}
      localJointOffset=<2 0 0>
     }]
    relativeSpaceBone {boneName=root}
   }
   n30 = Anim Dynamics - Bone: dyn_brow_02(ComponentPose=@n29, Alpha=1, ExternalForce="0,0,0") {
    angularDampingOverride=0.9999
    angularSpringConstant=500
    bAngularSpring=true
    bGravityOverrideInSimSpace=true
    bOverrideAngularDamping=true
    bUseGravityOverride=true
    simulationSpace=BoneRelative
    alphaBoolBlend {blendOption=Linear}
    boundBone {boneName=dyn_brow_02}
    componentAppliedLinearAccClamp=<3500 3500 3500>
    componentLinearAccScale=<1 1 1>
    componentLinearVelScale=<1 1 1>
    gravityOverride {z=-980}
    physicsBodyDefinitions [{
      sphereCollisionRadius=0
      boundBone {boneName=dyn_brow_02}
      boxExtents=<25 15 15>
      constraintSetup {angularLimitsMax=<10 5 15>, angularLimitsMin=<-10 -10 -15>, angularTarget=<0 0 0>}
      localJointOffset=<2 0 0>
     }]
    relativeSpaceBone {boneName=root}
   }
   n31 = Anim Dynamics - Bone: dyn_ear(ComponentPose=@n30, Alpha=1, ExternalForce="0,0,0") {
    angularDampingOverride=0.9999
    angularSpringConstant=100
    bGravityOverrideInSimSpace=true
    bOverrideAngularDamping=true
    bUseGravityOverride=true
    simulationSpace=BoneRelative
    alphaBoolBlend {blendOption=Linear}
    boundBone {boneName=dyn_ear}
    componentAppliedLinearAccClamp=<3500 3500 3500>
    componentLinearAccScale=<0.3 0.3 0.3>
    componentLinearVelScale=<0.3 0.3 0.3>
    gravityOverride {z=-980}
    physicsBodyDefinitions [{
      sphereCollisionRadius=0
      boundBone {boneName=dyn_ear}
      boxExtents=<35 20 20>
      constraintSetup {angularLimitsMax=<3 10 5>, angularLimitsMin=<-3 -10 -5>, angularTarget=<0 0 0>}
      localJointOffset=<1 0 0>
     }]
    relativeSpaceBone {boneName=root}
   }
   n32 = RigidBody!(ComponentPose=@n31, bAlphaBoolEnabled:bool, OverrideWorldGravity:Vector="0,0,-980", ExternalForce, bTransferBoneVelocities:bool=False) -> (Pose) {
    alphaInputType=Bool
    bClampLinearTranslationLimitToRefPose=true
    bOverrideWorldGravity=true
    overridePhysicsAsset=&/M_MED_Critter_Cuddle_FaceAcc_FE_Physics
    simulationSpace=BaseBoneSpace
    alphaBoolBlend {blendOption=Linear}
    baseBoneRef {boneName=root}
    componentAppliedLinearAccClamp=<5000 5000 5000>
    componentLinearAccScale=<0.4 0.4 0.4>
    componentLinearVelScale=<0.4 0.4 0.4>
    overrideWorldGravity {z=-980}
   }
   Output Pose(Result=Component To Local(ComponentPose=@n32)) {layerGroup=None}
   n33 = "Constraint - dyn_jaw2 (P)"(ComponentPose=Local To Component(LocalPose=Copy Pose From Mesh() {bUseAttachedParent=true}), Alpha=1, ConstraintWeights_0=1) {alphaBoolBlend {blendOption=Linear}, boneToModify {boneName=dyn_jaw2}, constraintSetup [{targetBone {boneName=head}}], constraintWeights [1]}
   n34 = Control Rig!(Source=Component To Local(ComponentPose=@n33), Alpha:float) -> (Pose) {
    alphaBoolBlend {blendOption=Linear}
    controlRigAssetReference {blueprintRigClass=&/M_MED_Critter_Cuddle_FaceAcc_CtrlRig.M_MED_Critter_Cuddle_FaceAcc_CtrlRig_C}
   }
   Save cached pose 'jawmovement'(Pose=Blend Poses by bool(
     BlendPose_0=Copy Pose From Mesh() {bUseAttachedParent=true}
     BlendPose_1=@n34
     BlendTime_1=3
     bActiveValue=Get bIsMoving() -> (bIsMoving:bool)
    ) {blendType=Linear}) {cacheName=jawmovement}
  ]}
 EventGraph {links [], nodes []}
]
variables []
